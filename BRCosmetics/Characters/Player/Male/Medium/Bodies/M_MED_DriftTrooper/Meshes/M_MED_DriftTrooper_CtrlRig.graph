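# Control Rig for the DriftTrooper medium male body that applies pose-driven corrections to its dynamic cloth volume bones.
# Two mirrored chains first aim each thigh control at its foot control.
# Spherical pose readers then measure both thighs against the pelvis and spine_02 against the gravity bone.
# Their outputs weight parent constraints and transform offsets on left and right cloth bones, with B-chain roots driven by the opposite thigh.
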ControlRigBlueprint /BRCosmetics/Characters/Player/Male/Medium/Bodies/M_MED_DriftTrooper/Meshes/M_MED_DriftTrooper_CtrlRig
graphs [
 ContainedGraph {nodes [
   n1 = RigVMFunctionEntryNode()
   SequenceAggregate = &RigVMFunction_Sequence(ExecuteContext=@n1.ExecuteContext)
   SequenceAggregate_1 = &RigVMFunction_Sequence(ExecuteContext=@SequenceAggregate.B)
   SequenceAggregate_2 = &RigVMFunction_Sequence(ExecuteContext=@SequenceAggregate_1.B)
   RigVMFunctionReturnNode(A=@SequenceAggregate.A, B=@SequenceAggregate_1.A, C=@SequenceAggregate_2.A, D=@SequenceAggregate_2.B)
  ]}
 RigVMFunctionLibrary {links [], nodes []}
 RigVMModel {nodes [
   RigUnit_BeginExecution = &RigUnit_BeginExecution()
   SequenceAggregate = RigVMAggregateNode(ExecuteContext=@RigUnit_BeginExecution.ExecutePin)
   &RigUnit_AimConstraintLocalSpaceOffset(
    AdvancedSettings=(DebugSettings=(bEnabled=false,Scale=10,WorldOffset=(Rotation=(X=0,Y=0,Z=0,W=1),Translation=(X=0,Y=0,Z=0),Scale3D=(X=1,Y=1,Z=1))),RotationOrderForFilter=XZY)
    Child=(Type=Control,Name="thigh_r_ctrl")
    Filter=(bX=True,bY=True,bZ=True)
    Parents=((Item=(Type=Control,Name="foot_r_ctrl"),Weight=1))
    Weight=1
    WorldUp=(Target=(X=0,Y=0,Z=1),Kind=Direction,Space=(Type=None,Name="None"))
    bMaintainOffset=true
    ExecutePin=@SequenceAggregate.A
   )
   > n1 = &RigUnit_SphericalPoseReader(
    ActiveRegionScaleFactors=(PositiveWidth=3,NegativeWidth=2,PositiveHeight=4,NegativeHeight=2)
    ActiveRegionSize=0.1
    Debug=(bDrawDebug=false,DebugScale=25,DebugSegments=20,DebugThickness=0.25)
    DriverAxis=(X=-1,Y=0,Z=0)
    DriverItem=(Type=Control,Name="thigh_r_ctrl")
    FalloffRegionScaleFactors=(PositiveWidth=0.3,NegativeWidth=2,PositiveHeight=1,NegativeHeight=1)
    FalloffSize=0.35
    FlipHeightScaling=false
    FlipWidthScaling=false
    OptionalParentItem=(Type=Bone,Name="pelvis")
    RotationOffset=(X=-30,Y=0,Z=0)
   )
   > &RigUnit_ParentConstraint(
    AdvancedSettings=(RotationOrderForFilter=XZY)
    Child=(Type=Bone,Name="dyn_Cloth_volume_l_B")
    Filter=(TranslationFilter=(bX=false,bY=false,bZ=false),RotationFilter=(bX=true,bY=true,bZ=true),ScaleFilter=(bX=True,bY=True,bZ=True))
    Parents=((Item=(Type=Bone,Name="thigh_twist_01_l"),Weight=0),(Item=(Type=Bone,Name="pelvis"),Weight=1))
    Weight=@n1.OutputParam
   )
   > &RigUnit_ModifyTransforms(
    ItemToModify="((Item=(Type=Bone,Name=\"dyn_Cloth_volume_l_B01\"),Transform=(Rotation=(X=0,Y=0,Z=0.2164,W=0.9763),Translation=(X=0,Y=0,Z=0),Scale3D=(X=1,Y=1,Z=1))),(Item=(Type=Bone,Name=\"dyn_Cloth_volume_l_C02\"),Transform=(Rotation=(X=0,Y=0,Z=0.4617,W=0.887),Translation=(X ...(truncated)"
    Weight=@n1.OutputParam
   )
   > n2 = &RigUnit_SphericalPoseReader(
    ActiveRegionScaleFactors=(PositiveWidth=1,NegativeWidth=1,PositiveHeight=1,NegativeHeight=1)
    ActiveRegionSize=0.1
    Debug=(bDrawDebug=false,DebugScale=25,DebugSegments=20,DebugThickness=0.25)
    DriverAxis=(X=-1,Y=0,Z=0)
    DriverItem=(Type=Control,Name="thigh_l_ctrl")
    FalloffRegionScaleFactors=(PositiveWidth=1,NegativeWidth=1,PositiveHeight=1,NegativeHeight=0.8)
    FalloffSize=0.6
    FlipHeightScaling=false
    FlipWidthScaling=false
    OptionalParentItem=(Type=Bone,Name="pelvis")
    RotationOffset=(X=-110,Y=10,Z=20)
   )
   > &RigUnit_ParentConstraint(
    AdvancedSettings=(RotationOrderForFilter=XZY)
    Child=(Type=Bone,Name="dyn_Cloth_volume_l_A")
    Filter=(TranslationFilter=(bX=true,bY=true,bZ=true),RotationFilter=(bX=True,bY=True,bZ=True),ScaleFilter=(bX=True,bY=True,bZ=True))
    Parents=((Item=(Type=Bone,Name="thigh_twist_01_l"),Weight=0.8),(Item=(Type=Bone,Name="pelvis"),Weight=0.2))
    Weight=@n2.OutputParam
   )
   > &RigUnit_ModifyTransforms(
    ItemToModify="((Item=(Type=Bone,Name=\"dyn_Cloth_volume_l_C\"),Transform=(Rotation=(X=0,Y=0,Z=0.1305,W=0.9914),Translation=(X=0,Y=0,Z=0),Scale3D=(X=1,Y=1,Z=1))),(Item=(Type=Bone,Name=\"dyn_Cloth_volume_l_C03\"),Transform=(Rotation=(X=0,Y=0,Z=0.1305,W=0.9914),Translation=(X= ...(truncated)"
    Weight=@n2.OutputParam
   )
   > n3 = &RigUnit_SphericalPoseReader(
    ActiveRegionScaleFactors=(PositiveWidth=3,NegativeWidth=2,PositiveHeight=2,NegativeHeight=2)
    ActiveRegionSize=0.1
    Debug=(bDrawDebug=false,DebugScale=25,DebugSegments=20,DebugThickness=0.25)
    DriverAxis=(X=1,Y=0,Z=0)
    DriverItem=(Type=Bone,Name="spine_02")
    FalloffRegionScaleFactors=(PositiveWidth=1,NegativeWidth=1,PositiveHeight=1,NegativeHeight=1)
    FalloffSize=0.45
    FlipWidthScaling=false
    OptionalParentItem=(Type=Bone,Name="gravity")
    RotationOffset=(X=90,Y=0,Z=5)
   )
   > &RigUnit_ParentConstraint(
    AdvancedSettings=(RotationOrderForFilter=XZY)
    Child=(Type=Bone,Name="dyn_Cloth_volume_l_C02")
    Filter=(TranslationFilter=(bX=false,bY=false,bZ=false),RotationFilter=(bX=True,bY=True,bZ=True),ScaleFilter=(bX=True,bY=True,bZ=True))
    Parents=((Item=(Type=Bone,Name="dyn_Cloth_volume_l_C02"),Weight=0.5),(Item=(Type=Bone,Name="pelvis"),Weight=0.5))
    Weight=@n3.OutputParam
   )
   > &RigUnit_ModifyTransforms(
    ItemToModify="((Item=(Type=Bone,Name=\"dyn_Cloth_volume_l_B01\"),Transform=(Rotation=(X=0,Y=-0.1736,Z=0,W=0.9848),Translation=(X=0,Y=0,Z=0),Scale3D=(X=1,Y=1,Z=1))),(Item=(Type=Bone,Name=\"dyn_Cloth_volume_l_C02\"),Transform=(Rotation=(X=0,Y=0,Z=0,W=1),Translation=(X=0,Y=0,Z ...(truncated)"
    ItemToModify.1.Transform.Rotation=&RigVMFunction_MathQuaternionFromEuler(Euler=(X=0,Y=0,Z=-10)).Result
    Weight=@n3.OutputParam
   )
   &RigUnit_AimConstraintLocalSpaceOffset(
    AdvancedSettings=(DebugSettings=(bEnabled=false,Scale=10,WorldOffset=(Rotation=(X=0,Y=0,Z=0,W=1),Translation=(X=0,Y=0,Z=0),Scale3D=(X=1,Y=1,Z=1))),RotationOrderForFilter=XZY)
    Child=(Type=Control,Name="thigh_l_ctrl")
    Filter=(bX=True,bY=True,bZ=True)
    Parents=((Item=(Type=Control,Name="foot_l_ctrl"),Weight=1))
    Weight=1
    WorldUp=(Target=(X=0,Y=0,Z=1),Kind=Direction,Space=(Type=None,Name="None"))
    bMaintainOffset=true
    ExecutePin=@SequenceAggregate.B
   )
   > n4 = &RigUnit_SphericalPoseReader(
    ActiveRegionScaleFactors=(PositiveWidth=3,NegativeWidth=2,PositiveHeight=4,NegativeHeight=2)
    ActiveRegionSize=0.1
    Debug=(bDrawDebug=false,DebugScale=25,DebugSegments=20,DebugThickness=0.25)
    DriverAxis=(X=1,Y=0,Z=0)
    DriverItem=(Type=Control,Name="thigh_l_ctrl")
    FalloffRegionScaleFactors=(PositiveWidth=0.3,NegativeWidth=2,PositiveHeight=1,NegativeHeight=1)
    FalloffSize=0.35
    FlipHeightScaling=false
    FlipWidthScaling=true
    OptionalParentItem=(Type=Bone,Name="pelvis")
    RotationOffset=(X=-220,Y=0,Z=-5)
   )
   > &RigUnit_ParentConstraint(
    AdvancedSettings=(RotationOrderForFilter=XZY)
    Child=(Type=Bone,Name="dyn_Cloth_volume_r_B")
    Filter=(TranslationFilter=(bX=false,bY=false,bZ=false),RotationFilter=(bX=true,bY=true,bZ=true),ScaleFilter=(bX=True,bY=True,bZ=True))
    Parents=((Item=(Type=Bone,Name="thigh_twist_01_r"),Weight=0),(Item=(Type=Bone,Name="pelvis"),Weight=1))
    Weight=@n4.OutputParam
   )
   > &RigUnit_ModifyTransforms(
    ItemToModify="((Item=(Type=Bone,Name=\"dyn_Cloth_volume_r_B01\"),Transform=(Rotation=(X=0.0778,Y=-0.2468,Z=0.2905,W=0.9212),Translation=(X=0,Y=0,Z=0),Scale3D=(X=1,Y=1,Z=1))),(Item=(Type=Bone,Name=\"dyn_Cloth_volume_r_C02\"),Transform=(Rotation=(X=0,Y=0,Z=0.4617,W=0.887),Tra ...(truncated)"
    Weight=@n4.OutputParam
   )
   > n5 = &RigUnit_SphericalPoseReader(
    ActiveRegionScaleFactors=(PositiveWidth=1,NegativeWidth=1,PositiveHeight=1,NegativeHeight=1)
    ActiveRegionSize=0.1
    Debug=(bDrawDebug=false,DebugScale=25,DebugSegments=20,DebugThickness=0.25)
    DriverAxis=(X=1,Y=0,Z=0)
    DriverItem=(Type=Control,Name="thigh_r_ctrl")
    FalloffRegionScaleFactors=(PositiveWidth=1,NegativeWidth=1,PositiveHeight=1,NegativeHeight=0.8)
    FalloffSize=0.6
    FlipHeightScaling=false
    FlipWidthScaling=true
    OptionalParentItem=(Type=Bone,Name="pelvis")
    RotationOffset=(X=70,Y=10,Z=20)
   )
   > &RigUnit_ParentConstraint(
    AdvancedSettings=(RotationOrderForFilter=XZY)
    Child=(Type=Bone,Name="dyn_Cloth_volume_r_A")
    Filter=(TranslationFilter=(bX=true,bY=true,bZ=true),RotationFilter=(bX=True,bY=True,bZ=True),ScaleFilter=(bX=True,bY=True,bZ=True))
    Parents=((Item=(Type=Bone,Name="thigh_twist_01_r"),Weight=0.8),(Item=(Type=Bone,Name="pelvis"),Weight=0.2))
    Weight=@n5.OutputParam
   )
   > &RigUnit_ModifyTransforms(
    ItemToModify="((Item=(Type=Bone,Name=\"dyn_Cloth_volume_r_C\"),Transform=(Rotation=(X=-0.0076,Y=0.0868,Z=0.0868,W=0.9924),Translation=(X=0,Y=0,Z=0),Scale3D=(X=1,Y=1,Z=1))),(Item=(Type=Bone,Name=\"dyn_Cloth_volume_r_C03\"),Transform=(Rotation=(X=-0.0076,Y=0.0868,Z=0.0868,W=0 ...(truncated)"
    Weight=@n5.OutputParam
   )
   > n6 = &RigUnit_SphericalPoseReader(
    ActiveRegionScaleFactors=(PositiveWidth=3,NegativeWidth=2,PositiveHeight=2,NegativeHeight=2)
    ActiveRegionSize=0.1
    Debug=(bDrawDebug=false,DebugScale=25,DebugSegments=20,DebugThickness=0.25)
    DriverAxis=(X=1,Y=0,Z=0)
    DriverItem=(Type=Bone,Name="spine_02")
    FalloffRegionScaleFactors=(PositiveWidth=1,NegativeWidth=1,PositiveHeight=1,NegativeHeight=1)
    FalloffSize=0.45
    FlipWidthScaling=false
    OptionalParentItem=(Type=Bone,Name="gravity")
    RotationOffset=(X=90,Y=0,Z=5)
   )
   > &RigUnit_ParentConstraint(
    AdvancedSettings=(RotationOrderForFilter=XZY)
    Child=(Type=Bone,Name="dyn_Cloth_volume_r_C02")
    Filter=(TranslationFilter=(bX=false,bY=false,bZ=false),RotationFilter=(bX=True,bY=True,bZ=True),ScaleFilter=(bX=True,bY=True,bZ=True))
    Parents=((Item=(Type=Bone,Name="dyn_Cloth_volume_r_C02"),Weight=0.5),(Item=(Type=Bone,Name="pelvis"),Weight=0.5))
    Weight=@n6.OutputParam
   )
   > &RigUnit_ModifyTransforms(
    ItemToModify="((Item=(Type=Bone,Name=\"dyn_Cloth_volume_r_B01\"),Transform=(Rotation=(X=0,Y=0.1736,Z=0,W=0.9848),Translation=(X=0,Y=0,Z=0),Scale3D=(X=1,Y=1,Z=1))),(Item=(Type=Bone,Name=\"dyn_Cloth_volume_r_C02\"),Transform=(Rotation=(X=0,Y=0,Z=0,W=1),Translation=(X=0,Y=0,Z= ...(truncated)"
    ItemToModify.1.Transform.Rotation=&RigVMFunction_MathQuaternionFromEuler(Euler=(X=0,Y=0,Z=-10)).Result
    Weight=@n6.OutputParam
   )
  ]}
]
hierarchy [
 root {loc=<0 0 0>, rot=<0 0 0 1>}
 pelvis {parent=root, loc=<0 -0.7071 93.1191>, rot=<0.0127 0.707 -0.0127 -0.707>}
 spine_01 {parent=pelvis, loc=<6.7494 -0.6388 0>, rot=<0 0 0.034 -0.9994>}
 spine_02 {parent=spine_01, loc=<7.2518 -0.035 0>, rot=<0 0 0.0238 -0.9997>}
 spine_03 {parent=spine_02, loc=<6.0104 0 0>, rot=<0 0 0.0149 -0.9999>}
 spine_04 {parent=spine_03, loc=<8.1024 0 0>, rot=<0 0 -0.0132 -0.9999>}
 spine_05 {parent=spine_04, loc=<10.0146 0 0>, rot=<0 0 -0.0015 -1>}
 attach_cape {parent=spine_05, loc=<9.4015 -13.8945 0>, rot=<0.0616 -0.7044 -0.0616 -0.7044>}
 neck_01 {parent=spine_05, loc=<11.3138 0.1529 0>, rot=<0 0 -0.2003 -0.9797>}
 neck_02 {parent=neck_01, loc=<4.5627 0 0>, rot=<0 0 0.0504 -0.9987>}
 head {parent=neck_02, loc=<5.3914 0 0>, rot=<0 0 0.115 -0.9934>}
 attach_backpack {parent=spine_05, loc=<-0.181 -15.6864 0>, rot=<-0.0093 -0.707 0.0093 -0.707>}
 clavicle_l {parent=spine_05, loc=<4.3575 -0.1466 -2.3223>, rot=<0.0465 -0.7289 -0.009 -0.6829>}
 upperarm_l {parent=clavicle_l, loc=<15.5111 0 0>, rot=<0.0226 -0.3234 -0.0088 -0.9459>}
 lowerarm_l {parent=upperarm_l, loc=<26.6514 0 0>, rot=<0.0391 -0.0099 -0.2453 -0.9686>}
 hand_l {parent=lowerarm_l, loc=<22.3491 0 0>, rot=<-0.7081 -0.0507 0.0975 -0.6975>}
 thumb_01_l {parent=hand_l, loc=<1.3118 -0.3191 -2.459>, rot=<0.6546 -0.4042 0.0664 -0.6354>}
 thumb_02_l {parent=thumb_01_l, loc=<5.9336 -0.0448 0.0401>, rot=<-0.0037 0.0213 0.298 -0.9543>}
 thumb_03_l {parent=thumb_02_l, loc=<2.2223 0.3682 -0.0907>, rot=<0.0118 -0.013 0.0066 -0.9998>}
 index_metacarpal_l {parent=hand_l, loc=<1.6535 0.0183 -2.96>, rot=<0.0158 -0.0124 -0.0173 -0.9996>}
 index_01_l {parent=index_metacarpal_l, loc=<8.3008 0.0357 -0.0524>, rot=<-0.0044 0.0485 0.0924 -0.9945>}
 index_02_l {parent=index_01_l, loc=<3.3587 -0.3004 0.0452>, rot=<-0.002 0.0027 0.3225 -0.9466>}
 index_03_l {parent=index_02_l, loc=<2.5578 -0.123 -0.0252>, rot=<-0.002 0.0067 0.2558 -0.9667>}
 middle_metacarpal_l {parent=hand_l, loc=<2.1856 -0.1725 -0.5984>, rot=<0.006 0.0168 -0.0329 -0.9993>}
 middle_01_l {parent=middle_metacarpal_l, loc=<7.3275 -0.2447 0.0129>, rot=<-0.0208 0.0042 0.0858 -0.9961>}
 middle_02_l {parent=middle_01_l, loc=<4.0123 0.1401 0.1175>, rot=<-0.0218 0.0258 0.3512 -0.9357>}
 middle_03_l {parent=middle_02_l, loc=<2.6077 0.0534 0.1373>, rot=<0.0082 0.0188 0.1708 -0.9851>}
 ring_metacarpal_l {parent=hand_l, loc=<2.2074 -0.3625 1.3638>, rot=<-0.0284 0.0722 -0.0458 -0.9959>}
 ring_01_l {parent=ring_metacarpal_l, loc=<7.0295 -0.6197 -0.1446>, rot=<-0.0489 -0.0252 0.1892 -0.9804>}
 ring_02_l {parent=ring_01_l, loc=<3.6174 0.4647 -0.1027>, rot=<0 0 0.2806 -0.9598>}
 ring_03_l {parent=ring_02_l, loc=<2.4152 0.114 0.0199>, rot=<0 0 0.1189 -0.9929>}
 pinky_metacarpal_l {parent=hand_l, loc=<1.7616 -0.9753 3.3814>, rot=<-0.0397 0.0814 -0.069 -0.9935>}
 pinky_01_l {parent=pinky_metacarpal_l, loc=<7.1709 -0.4831 -0.1094>, rot=<-0.0654 -0.0035 0.2242 -0.9723>}
 pinky_02_l {parent=pinky_01_l, loc=<2.8274 0.0893 -0.1043>, rot=<0.0037 0.0011 0.2854 -0.9584>}
 pinky_03_l {parent=pinky_02_l, loc=<2.2046 0.0307 -0.0078>, rot=<-0.0002 0.0098 0.1445 -0.9895>}
 hand_attach_l {parent=hand_l, loc=<8.5 -3 0>, rot=<0.7071 0.7071 0 0>}
 weapon_l {parent=hand_l, loc=<-0.1974 -3.4111 0.3811>, rot=<0 0 0 1>}
 lowerarm_twist_02_l {parent=lowerarm_l, loc=<9.6199 0 0>, rot=<0 0 0 -1>}
 lowerarm_twist_01_l {parent=lowerarm_l, loc=<17.8549 0 0>, rot=<0 0 0 -1>}
 deform_elbow_in_base_l {parent=lowerarm_l, loc=<0.2112 5.259 0.8274>, rot=<0 0 0 -1>}
 deform_elbow_in_l {parent=deform_elbow_in_base_l, loc=<0 0 0>, rot=<0 0 0 1>}
 deform_elbow_out_base_l {parent=lowerarm_l, loc=<0 -4 0>, rot=<0 0 0 -1>}
 deform_elbow_out_l {parent=deform_elbow_out_base_l, loc=<0 0 0>, rot=<0 0 0 1>}
 upperarm_twist_01_l {parent=upperarm_l, loc=<0.0041 0 0>, rot=<0 0 0 1>}
 upperarm_twist_02_l {parent=upperarm_l, loc=<13.7941 0 0>, rot=<0 0 0 1>}
 dyn_upperarm_cloth_l {parent=upperarm_l, loc=<8.9727 0 0>, rot=<0 0 0 1>}
 dyn_upperarm_cloth_al {parent=dyn_upperarm_cloth_l, loc=<21.1728 12.5648 0>, rot=<0 0 -0.179 -0.9839>}
 dyn_upperarm_cloth_bl {parent=dyn_upperarm_cloth_l, loc=<14.9101 -0.0719 0>, rot=<0 0 -0.021 -0.9998>}
 upperarm_l_fn {parent=upperarm_l, loc=<0 1 0>, rot=<0 0 0 1>}
 upperarm_l_bn {parent=upperarm_l, loc=<0 -1 0>, rot=<0 0 0 1>}
 clavicle_r {parent=spine_05, loc=<4.3575 -0.1466 2.3223>, rot=<0.7289 0.0465 -0.6829 0.009>}
 upperarm_r {parent=clavicle_r, loc=<-15.5111 0 0>, rot=<0.0226 -0.3234 -0.0088 -0.9459>}
 lowerarm_r {parent=upperarm_r, loc=<-26.6514 0 0>, rot=<0.0391 -0.0099 -0.2453 -0.9686>}
 hand_r {parent=lowerarm_r, loc=<-22.3491 0 0>, rot=<-0.7081 -0.0507 0.0975 -0.6975>}
 weapon_r {parent=hand_r, loc=<0.1974 3.4111 -0.3811>, rot=<0 0 0 1>}
 thumb_01_r {parent=hand_r, loc=<-1.3118 0.3191 2.459>, rot=<0.6546 -0.4042 0.0664 -0.6354>}
 thumb_02_r {parent=thumb_01_r, loc=<-5.9336 0.0448 -0.0401>, rot=<-0.0037 0.0213 0.298 -0.9543>}
 thumb_03_r {parent=thumb_02_r, loc=<-2.2223 -0.3682 0.0907>, rot=<0.0118 -0.013 0.0066 -0.9998>}
 index_metacarpal_r {parent=hand_r, loc=<-1.6535 -0.0183 2.96>, rot=<0.0158 -0.0124 -0.0173 -0.9996>}
 index_01_r {parent=index_metacarpal_r, loc=<-8.3008 -0.0357 0.0524>, rot=<-0.0044 0.0485 0.0924 -0.9945>}
 index_02_r {parent=index_01_r, loc=<-3.3587 0.3004 -0.0452>, rot=<-0.002 0.0027 0.3225 -0.9466>}
 index_03_r {parent=index_02_r, loc=<-2.5578 0.123 0.0252>, rot=<-0.002 0.0067 0.2558 -0.9667>}
 middle_metacarpal_r {parent=hand_r, loc=<-2.1856 0.1725 0.5984>, rot=<0.006 0.0168 -0.0329 -0.9993>}
 middle_01_r {parent=middle_metacarpal_r, loc=<-7.3275 0.2447 -0.0129>, rot=<-0.0208 0.0042 0.0858 -0.9961>}
 middle_02_r {parent=middle_01_r, loc=<-4.0123 -0.1401 -0.1175>, rot=<-0.0218 0.0258 0.3512 -0.9357>}
 middle_03_r {parent=middle_02_r, loc=<-2.6077 -0.0534 -0.1373>, rot=<0.0082 0.0188 0.1708 -0.9851>}
 ring_metacarpal_r {parent=hand_r, loc=<-2.2074 0.3625 -1.3638>, rot=<-0.0284 0.0722 -0.0458 -0.9959>}
 ring_01_r {parent=ring_metacarpal_r, loc=<-7.0295 0.6197 0.1446>, rot=<-0.0489 -0.0252 0.1892 -0.9804>}
 ring_02_r {parent=ring_01_r, loc=<-3.6174 -0.4647 0.1027>, rot=<0 0 0.2806 -0.9598>}
 ring_03_r {parent=ring_02_r, loc=<-2.4152 -0.114 -0.0199>, rot=<0 0 0.1189 -0.9929>}
 pinky_metacarpal_r {parent=hand_r, loc=<-1.7616 0.9753 -3.3814>, rot=<-0.0397 0.0814 -0.069 -0.9935>}
 pinky_01_r {parent=pinky_metacarpal_r, loc=<-7.1709 0.4831 0.1094>, rot=<-0.0654 -0.0035 0.2242 -0.9723>}
 pinky_02_r {parent=pinky_01_r, loc=<-2.8274 -0.0893 0.1043>, rot=<0.0037 0.0011 0.2854 -0.9584>}
 pinky_03_r {parent=pinky_02_r, loc=<-2.2046 -0.0307 0.0078>, rot=<-0.0002 0.0098 0.1445 -0.9895>}
 hand_attach_r {parent=hand_r, loc=<-8.5001 2.9999 0>, rot=<0 0 -0.7071 -0.7071>}
 lowerarm_twist_02_r {parent=lowerarm_r, loc=<-9.6199 0 0>, rot=<0 0 0 -1>}
 lowerarm_twist_01_r {parent=lowerarm_r, loc=<-17.8549 0 0>, rot=<0 0 0 -1>}
 deform_elbow_in_base_r {parent=lowerarm_r, loc=<-0.2111 -5.2591 -0.8275>, rot=<0 0 0 -1>}
 deform_elbow_in_r {parent=deform_elbow_in_base_r, loc=<-0.0001 0 0.0002>, rot=<0 0 0 1>}
 deform_elbow_out_base_r {parent=lowerarm_r, loc=<0 4 -0.0001>, rot=<0 0 0 -1>}
 deform_elbow_out_r {parent=deform_elbow_out_base_r, loc=<0 0 0>, rot=<0 0 0 1>}
 upperarm_twist_01_r {parent=upperarm_r, loc=<-0.0041 0 0>, rot=<0 0 0 -1>}
 upperarm_twist_02_r {parent=upperarm_r, loc=<-13.7941 0 0>, rot=<0 0 0 -1>}
 dyn_upperarm_croth_r {parent=upperarm_r, loc=<-8.973 0 0.0003>, rot=<0 0 0 1>}
 dyn_upperarm_croth_ar {parent=dyn_upperarm_croth_r, loc=<-21.1724 -12.5648 -0.0005>, rot=<0 0 -0.179 -0.9839>}
 dyn_upperarm_croth_br {parent=dyn_upperarm_croth_r, loc=<-14.9098 0.0719 -0.0003>, rot=<0 0 -0.021 -0.9998>}
 upperarm_r_fn {parent=upperarm_r, loc=<0 -1 0>, rot=<0 0 0 1>}
 upperarm_r_bn {parent=upperarm_r, loc=<0 1 0>, rot=<0 0 0 1>}
 fix_collar {parent=spine_05, loc=<13.8726 10.2574 0>, rot=<0 0 -0.0514 -0.9987>}
 fix_spine_r_01 {parent=spine_05, loc=<0 11.6293 10.289>, rot=<0 0 0 1>}
 fix_spine_r_02 {parent=spine_05, loc=<13.2518 11.6293 11.7559>, rot=<0 0 0 1>}
 fix_spine_l_01 {parent=spine_05, loc=<0.0001 11.6293 -10.289>, rot=<0 0 1 0>}
 fix_spine_l_02 {parent=spine_05, loc=<13.2517 11.6293 -11.7559>, rot=<0 0 1 0>}
 dyn_Cloth_volume_l_Up01 {parent=spine_04, loc=<0.4948 11.3671 -8.6143>, rot=<0.9879 0.0459 0.0054 0.1479>}
 dyn_Cloth_volume_l_Up02 {parent=dyn_Cloth_volume_l_Up01, loc=<-8.6737 -0.8496 -0.5036>, rot=<0 0 -0.0088 -1>}
 dyn_Cloth_volume_l_A {parent=dyn_Cloth_volume_l_Up02, loc=<-10.0664 -1.1336 0.5036>, rot=<0 0 -0.0879 -0.9961>}
 dyn_Cloth_volume_l_A01 {parent=dyn_Cloth_volume_l_A, loc=<-6.9566 0.1745 0.246>, rot=<0 0 0 1>}
 dyn_Cloth_volume_l_A02 {parent=dyn_Cloth_volume_l_A01, loc=<-12.2733 1.4635 1.2998>, rot=<0 0 0 1>}
 dyn_Cloth_volume_r_Up01 {parent=spine_04, loc=<0.4948 11.3671 8.6143>, rot=<0.9879 0.0459 -0.0054 -0.1479>}
 dyn_Cloth_volume_r_Up02 {parent=dyn_Cloth_volume_r_Up01, loc=<-8.6723 -1.0037 0>, rot=<0 0 -0.0088 -1>}
 dyn_Cloth_volume_r_A {parent=dyn_Cloth_volume_r_Up02, loc=<-10.0651 -0.9794 0>, rot=<0 0 -0.0879 -0.9961>}
 dyn_Cloth_volume_r_A01 {parent=dyn_Cloth_volume_r_A, loc=<-6.9473 0.0632 -0.2269>, rot=<0 0 0.0286 -0.9996>}
 dyn_Cloth_volume_r_A02 {parent=dyn_Cloth_volume_r_A01, loc=<-12.3245 0.3248 -1.2998>, rot=<0 0 0 1>}
 thigh_l {parent=pelvis, loc=<-2.5693 -0.0109 -7.7501>, rot=<0.0012 0.0572 0.0155 -0.9982>}
 calf_l {parent=thigh_l, loc=<-39.0484 0 0>, rot=<0.0006 -0.0161 -0.0391 -0.9991>}
 foot_l {parent=calf_l, loc=<-42.561 0.1077 0.4292>, rot=<0.0418 -0.0455 0.0343 -0.9975>}
 ball_l {parent=foot_l, loc=<-6.1599 13.8264 0.6806>, rot=<0.0001 -0.0001 -0.7063 -0.7079>}
 foot_l_ctrl {controlType=EULER_TRANSFORM, parent=foot_l, type=Control, loc=<0 0 0>, rot=<0 0 0 1>}
 calf_twist_02_l {parent=calf_l, loc=<-18.8064 -0.0967 0.0397>, rot=<0.0237 0.0095 -0.0035 -0.9997>}
 calf_twist_01_l {parent=calf_l, loc=<-35.813 -0.0967 0.0397>, rot=<0.0237 0.0095 -0.0035 -0.9997>}
 deform_knee_in_base_l {parent=calf_l, loc=<0.5167 -4.9868 -0.0238>, rot=<-0.0017 -0.0216 -0.0001 -0.9998>}
 deform_knee_in_l {parent=deform_knee_in_base_l, loc=<0 0 0>, rot=<0 0 0 1>}
 deform_knee_out_base_l {parent=calf_l, loc=<-0.2966 4.0369 -0.2038>, rot=<-0.0017 -0.0216 -0.0001 -0.9998>}
 deform_knee_out_l {parent=deform_knee_out_base_l, loc=<0 0 0>, rot=<0 0 0 1>}
 dyn_zipper_l_01 {parent=calf_l, loc=<-17.9732 -9.4374 0.339>, rot=<-0.0031 -0.0411 0.0414 -0.9983>}
 thigh_twist_01_l {parent=thigh_l, loc=<-21.6303 0 0>, rot=<0.0013 0 0 -1>}
 thigh_l_ctrl {controlType=EULER_TRANSFORM, parent=thigh_l, type=Control, loc=<0 0 0>, rot=<0 0 0 1>}
 thigh_r {parent=pelvis, loc=<-2.5693 -0.0109 7.7501>, rot=<0.0572 -0.0012 0.9982 0.0155>}
 calf_r {parent=thigh_r, loc=<39.0484 0 0>, rot=<0.0006 -0.0161 -0.0391 -0.9991>}
 foot_r {parent=calf_r, loc=<42.561 -0.1077 -0.4292>, rot=<0.0418 -0.0455 0.0343 -0.9975>}
 ball_r {parent=foot_r, loc=<6.1599 -13.8264 -0.6806>, rot=<0.0001 -0.0001 -0.7063 -0.7079>}
 foot_r_ctrl {controlType=EULER_TRANSFORM, parent=foot_r, type=Control, loc=<0 0 0>, rot=<0 0 0 1>}
 calf_twist_02_r {parent=calf_r, loc=<18.8064 0.0967 -0.0397>, rot=<0.0237 0.0095 -0.0035 -0.9997>}
 calf_twist_01_r {parent=calf_r, loc=<35.813 0.0967 -0.0397>, rot=<0.0237 0.0095 -0.0035 -0.9997>}
 deform_knee_in_base_r {parent=calf_r, loc=<-0.5167 4.9868 0.0237>, rot=<-0.0017 -0.0216 -0.0001 -0.9998>}
 deform_knee_in_r {parent=deform_knee_in_base_r, loc=<0 0 0>, rot=<0 0 0 1>}
 deform_knee_out_base_r {parent=calf_r, loc=<0.2966 -4.0369 0.2038>, rot=<-0.0017 -0.0216 -0.0001 -0.9998>}
 deform_knee_out_r {parent=deform_knee_out_base_r, loc=<0 0 0>, rot=<0 0 0 1>}
 dyn_zipper_r_01 {parent=calf_r, loc=<17.9817 9.4364 -0.235>, rot=<0.0411 -0.0031 0.9983 0.0414>}
 thigh_twist_01_r {parent=thigh_r, loc=<21.6303 0 0>, rot=<0.0013 0 0 -1>}
 thigh_r_ctrl {controlType=EULER_TRANSFORM, parent=thigh_r, type=Control, loc=<0 0 0>, rot=<0 0 0 1>}
 dyn_Cloth_volume_bk {parent=pelvis, loc=<12.4262 -13.7579 0.3385>, rot=<0 0 -0.0472 -0.9989>}
 dyn_Cloth_volume_bk01 {parent=dyn_Cloth_volume_bk, loc=<-6.1856 0.7544 -0.0664>, rot=<0 0 0 1>}
 dyn_Cloth_volume_bk02 {parent=dyn_Cloth_volume_bk01, loc=<-11.8523 -0.1195 -0.2026>, rot=<0.024 0 0 -0.9997>}
 dyn_Cloth_volume_bk03 {parent=dyn_Cloth_volume_bk02, loc=<-14.423 0 0>, rot=<0 0 0 1>}
 dyn_Cloth_volume_bk04 {parent=dyn_Cloth_volume_bk03, loc=<-14.6821 0 0>, rot=<0 0 0 1>}
 dyn_Cloth_volume_l_B {parent=pelvis, loc=<12.5408 -0.8485 -15.6013>, rot=<-0.7099 0.0268 -0.0014 -0.7038>}
 dyn_Cloth_volume_l_B01 {parent=dyn_Cloth_volume_l_B, loc=<-6.4508 -1.8458 0>, rot=<0 0 0 1>}
 dyn_Cloth_volume_l_B02 {parent=dyn_Cloth_volume_l_B01, loc=<-12.5849 -2.0721 0>, rot=<0 0 0 1>}
 dyn_Cloth_volume_l_B03 {parent=dyn_Cloth_volume_l_B02, loc=<-12.7682 -1.7169 0>, rot=<0 0 0 1>}
 dyn_Cloth_volume_l_C {parent=pelvis, loc=<12.83 -11.3376 -10.8975>, rot=<-0.1375 0.0114 -0.0467 -0.9893>}
 dyn_Cloth_volume_l_C01 {parent=dyn_Cloth_volume_l_C, loc=<-5.9349 0.5481 -1.2968>, rot=<0 0 0 1>}
 dyn_Cloth_volume_l_C02 {parent=dyn_Cloth_volume_l_C01, loc=<-12.3087 0.179 -1.4696>, rot=<0 0.0333 0 -0.9994>}
 dyn_Cloth_volume_l_C03 {parent=dyn_Cloth_volume_l_C02, loc=<-14.423 0 0>, rot=<0 0 0 1>}
 dyn_Cloth_volume_l_C04 {parent=dyn_Cloth_volume_l_C03, loc=<-14.6821 0 0>, rot=<0 0 0 1>}
 dyn_Cloth_volume_l_C05 {parent=dyn_Cloth_volume_l_C04, loc=<-14.6854 0 0>, rot=<0 0 0 1>}
 dyn_Cloth_l_base {parent=dyn_Cloth_volume_l_C02, loc=<0 0 0>, rot=<0 0 0 1>}
 dyn_Cloth_l_01 {parent=dyn_Cloth_l_base, loc=<-1.1134 -0.7274 2.3265>, rot=<-0.1551 0.0284 -0.087 -0.9836>}
 dyn_Cloth_l_02 {parent=dyn_Cloth_l_01, loc=<-14.5232 0.3903 -0.4093>, rot=<0 0 0 1>}
 dyn_Cloth_l_03 {parent=dyn_Cloth_l_02, loc=<-14.6868 -0.0483 -0.7674>, rot=<0 0 0 1>}
 dyn_Cloth_l_04 {parent=dyn_Cloth_l_03, loc=<-14.696 -0.8089 -0.9576>, rot=<0 0 0 1>}
 dyn_Cloth_volume_r_B {parent=pelvis, loc=<12.5285 -1.1911 15.9591>, rot=<0.7096 -0.0329 -0.0074 -0.7038>}
 dyn_Cloth_volume_r_B01 {parent=dyn_Cloth_volume_r_B, loc=<-6.4721 -1.393 0.0012>, rot=<0 0 0 1>}
 dyn_Cloth_volume_r_B02 {parent=dyn_Cloth_volume_r_B01, loc=<-12.6173 -1.7284 0>, rot=<0 0 0 1>}
 dyn_Cloth_volume_r_B03 {parent=dyn_Cloth_volume_r_B02, loc=<-12.7682 -1.7169 0>, rot=<0 0 0 1>}
 dyn_Cloth_volume_r_C {parent=pelvis, loc=<12.9366 -11.4322 10.6309>, rot=<0.1373 -0.0132 -0.0595 -0.9887>}
 dyn_Cloth_volume_r_C01 {parent=dyn_Cloth_volume_r_C, loc=<-5.9091 1.0533 1.169>, rot=<0 0 0 1>}
 dyn_Cloth_volume_r_C02 {parent=dyn_Cloth_volume_r_C01, loc=<-12.2974 0.6969 1.4696>, rot=<0 -0.0549 0 -0.9985>}
 dyn_Cloth_volume_r_C03 {parent=dyn_Cloth_volume_r_C02, loc=<-14.423 0 0>, rot=<0 0 0 1>}
 dyn_Cloth_volume_r_C04 {parent=dyn_Cloth_volume_r_C03, loc=<-14.6821 0 0>, rot=<0 0 0 1>}
 dyn_Cloth_volume_r_C05 {parent=dyn_Cloth_volume_r_C04, loc=<-14.6854 0 0>, rot=<0 0 0 1>}
 dyn_Cloth_r_base {parent=dyn_Cloth_volume_r_C02, loc=<0 0 0>, rot=<0 0 0 1>}
 dyn_Cloth_r_01 {parent=dyn_Cloth_r_base, loc=<-1.1929 -1.2164 -1.9809>, rot=<0.2054 -0.0044 -0.0715 -0.976>}
 dyn_Cloth_r_02 {parent=dyn_Cloth_r_01, loc=<-14.4605 0.4748 0.1017>, rot=<0 0 0 1>}
 dyn_Cloth_r_03 {parent=dyn_Cloth_r_02, loc=<-14.8869 -0.1865 1.1372>, rot=<0 0 0 1>}
 dyn_Cloth_r_04 {parent=dyn_Cloth_r_03, loc=<-14.7294 -0.6475 0.9284>, rot=<0 0 0 1>}
 dyn_Cloth_collision_l_B01 {parent=pelvis, loc=<-20.7155 -8.1493 -20.7077>, rot=<-0.7099 0.0268 -0.0014 -0.7038>}
 dyn_Cloth_collision_l_B02 {parent=pelvis, loc=<-41.174 -9.1614 -24.4967>, rot=<-0.7099 0.0268 -0.0014 -0.7038>}
 dyn_Cloth_collision_r_B01 {parent=pelvis, loc=<-20.5894 -7.7667 21.0273>, rot=<0.7096 -0.0329 -0.0074 -0.7038>}
 dyn_Cloth_collision_r_B02 {parent=pelvis, loc=<-41.1236 -8.9779 24.608>, rot=<0.7096 -0.0329 -0.0074 -0.7038>}
 dyn_Cloth_collision_l_C01 {parent=pelvis, loc=<-21.0231 -14.5648 -12.6376>, rot=<-0.139 0.0444 -0.0421 -0.9884>}
 dyn_Cloth_collision_l_C02 {parent=pelvis, loc=<-32.9574 -15.8667 -13.7646>, rot=<-0.139 0.0444 -0.0421 -0.9884>}
 dyn_Cloth_collision_l_C03 {parent=pelvis, loc=<-50.2104 -17.5854 -15.1867>, rot=<-0.139 0.0444 -0.0421 -0.9884>}
 dyn_Cloth_collision_r_C01 {parent=pelvis, loc=<-21.0532 -14.4827 12.599>, rot=<0.1404 -0.0675 -0.0518 -0.9864>}
 dyn_Cloth_collision_r_C02 {parent=pelvis, loc=<-32.9413 -15.6457 14.0488>, rot=<0.1404 -0.0675 -0.0518 -0.9864>}
 dyn_Cloth_collision_r_C03 {parent=pelvis, loc=<-50.1731 -17.5869 15.2156>, rot=<0.1404 -0.0675 -0.0518 -0.9864>}
 dyn_Cloth_collision_bk01 {parent=pelvis, loc=<-16.9431 -16.6111 -0.0799>, rot=<0.024 0.0011 -0.0472 -0.9986>}
 dyn_Cloth_collision_bk02 {parent=pelvis, loc=<-33.0853 -18.4075 -0.0003>, rot=<0.024 0.0011 -0.0472 -0.9986>}
 dyn_Cloth_collision_l_B03 {parent=pelvis, loc=<-28.2882 1.1203 -24.6562>, rot=<-0.7099 0.0268 -0.0014 -0.7038>}
 dyn_Cloth_collision_r_B03 {parent=pelvis, loc=<-28.2785 1.3247 24.4467>, rot=<0.7096 -0.0329 -0.0074 -0.7038>}
 gravity {parent=root, loc=<0 0 0>, rot=<0 0 0 1>}
 gravity_upperarm_l {parent=root, loc=<0 0 0>, rot=<0 0 0 1>}
 gravity_upperarm_r {parent=root, loc=<0 0 0>, rot=<0 0 0 1>}
 attach {parent=root, loc=<0 0 0>, rot=<0 0 0 1>}
 ik_hand_root {parent=root, loc=<0 0 0>, rot=<0 0 0 1>}
 ik_hand_gun {parent=ik_hand_root, loc=<-51.9727 4.0883 102.6935>, rot=<0.6084 0.1793 0.4613 -0.6205>}
 ik_hand_r {parent=ik_hand_gun, loc=<0 0 0>, rot=<0 0 0 1>}
 ik_hand_l {parent=ik_hand_gun, loc=<53.0285 82.1781 35.2051>, rot=<-0.5102 -0.7906 -0.3387 0>}
 ik_foot_root {parent=root, loc=<0 0 0>, rot=<0 0 0 1>}
 ik_foot_l {parent=ik_foot_root, loc=<15.2608 -4.4176 9.4432>, rot=<0.0357 0.7045 0.0284 -0.7082>}
 ik_foot_r {parent=ik_foot_root, loc=<-15.2608 -4.4176 9.4432>, rot=<0.7082 0.0284 -0.7045 0.0357>}
 attach_fp {parent=root, loc=<0 0 0>, rot=<0 0 0 1>}
 camera_root {parent=root, loc=<0 0 0>, rot=<0 0 0 1>}
 camera {parent=camera_root, loc=<0 0 0>, rot=<0 0 0 1>}
 warp_point_01 {parent=root, loc=<0 0 0>, rot=<0 0 0 1>}
 warp_point_02 {parent=root, loc=<0 0 0>, rot=<0 0 0 1>}
 warp_point_03 {parent=root, loc=<0 0 0>, rot=<0 0 0 1>}
 gravity1 {parent=root, loc=<0 0 0>, rot=<0 0 0 1>}
]
