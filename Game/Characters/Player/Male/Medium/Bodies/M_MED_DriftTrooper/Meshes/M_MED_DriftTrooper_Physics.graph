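# Physics asset for the M_MED_DriftTrooper player body, covering only its dynamic cloth bones and the shapes around them.
# Left and right cloth chains each run from a tiny base capsule through four box or capsule segments with limited swing and twist.
# Each base joins a C02 volume capsule with free rotation.
# Unconstrained collision capsules on back and left/right B and C bones provide shapes for cloth collision.
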PhysicsAsset /Game/Characters/Player/Male/Medium/Bodies/M_MED_DriftTrooper/Meshes/M_MED_DriftTrooper_Physics
bodies {
 dyn_Cloth_collision_bk01 [{shapeType=Capsule, length=4.5, radius=10.5, shapeName=dyn_Cloth_collision_bk01_capsule, center=<0 10.5665 0>}]
 dyn_Cloth_collision_bk02 [{
   shapeType=Capsule
   length=11
   radius=13
   shapeName=dyn_Cloth_collision_bk02_capsule
   center=<-3.3169 11.3638 0.0587>
   rotation=<rot -0.0012 -0.0246 -4.9454>
  }]
 dyn_Cloth_collision_l_B01 [{shapeType=Capsule, length=2, radius=8.5, shapeName=dyn_Cloth_collision_l_B01_capsule, center=<0 4.093 -4.37>}]
 dyn_Cloth_collision_l_B02 [{
   shapeType=Capsule
   length=7.5
   radius=15
   shapeName=dyn_Cloth_collision_l_B02_capsule
   center=<-0.9006 8.4299 -9.4177>
   rotation=<rot 11.0912 -1.5192 -7.8498>
  }]
 dyn_Cloth_collision_l_B03 [{shapeType=Capsule, length=9, radius=10.5, shapeName=dyn_Cloth_collision_l_B03_capsule, center=<-1.5522 7.1844 0>}]
 dyn_Cloth_collision_l_C01 [{shapeType=Capsule, length=2, radius=14.5, shapeName=dyn_Cloth_collision_l_C01_capsule, center=<0 13.585 0>}]
 dyn_Cloth_collision_l_C02 [{shapeType=Capsule, length=3, radius=13.5, shapeName=dyn_Cloth_collision_l_C02_capsule, center=<-4.5658 12.8245 0>}]
 dyn_Cloth_collision_l_C03 [{
   shapeType=Capsule
   length=4.5
   radius=20
   shapeName=dyn_Cloth_collision_l_C03_capsule
   center=<-8.0337 17.3313 -0.1641>
   rotation=<rot 0 -0.0001 4.4353>
  }]
 dyn_Cloth_collision_r_B01 [{shapeType=Capsule, length=2, radius=8.5, shapeName=dyn_Cloth_collision_r_B01_capsule, center=<0 5.6001 3.5696>}]
 dyn_Cloth_collision_r_B02 [{
   shapeType=Capsule
   length=10.5
   radius=14
   shapeName=dyn_Cloth_collision_r_B02_capsule
   center=<-7.849 10.1263 7.5722>
   rotation=<rot -0.0001 0 13.1806>
  }]
 dyn_Cloth_collision_r_B03 [{shapeType=Capsule, length=11.5, radius=10, shapeName=dyn_Cloth_collision_r_B03_capsule, center=<-2.2471 9.1382 -0.5672>}]
 dyn_Cloth_collision_r_C01 [{shapeType=Capsule, length=0, radius=15, shapeName=dyn_Cloth_collision_r_C01_capsule, center=<-0.1862 13.7876 -1.884>}]
 dyn_Cloth_collision_r_C02 [{shapeType=Capsule, length=4.5, radius=15.5, shapeName=dyn_Cloth_collision_r_C02_capsule, center=<-3.822 12.7108 -1.884>}]
 dyn_Cloth_collision_r_C03 [{
   shapeType=Capsule
   length=2
   radius=19.5
   shapeName=dyn_Cloth_collision_r_C03_capsule
   center=<-9.5052 16.3706 -2.154>
   rotation=<rot 0 0 5.8435>
  }]
 dyn_Cloth_l_01 [{shapeType=Box, extentX=18.5, extentY=3.5, extentZ=7.5, shapeName=dyn_Cloth_l_01_box, center=<-7.7206 0 0>}]
 dyn_Cloth_l_02 [{shapeType=Box, extentX=18.5, extentY=3.5, extentZ=7.5, shapeName=dyn_Cloth_l_02_box, center=<-7.7206 0 0>}]
 dyn_Cloth_l_03 [{
   shapeType=Capsule
   length=13
   radius=3.5
   shapeName=dyn_Cloth_l_03_capsule
   center=<-7.5076 0 -0.0257>
   rotation=<rot 89.8039 179.9846 179.9848>
  }]
 dyn_Cloth_l_04 [{shapeType=Box, extentX=19.5, extentY=2.5, extentZ=17.5, shapeName=dyn_Cloth_l_04_box, center=<-6.2553 0 0>}]
 dyn_Cloth_l_base [{shapeType=Capsule, length=0.505, radius=0.505, shapeName=dyn_Cloth_l_base_capsule, rotation=<rot 0 -90 0>}]
 dyn_Cloth_r_01 [{shapeType=Box, extentX=18.5, extentY=3.5, extentZ=7.5, shapeName=dyn_Cloth_r_01_box, center=<-7.7206 0 0>, rotation=<rot 0 0 -12.5835>}]
 dyn_Cloth_r_02 [{
   shapeType=Box
   extentX=18.5
   extentY=4.5
   extentZ=7.5
   shapeName=dyn_Cloth_r_02_box
   center=<-7.7206 0 0>
   rotation=<rot -0.0001 -0.0001 -7.7116>
  }]
 dyn_Cloth_r_03 [{
   shapeType=Capsule
   length=14
   radius=2.5
   shapeName=dyn_Cloth_r_03_capsule
   center=<-7.5076 0 -0.0257>
   rotation=<rot 89.8042 -179.9855 -179.9854>
  }]
 dyn_Cloth_r_04 [{shapeType=Box, extentX=19.5, extentY=4.5, extentZ=17.5, shapeName=dyn_Cloth_r_04_box, center=<-6.2553 0.8056 0>}]
 dyn_Cloth_r_base [{shapeType=Capsule, length=0.505, radius=0.505, shapeName=dyn_Cloth_r_base_capsule, rotation=<rot 0 -90 0>}]
 dyn_Cloth_volume_l_C02 [{shapeType=Capsule, length=0, radius=12.5, shapeName=dyn_Cloth_volume_l_C02_capsule, center=<-3.1847 10.7231 0>}]
 dyn_Cloth_volume_r_C02 [{shapeType=Capsule, length=0, radius=12.5, shapeName=dyn_Cloth_volume_r_C02_capsule, center=<-3.1847 10.7231 0>}]
}
constraints [
 {bone1Name=dyn_Cloth_l_04, bone2Name=dyn_Cloth_l_03, swing1LimitDegrees=35, swing2LimitDegrees=25, twistLimitDegrees=22}
 {bone1Name=dyn_Cloth_l_03, bone2Name=dyn_Cloth_l_02, swing1LimitDegrees=35, swing2LimitDegrees=25, twistLimitDegrees=22}
 {bone1Name=dyn_Cloth_l_02, bone2Name=dyn_Cloth_l_01, swing1LimitDegrees=35, swing2LimitDegrees=25, twistLimitDegrees=22}
 {bone1Name=dyn_Cloth_l_01, bone2Name=dyn_Cloth_l_base, swing1LimitDegrees=35, swing2LimitDegrees=25, twistLimitDegrees=22}
 {bone1Name=dyn_Cloth_r_01, bone2Name=dyn_Cloth_r_base, swing1LimitDegrees=35, swing2LimitDegrees=25, twistLimitDegrees=22}
 {bone1Name=dyn_Cloth_r_02, bone2Name=dyn_Cloth_r_01, swing1LimitDegrees=35, swing2LimitDegrees=25, twistLimitDegrees=22}
 {bone1Name=dyn_Cloth_r_03, bone2Name=dyn_Cloth_r_02, swing1LimitDegrees=35, swing2LimitDegrees=25, twistLimitDegrees=22}
 {bone1Name=dyn_Cloth_r_04, bone2Name=dyn_Cloth_r_03, swing1LimitDegrees=35, swing2LimitDegrees=25, twistLimitDegrees=22}
 {
  bone1Name=dyn_Cloth_l_base
  bone2Name=dyn_Cloth_volume_l_C02
  swing1LimitDegrees=45
  swing1Motion=Free
  swing2LimitDegrees=45
  swing2Motion=Free
  twistLimitDegrees=45
  twistMotion=Free
 }
 {
  bone1Name=dyn_Cloth_r_base
  bone2Name=dyn_Cloth_volume_r_C02
  swing1LimitDegrees=45
  swing1Motion=Free
  swing2LimitDegrees=45
  swing2Motion=Free
  twistLimitDegrees=45
  twistMotion=Free
 }
]
