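# Animation Blueprint for the Drift Trooper male body cosmetic that adds corrective and cloth-bone motion to the character's pose.
# It copies the attached parent's pose, then chains weighted constraints and variable-driven bone offsets for attachments, thighs, shoulders, sleeves and collar.
# Two Control Rigs, one for the skirt, feed slide adjustments into cached poses.
# Skirt bone variants are blended by movement, crouch, dive and DBNO states.
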AnimBlueprint /BRCosmetics/Characters/Player/Male/Medium/Bodies/M_MED_DriftTrooper/Meshes/M_MED_DriftTrooper_AnimBP
target_skeleton=&/M_MED_DriftTrooper_Skeleton
graphs [
 AnimGraph {nodes [
   n1 = "Constraint - dyn_Cloth_volume_l_C01 (PP)"!(
    ComponentPose=Local To Component(LocalPose=Use cached pose 'SkirtCachePose'())
    Alpha=1
    ConstraintWeights_0=0.85
    ConstraintWeights_1:float=0.15
   ) -> (Pose) {
    alphaBoolBlend {blendOption=Linear}
    boneToModify {boneName=dyn_Cloth_volume_l_C01}
    constraintSetup [{targetBone {boneName=pelvis}}, {targetBone {boneName=dyn_Cloth_volume_l_B}}]
    constraintWeights [1,1]
   }
   n2 = "Constraint - dyn_Cloth_volume_r_C01 (PP)"!(ComponentPose=@n1, Alpha=1, ConstraintWeights_0=0.85, ConstraintWeights_1:float=0.15) -> (Pose) {
    alphaBoolBlend {blendOption=Linear}
    boneToModify {boneName=dyn_Cloth_volume_r_C01}
    constraintSetup [{targetBone {boneName=pelvis}}, {targetBone {boneName=dyn_Cloth_volume_r_B}}]
    constraintWeights [1,1]
   }
   n3 = "Constraint - dyn_Cloth_volume_l_C03 (PP)"!(ComponentPose=@n2, Alpha=1, ConstraintWeights_0=1, ConstraintWeights_1:float=0) -> (Pose) {
    alphaBoolBlend {blendOption=Linear}
    boneToModify {boneName=dyn_Cloth_volume_l_C03}
    constraintSetup [{targetBone {boneName=dyn_Cloth_volume_l_C02}}, {targetBone {boneName=dyn_Cloth_volume_r_C04}}]
    constraintWeights [1,1]
   }
   n4 = "Constraint - dyn_Cloth_volume_r_C03 (PP)"!(ComponentPose=@n3, Alpha=1, ConstraintWeights_0=1, ConstraintWeights_1:float=0) -> (Pose) {
    alphaBoolBlend {blendOption=Linear}
    boneToModify {boneName=dyn_Cloth_volume_r_C03}
    constraintSetup [{targetBone {boneName=dyn_Cloth_volume_r_C02}}, {targetBone {boneName=dyn_Cloth_volume_r_C04}}]
    constraintWeights [1,1]
   }
   n5 = "Transform (Modify) Bone - Bone: dyn_Cloth_volume_l_C01"(ComponentPose=@n4, Alpha=1, Translation="0,0,0", Rotation="0,20,0", Scale="1,1,1") {
    rotationMode=BMM_Additive
    rotationSpace=BCS_BoneSpace
    scaleMode=BMM_Additive
    scaleSpace=BCS_BoneSpace
    translationMode=BMM_Additive
    translationSpace=BCS_BoneSpace
    alphaBoolBlend {blendOption=Linear}
    boneToModify {boneName=dyn_Cloth_volume_l_C01}
    rotation {yaw=20}
   }
   n6 = "Transform (Modify) Bone - Bone: dyn_Cloth_volume_r_C01"(ComponentPose=@n5, Alpha=1, Translation="0,0,0", Rotation="0,20,0", Scale="1,1,1") {
    rotationMode=BMM_Additive
    rotationSpace=BCS_BoneSpace
    scaleMode=BMM_Additive
    scaleSpace=BCS_BoneSpace
    translationMode=BMM_Additive
    translationSpace=BCS_BoneSpace
    alphaBoolBlend {blendOption=Linear}
    boneToModify {boneName=dyn_Cloth_volume_r_C01}
    rotation {yaw=20}
   }
   n7 = "Transform (Modify) Bone - Bone: dyn_Cloth_volume_l_C03"(ComponentPose=@n6, Alpha=1, Translation="0,0,0", Rotation="0,10,0", Scale="1,1,1") {
    rotationMode=BMM_Additive
    rotationSpace=BCS_BoneSpace
    scaleMode=BMM_Additive
    scaleSpace=BCS_BoneSpace
    translationMode=BMM_Additive
    translationSpace=BCS_BoneSpace
    alphaBoolBlend {blendOption=Linear}
    boneToModify {boneName=dyn_Cloth_volume_l_C03}
    rotation {yaw=10}
   }
   n8 = "Transform (Modify) Bone - Bone: dyn_Cloth_volume_r_C03"(ComponentPose=@n7, Alpha=1, Translation="0,0,0", Rotation="0,10,0", Scale="1,1,1") {
    rotationMode=BMM_Additive
    rotationSpace=BCS_BoneSpace
    scaleMode=BMM_Additive
    scaleSpace=BCS_BoneSpace
    translationMode=BMM_Additive
    translationSpace=BCS_BoneSpace
    alphaBoolBlend {blendOption=Linear}
    boneToModify {boneName=dyn_Cloth_volume_r_C03}
    rotation {yaw=10}
   }
   n9 = "Constraint - dyn_Cloth_volume_bk (PP)"!(ComponentPose=@n8, Alpha=1, ConstraintWeights_0=0.5, ConstraintWeights_1:float=0.5) -> (Pose) {
    alphaBoolBlend {blendOption=Linear}
    boneToModify {boneName=dyn_Cloth_volume_bk}
    constraintSetup [{targetBone {boneName=dyn_Cloth_volume_l_C01}}, {targetBone {boneName=dyn_Cloth_volume_r_C01}}]
    constraintWeights [1,1]
   }
   n10 = "Constraint - dyn_Cloth_volume_bk02 (PP)"!(ComponentPose=@n9, Alpha=1, ConstraintWeights_0=0.5, ConstraintWeights_1:float=0.5) -> (Pose) {
    alphaBoolBlend {blendOption=Linear}
    boneToModify {boneName=dyn_Cloth_volume_bk02}
    constraintSetup [{targetBone {boneName=dyn_Cloth_volume_l_C03}}, {targetBone {boneName=dyn_Cloth_volume_r_C03}}]
    constraintWeights [1,1]
   }
   n11 = Component To Local(ComponentPose=@n10)
   n12 = "Transform (Modify) Bone - Bone: dyn_Cloth_volume_l_C01"(
    ComponentPose=Local To Component(LocalPose=Use cached pose 'SkirtCachePose'())
    Alpha=1
    Translation="0,0,0"
    Rotation="0,0,0"
    Scale="1,1,1"
   ) {
    rotationMode=BMM_Additive
    rotationSpace=BCS_BoneSpace
    scaleMode=BMM_Additive
    scaleSpace=BCS_BoneSpace
    translationMode=BMM_Additive
    translationSpace=BCS_BoneSpace
    alphaBoolBlend {blendOption=Linear}
    boneToModify {boneName=dyn_Cloth_volume_l_C01}
   }
   n13 = "Transform (Modify) Bone - Bone: dyn_Cloth_volume_r_C01"(ComponentPose=@n12, Alpha=1, Translation="0,0,0", Rotation="0,15,0", Scale="1,1,1") {
    rotationMode=BMM_Additive
    rotationSpace=BCS_BoneSpace
    scaleMode=BMM_Additive
    scaleSpace=BCS_BoneSpace
    translationMode=BMM_Additive
    translationSpace=BCS_BoneSpace
    alphaBoolBlend {blendOption=Linear}
    boneToModify {boneName=dyn_Cloth_volume_r_C01}
    rotation {yaw=15}
   }
   n14 = "Transform (Modify) Bone - Bone: dyn_Cloth_volume_l_C04"(ComponentPose=@n13, Alpha=1, Translation="0,0,0", Rotation="0,-10,0", Scale="1,1,1") {
    rotationMode=BMM_Additive
    rotationSpace=BCS_BoneSpace
    scaleMode=BMM_Additive
    scaleSpace=BCS_BoneSpace
    translationMode=BMM_Additive
    translationSpace=BCS_BoneSpace
    alphaBoolBlend {blendOption=Linear}
    boneToModify {boneName=dyn_Cloth_volume_l_C04}
    rotation {yaw=-10}
   }
   n15 = "Transform (Modify) Bone - Bone: dyn_Cloth_volume_r_C04"(ComponentPose=@n14, Alpha=1, Translation="0,0,0", Rotation="0,-30,0", Scale="1,1,1") {
    rotationMode=BMM_Additive
    rotationSpace=BCS_BoneSpace
    scaleMode=BMM_Additive
    scaleSpace=BCS_BoneSpace
    translationMode=BMM_Additive
    translationSpace=BCS_BoneSpace
    alphaBoolBlend {blendOption=Linear}
    boneToModify {boneName=dyn_Cloth_volume_r_C04}
    rotation {yaw=-30}
   }
   n16 = "Constraint - dyn_Cloth_volume_l_C01 (PP)"(ComponentPose=@n15, Alpha=1, ConstraintWeights_0=0.1, ConstraintWeights_1=0.9) {
    alphaBoolBlend {blendOption=Linear}
    boneToModify {boneName=dyn_Cloth_volume_l_C01}
    constraintSetup [{targetBone {boneName=thigh_l}}, {targetBone {boneName=dyn_Cloth_volume_l_C}}]
    constraintWeights [1,1]
   }
   n17 = "Constraint - dyn_Cloth_volume_r_C01 (PP)"(ComponentPose=@n16, Alpha=1, ConstraintWeights_0=0.1, ConstraintWeights_1=0.9) {
    alphaBoolBlend {blendOption=Linear}
    boneToModify {boneName=dyn_Cloth_volume_r_C01}
    constraintSetup [{targetBone {boneName=thigh_r}}, {targetBone {boneName=dyn_Cloth_volume_r_C}}]
    constraintWeights [1,1]
   }
   n18 = "Constraint - dyn_Cloth_volume_l_C03 (PP)"(ComponentPose=@n17, Alpha=1, ConstraintWeights_0=0.5, ConstraintWeights_1=0.5) {
    alphaBoolBlend {blendOption=Linear}
    boneToModify {boneName=dyn_Cloth_volume_l_C03}
    constraintSetup [{targetBone {boneName=calf_l}}, {targetBone {boneName=dyn_Cloth_volume_l_C03}}]
    constraintWeights [1,1]
   }
   n19 = "Constraint - dyn_Cloth_volume_r_C03 (PP)"(ComponentPose=@n18, Alpha=1, ConstraintWeights_0=0.5, ConstraintWeights_1=0.5) {
    alphaBoolBlend {blendOption=Linear}
    boneToModify {boneName=dyn_Cloth_volume_r_C03}
    constraintSetup [{targetBone {boneName=calf_r}}, {targetBone {boneName=dyn_Cloth_volume_r_C03}}]
    constraintWeights [1,1]
   }
   n20 = "Constraint - dyn_Cloth_volume_bk01 (PP)"!(ComponentPose=@n19, Alpha=1, ConstraintWeights_0=0.5, ConstraintWeights_1:float=0.5) -> (Pose) {
    alphaBoolBlend {blendOption=Linear}
    boneToModify {boneName=dyn_Cloth_volume_bk01}
    constraintSetup [{targetBone {boneName=dyn_Cloth_volume_l_C01}}, {targetBone {boneName=dyn_Cloth_volume_r_C01}}]
    constraintWeights [1,1]
   }
   n21 = "Constraint - dyn_Cloth_volume_bk03 (PP)"!(ComponentPose=@n20, Alpha=1, ConstraintWeights_0=0.5, ConstraintWeights_1:float=0.5) -> (Pose) {
    alphaBoolBlend {blendOption=Linear}
    boneToModify {boneName=dyn_Cloth_volume_bk03}
    constraintSetup [{targetBone {boneName=dyn_Cloth_volume_l_C03}}, {targetBone {boneName=dyn_Cloth_volume_r_C03}}]
    constraintWeights [1,1]
   }
   n22 = Component To Local(ComponentPose=@n21)
   n23 = Get MovingAnd NotCrouchNotDivingNotEmote() -> (MovingAnd NotCrouchNotDivingNotEmote:bool)
   n24 = "Constraint - dyn_Cloth_volume_l_B (PP)"!(
    ComponentPose=Local To Component(LocalPose=Use cached pose 'SkirtCachePose'())
    Alpha=1
    ConstraintWeights_0=0.5
    ConstraintWeights_1:float=0.5
   ) -> (Pose) {
    alphaBoolBlend {blendOption=Linear}
    boneToModify {boneName=dyn_Cloth_volume_l_B}
    constraintSetup [{targetBone {boneName=dyn_Cloth_volume_l_C}}, {targetBone {boneName=dyn_Cloth_volume_l_A}}]
    constraintWeights [1,1]
   }
   n25 = "Constraint - dyn_Cloth_volume_r_B (PP)"!(ComponentPose=@n24, Alpha=1, ConstraintWeights_0=0.5, ConstraintWeights_1:float=0.5) -> (Pose) {
    alphaBoolBlend {blendOption=Linear}
    boneToModify {boneName=dyn_Cloth_volume_r_B}
    constraintSetup [{targetBone {boneName=dyn_Cloth_volume_r_C}}, {targetBone {boneName=dyn_Cloth_volume_r_A}}]
    constraintWeights [1,1]
   }
   n26 = Two Way Blend(A=Use cached pose 'DBNOPose'(), B=Component To Local(ComponentPose=@n25), Alpha=Get DBNO_Alpha() -> (DBNO_Alpha:double)) {blendNode {alphaBoolBlend {blendOption=Linear}}}
   n27 = Use cached pose 'SkirtDynPose'()
   n28 = Get DivingOrDBNO() -> (DivingOrDBNO:bool)
   Output Pose(Result=Blend Poses by bool(
     BlendPose_0=Blend Poses by bool(BlendPose_0=@n26, BlendPose_1=@n27, bActiveValue=@n28) {blendType=Linear}
     BlendPose_1=Blend Poses by bool(BlendPose_0=@n11, BlendPose_1=@n22, BlendTime_0=0.2, BlendTime_1=0.2, bActiveValue=@n23) {blendType=Linear}
    ) {blendType=Linear}) {layerGroup=None}
   n29 = "Constraint - attach_cape (PPPPP)"!(
    ComponentPose=Local To Component(LocalPose=Copy Pose From Mesh() {bUseAttachedParent=true})
    Alpha=1
    ConstraintWeights_0=0.8787
    ConstraintWeights_1:float=0.0589
    ConstraintWeights_2:float=0.0389
    ConstraintWeights_3:float=0.0234
    ConstraintWeights_4:float=0.0002
   ) -> (Pose) {
    alphaBoolBlend {blendOption=Linear}
    boneToModify {boneName=attach_cape}
    constraintSetup [
     {targetBone {boneName=spine_05}}
     {targetBone {boneName=spine_04}}
     {targetBone {boneName=neck_02}}
     {targetBone {boneName=neck_01}}
     {targetBone {boneName=clavicle_r}}
    ]
    constraintWeights [1,1,1,1,1]
   }
   n30 = "Constraint - attach_backpack (PPP)"!(ComponentPose=@n29, Alpha=1, ConstraintWeights_0=0.5563, ConstraintWeights_1:float=0.3613, ConstraintWeights_2:float=0.0824) -> (Pose) {
    alphaBoolBlend {blendOption=Linear}
    boneToModify {boneName=attach_backpack}
    constraintSetup [{targetBone {boneName=spine_05}}, {targetBone {boneName=spine_04}}, {targetBone {boneName=spine_03}}]
    constraintWeights [1,1,1]
   }
   n31 = "Constraint - gravity (TT)"!(ComponentPose=@n30, Alpha=1, ConstraintWeights_0=0.3, ConstraintWeights_1:float=0.7) -> (Pose) {
    alphaBoolBlend {blendOption=Linear}
    boneToModify {boneName=gravity}
    constraintSetup [{transformType=Translation, targetBone {boneName=root}}, {transformType=Translation, targetBone {boneName=pelvis}}]
    constraintWeights [1,1]
   }
   n32 = "Constraint - gravity_upperarm_l (TT)"!(ComponentPose=@n31, Alpha=1, ConstraintWeights_0=0.1, ConstraintWeights_1:float=0.9) -> (Pose) {
    alphaBoolBlend {blendOption=Linear}
    boneToModify {boneName=gravity_upperarm_l}
    constraintSetup [{transformType=Translation, targetBone {boneName=root}}, {transformType=Translation, targetBone {boneName=upperarm_twist_02_l}}]
    constraintWeights [1,1]
   }
   n33 = "Constraint - gravity_upperarm_r (TT)"!(ComponentPose=@n32, Alpha=1, ConstraintWeights_0=0.1, ConstraintWeights_1:float=0.9) -> (Pose) {
    alphaBoolBlend {blendOption=Linear}
    boneToModify {boneName=gravity_upperarm_r}
    constraintSetup [{transformType=Translation, targetBone {boneName=root}}, {transformType=Translation, targetBone {boneName=upperarm_twist_02_r}}]
    constraintWeights [1,1]
   }
   n34 = "Constraint - thigh_volume_bk_l (PP)"!(ComponentPose=@n33, Alpha=1, ConstraintWeights_0=0, ConstraintWeights_1:float=1) -> (Pose) {
    alphaBoolBlend {blendOption=Linear}
    boneToModify {boneName=thigh_volume_bk_l}
    constraintSetup [{targetBone {boneName=thigh_twist_01_l}}, {targetBone {boneName=pelvis}}]
    constraintWeights [1,1]
   }
   n35 = "Constraint - thigh_volume_bk_r (PP)"!(ComponentPose=@n34, Alpha=1, ConstraintWeights_0=0, ConstraintWeights_1:float=1) -> (Pose) {
    alphaBoolBlend {blendOption=Linear}
    boneToModify {boneName=thigh_volume_bk_r}
    constraintSetup [{targetBone {boneName=thigh_twist_01_r}}, {targetBone {boneName=pelvis}}]
    constraintWeights [1,1]
   }
   n36 = Get thigh_shape_fn_l() -> (thigh_shape_fn_l:double)
   n37 = "Constraint - thigh_volume_bk_l (PP)"(ComponentPose=@n35, Alpha=@n36, ConstraintWeights_0=0, ConstraintWeights_1=1) {
    alphaBoolBlend {blendOption=Linear}
    boneToModify {boneName=thigh_volume_bk_l}
    constraintSetup [{targetBone {boneName=thigh_twist_01_l}}, {targetBone {boneName=pelvis}}]
    constraintWeights [1,1]
   }
   n38 = Get thigh_shape_fn_r() -> (thigh_shape_fn_r:double)
   n39 = "Constraint - thigh_volume_bk_r (PP)"(ComponentPose=@n37, Alpha=@n38, ConstraintWeights_0=0, ConstraintWeights_1=1) {
    alphaBoolBlend {blendOption=Linear}
    boneToModify {boneName=thigh_volume_bk_r}
    constraintSetup [{targetBone {boneName=thigh_twist_01_r}}, {targetBone {boneName=pelvis}}]
    constraintWeights [1,1]
   }
   n40 = "Constraint - thigh_volume_fn_l (PP)"!(ComponentPose=@n39, Alpha=@n36, ConstraintWeights_0=0.5, ConstraintWeights_1:float=0.5) -> (Pose) {
    alphaBoolBlend {blendOption=Linear}
    boneToModify {boneName=thigh_volume_fn_l}
    constraintSetup [{targetBone {boneName=thigh_twist_01_l}}, {targetBone {boneName=pelvis}}]
    constraintWeights [1,1]
   }
   n41 = "Constraint - thigh_volume_fn_r (PP)"!(ComponentPose=@n40, Alpha=@n38, ConstraintWeights_0=0.5, ConstraintWeights_1:float=0.5) -> (Pose) {
    alphaBoolBlend {blendOption=Linear}
    boneToModify {boneName=thigh_volume_fn_r}
    constraintSetup [{targetBone {boneName=thigh_twist_01_r}}, {targetBone {boneName=pelvis}}]
    constraintWeights [1,1]
   }
   n42 = "Constraint - thigh_volume_lt_l1 (PP)"!(ComponentPose=@n41, Alpha=1, ConstraintWeights_0=0.5, ConstraintWeights_1:float=0.5) -> (Pose) {
    alphaBoolBlend {blendOption=Linear}
    boneToModify {boneName=thigh_volume_lt_l1}
    constraintSetup [{targetBone {boneName=thigh_twist_01_l}}, {targetBone {boneName=pelvis}}]
    constraintWeights [1,1]
   }
   n43 = "Constraint - thigh_volume_rt_r (PP)"!(ComponentPose=@n42, Alpha=1, ConstraintWeights_0=1, ConstraintWeights_1:float=0) -> (Pose) {
    alphaBoolBlend {blendOption=Linear}
    boneToModify {boneName=thigh_volume_rt_r}
    constraintSetup [{targetBone {boneName=thigh_twist_01_r}}, {targetBone {boneName=pelvis}}]
    constraintWeights [1,1]
   }
   n44 = Get thigh_shape_fn_l()
   n45 = "Transform (Modify) Bone - Bone: thigh_volume_fn_l"(ComponentPose=@n43, Alpha=@n44, Translation="-5,-10,5", Rotation="0,0,0", Scale="1,1,1") {
    rotationMode=BMM_Additive
    rotationSpace=BCS_BoneSpace
    scaleMode=BMM_Additive
    scaleSpace=BCS_BoneSpace
    translationMode=BMM_Additive
    translationSpace=BCS_BoneSpace
    alphaBoolBlend {blendOption=Linear}
    boneToModify {boneName=thigh_volume_fn_l}
    translation=<-5 -10 5>
   }
   n46 = Get thigh_shape_fn_r()
   n47 = "Transform (Modify) Bone - Bone: thigh_volume_fn_r"(ComponentPose=@n45, Alpha=@n46, Translation="0,-10,5", Rotation="0,0,0", Scale="1,1,1") {
    rotationSpace=BCS_BoneSpace
    scaleMode=BMM_Additive
    scaleSpace=BCS_BoneSpace
    translationMode=BMM_Additive
    translationSpace=BCS_BoneSpace
    alphaBoolBlend {blendOption=Linear}
    boneToModify {boneName=thigh_volume_fn_r}
    translation {y=-10, z=5}
   }
   n48 = "Transform (Modify) Bone - Bone: thigh_volume_bk_l"(ComponentPose=@n47, Alpha=@n44, Translation="-15,0,0", Rotation="0,0,0", Scale="1,1,1") {
    rotationMode=BMM_Additive
    rotationSpace=BCS_BoneSpace
    scaleMode=BMM_Additive
    scaleSpace=BCS_BoneSpace
    translationMode=BMM_Additive
    translationSpace=BCS_BoneSpace
    alphaBoolBlend {blendOption=Linear}
    boneToModify {boneName=thigh_volume_bk_l}
    translation {x=-15}
   }
   n49 = "Transform (Modify) Bone - Bone: thigh_volume_bk_r"(ComponentPose=@n48, Alpha=@n46, Translation="10,0,0", Rotation="0,0,0", Scale="1,1,1") {
    rotationSpace=BCS_BoneSpace
    scaleMode=BMM_Additive
    scaleSpace=BCS_BoneSpace
    translationMode=BMM_Additive
    translationSpace=BCS_BoneSpace
    alphaBoolBlend {blendOption=Linear}
    boneToModify {boneName=thigh_volume_bk_r}
    translation {x=10}
   }
   n50 = "Transform (Modify) Bone - Bone: thigh_volume_lt_l1"(ComponentPose=@n49, Alpha=Get thigh_shape_lt_l() -> (thigh_shape_lt_l:double), Translation="0,20,0", Rotation="0,0,0", Scale="1,1,1") {
    rotationMode=BMM_Additive
    rotationSpace=BCS_BoneSpace
    scaleMode=BMM_Additive
    scaleSpace=BCS_BoneSpace
    translationMode=BMM_Additive
    translationSpace=BCS_BoneSpace
    alphaBoolBlend {blendOption=Linear}
    boneToModify {boneName=thigh_volume_lt_l1}
    translation {y=20}
   }
   n51 = "Transform (Modify) Bone - Bone: thigh_volume_rt_r"(ComponentPose=@n50, Alpha=Get thigh_shape_rt_r() -> (thigh_shape_rt_r:double), Translation="0,-20,0", Rotation="0,0,0", Scale="1,1,1") {
    rotationSpace=BCS_BoneSpace
    scaleMode=BMM_Additive
    scaleSpace=BCS_BoneSpace
    translationMode=BMM_Additive
    translationSpace=BCS_BoneSpace
    alphaBoolBlend {blendOption=Linear}
    boneToModify {boneName=thigh_volume_rt_r}
    translation {y=-20}
   }
   n52 = "Constraint - upperarm_l_fn (PP)"!(ComponentPose=@n51, Alpha=1, ConstraintWeights_0=0.5, ConstraintWeights_1:float=0.5) -> (Pose) {
    alphaBoolBlend {blendOption=Linear}
    boneToModify {boneName=upperarm_l_fn}
    constraintSetup [{targetBone {boneName=clavicle_l}}, {targetBone {boneName=upperarm_twist_01_l}}]
    constraintWeights [1,1]
   }
   n53 = "Constraint - upperarm_l_bn (PP)"!(ComponentPose=@n52, Alpha=1, ConstraintWeights_0=0.5, ConstraintWeights_1:float=0.5) -> (Pose) {
    alphaBoolBlend {blendOption=Linear}
    boneToModify {boneName=upperarm_l_bn}
    constraintSetup [{targetBone {boneName=clavicle_l}}, {targetBone {boneName=upperarm_twist_01_l}}]
    constraintWeights [1,1]
   }
   n54 = Get Shoulder_rot_f_l() -> (Shoulder_rot_f_l:double)
   n55 = "Transform (Modify) Bone - Bone: upperarm_l_fn"(ComponentPose=@n53, Alpha=@n54, Translation="0,20,0", Rotation="0,0,0", Scale="1,1,1") {
    rotationMode=BMM_Additive
    rotationSpace=BCS_BoneSpace
    scaleMode=BMM_Additive
    scaleSpace=BCS_BoneSpace
    translationMode=BMM_Additive
    translationSpace=BCS_BoneSpace
    alphaBoolBlend {blendOption=Linear}
    boneToModify {boneName=upperarm_l_fn}
    translation {y=20}
   }
   n56 = "Transform (Modify) Bone - Bone: upperarm_l_bn"(ComponentPose=@n55, Alpha=@n54, Translation="5,-5,0", Rotation="0,0,0", Scale="1,1,1") {
    rotationMode=BMM_Additive
    rotationSpace=BCS_BoneSpace
    scaleMode=BMM_Additive
    scaleSpace=BCS_BoneSpace
    translationMode=BMM_Additive
    translationSpace=BCS_BoneSpace
    alphaBoolBlend {blendOption=Linear}
    boneToModify {boneName=upperarm_l_bn}
    translation=<5 -5>
   }
   n57 = "Constraint - upperarm_r_fn (PP)"!(ComponentPose=@n56, Alpha=1, ConstraintWeights_0=0.5, ConstraintWeights_1:float=0.5) -> (Pose) {
    alphaBoolBlend {blendOption=Linear}
    boneToModify {boneName=upperarm_r_fn}
    constraintSetup [{targetBone {boneName=clavicle_r}}, {targetBone {boneName=upperarm_twist_01_r}}]
    constraintWeights [1,1]
   }
   n58 = "Constraint - upperarm_r_bn (PP)"!(ComponentPose=@n57, Alpha=1, ConstraintWeights_0=0.5, ConstraintWeights_1:float=0.5) -> (Pose) {
    alphaBoolBlend {blendOption=Linear}
    boneToModify {boneName=upperarm_r_bn}
    constraintSetup [{targetBone {boneName=clavicle_r}}, {targetBone {boneName=upperarm_twist_01_r}}]
    constraintWeights [1,1]
   }
   n59 = Get Shoulder_rot_f_r() -> (Shoulder_rot_f_r:double)
   n60 = "Transform (Modify) Bone - Bone: upperarm_r_fn"(ComponentPose=@n58, Alpha=@n59, Translation="0,-20,0", Rotation="0,0,0", Scale="1,1,1") {
    rotationMode=BMM_Additive
    rotationSpace=BCS_BoneSpace
    scaleMode=BMM_Additive
    scaleSpace=BCS_BoneSpace
    translationMode=BMM_Additive
    translationSpace=BCS_BoneSpace
    alphaBoolBlend {blendOption=Linear}
    boneToModify {boneName=upperarm_r_fn}
    translation {y=-20}
   }
   n61 = "Transform (Modify) Bone - Bone: upperarm_r_bn"(ComponentPose=@n60, Alpha=@n59, Translation="5,5,5", Rotation="0,0,0", Scale="1,1,1") {
    rotationMode=BMM_Additive
    rotationSpace=BCS_BoneSpace
    scaleMode=BMM_Additive
    scaleSpace=BCS_BoneSpace
    translationMode=BMM_Additive
    translationSpace=BCS_BoneSpace
    alphaBoolBlend {blendOption=Linear}
    boneToModify {boneName=upperarm_r_bn}
    translation=<5 5 5>
   }
   n62 = "Constraint - dyn_upperarm_cloth_l (PPP)"!(ComponentPose=@n61, Alpha=1, ConstraintWeights_0=0, ConstraintWeights_1:float=0.2, ConstraintWeights_2:float=0.8) -> (Pose) {
    alphaBoolBlend {blendOption=Linear}
    boneToModify {boneName=dyn_upperarm_cloth_l}
    constraintSetup [{targetBone {boneName=upperarm_twist_01_l}}, {targetBone {boneName=lowerarm_l}}, {targetBone {boneName=upperarm_twist_02_l}}]
    constraintWeights [1,1,1]
   }
   n63 = "Constraint - dyn_upperarm_croth_r (PPP)"!(ComponentPose=@n62, Alpha=1, ConstraintWeights_0=0, ConstraintWeights_1:float=0.2, ConstraintWeights_2:float=0.8) -> (Pose) {
    alphaBoolBlend {blendOption=Linear}
    boneToModify {boneName=dyn_upperarm_croth_r}
    constraintSetup [{targetBone {boneName=upperarm_twist_01_r}}, {targetBone {boneName=lowerarm_r}}, {targetBone {boneName=upperarm_twist_02_r}}]
    constraintWeights [1,1,1]
   }
   n64 = Get lowerarm_rot_up_l() -> (lowerarm_rot_up_l:double)
   n65 = "Transform (Modify) Bone - Bone: upperarm_twist_01_l"(ComponentPose=@n63, Alpha=@n64, Translation="0,0,0", Rotation="0,0,0", Scale="1,1.2,1.2") {
    rotationMode=BMM_Additive
    rotationSpace=BCS_BoneSpace
    scaleMode=BMM_Additive
    scaleSpace=BCS_BoneSpace
    translationMode=BMM_Additive
    translationSpace=BCS_BoneSpace
    alphaBoolBlend {blendOption=Linear}
    boneToModify {boneName=upperarm_twist_01_l}
    scale {y=1.2, z=1.2}
   }
   n66 = "Constraint - dyn_upperarm_cloth_al (PP)"!(ComponentPose=@n65, Alpha=@n64, ConstraintWeights_0=0.1, ConstraintWeights_1:float=0.9) -> (Pose) {
    alphaBoolBlend {blendOption=Linear}
    boneToModify {boneName=dyn_upperarm_cloth_al}
    constraintSetup [{targetBone {boneName=upperarm_twist_02_l}}, {targetBone {boneName=lowerarm_l}}]
    constraintWeights [1,1]
   }
   n67 = "Constraint - dyn_upperarm_cloth_bl (PP)"!(ComponentPose=@n66, Alpha=@n64, ConstraintWeights_0=0.5, ConstraintWeights_1:float=0.5) -> (Pose) {
    alphaBoolBlend {blendOption=Linear}
    boneToModify {boneName=dyn_upperarm_cloth_bl}
    constraintSetup [{targetBone {boneName=upperarm_twist_02_l}}, {targetBone {boneName=lowerarm_l}}]
    constraintWeights [1,1]
   }
   n68 = "Transform (Modify) Bone - Bone: dyn_upperarm_cloth_al"(ComponentPose=@n67, Alpha=@n64, Translation="6,4,2", Rotation="-10,-20,0", Scale="1,1,1") {
    rotationMode=BMM_Additive
    rotationSpace=BCS_BoneSpace
    translationMode=BMM_Additive
    translationSpace=BCS_BoneSpace
    alphaBoolBlend {blendOption=Linear}
    boneToModify {boneName=dyn_upperarm_cloth_al}
    rotation {pitch=-10, yaw=-20}
    translation=<6 4 2>
   }
   n69 = "Transform (Modify) Bone - Bone: dyn_upperarm_cloth_bl"(ComponentPose=@n68, Alpha=@n64, Translation="0,0,3", Rotation="0,-10,0", Scale="1,1,1") {
    rotationMode=BMM_Additive
    rotationSpace=BCS_BoneSpace
    translationMode=BMM_Additive
    translationSpace=BCS_BoneSpace
    alphaBoolBlend {blendOption=Linear}
    boneToModify {boneName=dyn_upperarm_cloth_bl}
    rotation {yaw=-10}
    translation {z=3}
   }
   n70 = Get lowerarm_rot_up_r() -> (lowerarm_rot_up_r:double)
   n71 = "Constraint - dyn_upperarm_croth_ar (PP)"!(ComponentPose=@n69, Alpha=@n70, ConstraintWeights_0=0.1, ConstraintWeights_1:float=0.9) -> (Pose) {
    alphaBoolBlend {blendOption=Linear}
    boneToModify {boneName=dyn_upperarm_croth_ar}
    constraintSetup [{targetBone {boneName=upperarm_twist_02_r}}, {targetBone {boneName=lowerarm_r}}]
    constraintWeights [1,1]
   }
   n72 = "Constraint - dyn_upperarm_croth_br (PP)"!(ComponentPose=@n71, Alpha=@n70, ConstraintWeights_0=0.5, ConstraintWeights_1:float=0.5) -> (Pose) {
    alphaBoolBlend {blendOption=Linear}
    boneToModify {boneName=dyn_upperarm_croth_br}
    constraintSetup [{targetBone {boneName=upperarm_twist_02_r}}, {targetBone {boneName=lowerarm_r}}]
    constraintWeights [1,1]
   }
   n73 = "Transform (Modify) Bone - Bone: dyn_upperarm_croth_ar"(ComponentPose=@n72, Alpha=@n70, Translation="-6,-4,-2", Rotation="-10,-20,0", Scale="1,1,1") {
    rotationMode=BMM_Additive
    rotationSpace=BCS_BoneSpace
    translationMode=BMM_Additive
    translationSpace=BCS_BoneSpace
    alphaBoolBlend {blendOption=Linear}
    boneToModify {boneName=dyn_upperarm_croth_ar}
    rotation {pitch=-10, yaw=-20}
    translation=<-6 -4 -2>
   }
   n74 = "Transform (Modify) Bone - Bone: dyn_upperarm_croth_br"(ComponentPose=@n73, Alpha=@n70, Translation="0,0,-3", Rotation="0,-10,0", Scale="1,1,1") {
    rotationMode=BMM_Additive
    rotationSpace=BCS_BoneSpace
    translationMode=BMM_Additive
    translationSpace=BCS_BoneSpace
    alphaBoolBlend {blendOption=Linear}
    boneToModify {boneName=dyn_upperarm_croth_br}
    rotation {yaw=-10}
    translation {z=-3}
   }
   n75 = Control Rig!(Source=Component To Local(ComponentPose=@n74), Alpha, is sliding:double=0) -> (Pose) {
    alphaBoolBlend {blendOption=Linear}
    controlRigAssetReference {blueprintRigClass=&/M_MED_DriftTrooper_Skirt_CtrlRig.M_MED_DriftTrooper_Skirt_CtrlRig_C}
   }
   n76 = Control Rig!(Source=@n75, Alpha) -> (Pose) {
    alphaBoolBlend {blendOption=Linear}
    controlRigAssetReference {blueprintRigClass=&/M_MED_DriftTrooper_CtrlRig.M_MED_DriftTrooper_CtrlRig_C}
   }
   n77 = Get is sliding() -> (is sliding:double)
   n78 = "Constraint - dyn_Cloth_volume_l_C01 (PP)"(ComponentPose=Local To Component(LocalPose=@n76), Alpha=@n77, ConstraintWeights_0=0.5, ConstraintWeights_1=0.5) {
    alphaBoolBlend {blendOption=Linear}
    boneToModify {boneName=dyn_Cloth_volume_l_C01}
    constraintSetup [{targetBone {boneName=dyn_Cloth_volume_l_C}}, {targetBone {boneName=spine_05}}]
    constraintWeights [1,1]
   }
   n79 = "Constraint - dyn_Cloth_volume_r_C01 (PP)"(ComponentPose=@n78, Alpha=@n77, ConstraintWeights_0=0.5, ConstraintWeights_1=0.5) {
    alphaBoolBlend {blendOption=Linear}
    boneToModify {boneName=dyn_Cloth_volume_r_C01}
    constraintSetup [{targetBone {boneName=dyn_Cloth_volume_r_C}}, {targetBone {boneName=spine_05}}]
    constraintWeights [1,1]
   }
   n80 = "Transform (Modify) Bone - Bone: dyn_Cloth_volume_l_C"(ComponentPose=@n79, Alpha=@n77, Translation="0,0,0", Rotation="20,40,0", Scale="1,1,1") {
    rotationMode=BMM_Additive
    rotationSpace=BCS_BoneSpace
    scaleMode=BMM_Additive
    scaleSpace=BCS_BoneSpace
    translationMode=BMM_Additive
    translationSpace=BCS_BoneSpace
    alphaBoolBlend {blendOption=Linear}
    boneToModify {boneName=dyn_Cloth_volume_l_C}
    rotation {pitch=20, yaw=40}
   }
   n81 = "Transform (Modify) Bone - Bone: dyn_Cloth_volume_r_C"(ComponentPose=@n80, Alpha=@n77, Translation="0,0,0", Rotation="-20,40,0", Scale="1,1,1") {
    rotationMode=BMM_Additive
    rotationSpace=BCS_BoneSpace
    scaleMode=BMM_Additive
    scaleSpace=BCS_BoneSpace
    translationMode=BMM_Additive
    translationSpace=BCS_BoneSpace
    alphaBoolBlend {blendOption=Linear}
    boneToModify {boneName=dyn_Cloth_volume_r_C}
    rotation {pitch=-20, yaw=40}
   }
   n82 = "Transform (Modify) Bone - Bone: dyn_Cloth_volume_l_C03"(ComponentPose=@n81, Alpha=@n77, Translation="0,0,0", Rotation="0,-10,0", Scale="1,1,1") {
    rotationMode=BMM_Additive
    rotationSpace=BCS_BoneSpace
    scaleMode=BMM_Additive
    scaleSpace=BCS_BoneSpace
    translationMode=BMM_Additive
    translationSpace=BCS_BoneSpace
    alphaBoolBlend {blendOption=Linear}
    boneToModify {boneName=dyn_Cloth_volume_l_C03}
    rotation {yaw=-10}
   }
   n83 = "Transform (Modify) Bone - Bone: dyn_Cloth_volume_r_C03"(ComponentPose=@n82, Alpha=@n77, Translation="0,0,0", Rotation="0,-10,0", Scale="1,1,1") {
    rotationMode=BMM_Additive
    rotationSpace=BCS_BoneSpace
    scaleMode=BMM_Additive
    scaleSpace=BCS_BoneSpace
    translationMode=BMM_Additive
    translationSpace=BCS_BoneSpace
    alphaBoolBlend {blendOption=Linear}
    boneToModify {boneName=dyn_Cloth_volume_r_C03}
    rotation {yaw=-10}
   }
   n84 = "Transform (Modify) Bone - Bone: dyn_Cloth_volume_l_B"(ComponentPose=@n83, Alpha=@n77, Translation="0,0,0", Rotation="0,10,0", Scale="1,1,1") {
    rotationMode=BMM_Additive
    rotationSpace=BCS_BoneSpace
    scaleMode=BMM_Additive
    scaleSpace=BCS_BoneSpace
    translationMode=BMM_Additive
    translationSpace=BCS_BoneSpace
    alphaBoolBlend {blendOption=Linear}
    boneToModify {boneName=dyn_Cloth_volume_l_B}
    rotation {yaw=10}
   }
   n85 = "Transform (Modify) Bone - Bone: dyn_Cloth_volume_r_B"(ComponentPose=@n84, Alpha=@n77, Translation="0,0,0", Rotation="0,10,0", Scale="1,1,1") {
    rotationMode=BMM_Additive
    rotationSpace=BCS_BoneSpace
    scaleMode=BMM_Additive
    scaleSpace=BCS_BoneSpace
    translationMode=BMM_Additive
    translationSpace=BCS_BoneSpace
    alphaBoolBlend {blendOption=Linear}
    boneToModify {boneName=dyn_Cloth_volume_r_B}
    rotation {yaw=10}
   }
   n86 = "Constraint - dyn_Cloth_volume_bk (PP)"(ComponentPose=@n85, Alpha=1, ConstraintWeights_0=0.5, ConstraintWeights_1=0.5) {
    alphaBoolBlend {blendOption=Linear}
    boneToModify {boneName=dyn_Cloth_volume_bk}
    constraintSetup [{targetBone {boneName=dyn_Cloth_volume_l_C}}, {targetBone {boneName=dyn_Cloth_volume_r_C}}]
    constraintWeights [1,1]
   }
   n87 = "Constraint - dyn_Cloth_volume_bk01 (PP)"(ComponentPose=@n86, Alpha=1, ConstraintWeights_0=0.5, ConstraintWeights_1=0.5) {
    alphaBoolBlend {blendOption=Linear}
    boneToModify {boneName=dyn_Cloth_volume_bk01}
    constraintSetup [{targetBone {boneName=dyn_Cloth_volume_l_C01}}, {targetBone {boneName=dyn_Cloth_volume_r_C01}}]
    constraintWeights [1,1]
   }
   n88 = "Constraint - dyn_Cloth_volume_bk02 (PP)"(ComponentPose=@n87, Alpha=1, ConstraintWeights_0=0.5, ConstraintWeights_1=0.5) {
    alphaBoolBlend {blendOption=Linear}
    boneToModify {boneName=dyn_Cloth_volume_bk02}
    constraintSetup [{targetBone {boneName=dyn_Cloth_volume_l_C02}}, {targetBone {boneName=dyn_Cloth_volume_r_C02}}]
    constraintWeights [1,1]
   }
   n89 = "Constraint - dyn_Cloth_volume_bk03 (PP)"(ComponentPose=@n88, Alpha=1, ConstraintWeights_0=0.5, ConstraintWeights_1=0.5) {
    alphaBoolBlend {blendOption=Linear}
    boneToModify {boneName=dyn_Cloth_volume_bk03}
    constraintSetup [{targetBone {boneName=dyn_Cloth_volume_l_C03}}, {targetBone {boneName=dyn_Cloth_volume_r_C03}}]
    constraintWeights [1,1]
   }
   n90 = "Constraint - dyn_Cloth_volume_bk04 (PP)"!(ComponentPose=@n89, Alpha=1, ConstraintWeights_0=0.5, ConstraintWeights_1:float=0.5) -> (Pose) {
    alphaBoolBlend {blendOption=Linear}
    boneToModify {boneName=dyn_Cloth_volume_bk04}
    constraintSetup [{targetBone {boneName=dyn_Cloth_volume_l_C04}}, {targetBone {boneName=dyn_Cloth_volume_r_C04}}]
    constraintWeights [1,1]
   }
   n91 = "Constraint - fix_spine_l_02 (PPP)"!(ComponentPose=@n90, Alpha=1, ConstraintWeights_0=0.4, ConstraintWeights_1:float=0.4, ConstraintWeights_2:float=0.2) -> (Pose) {
    alphaBoolBlend {blendOption=Linear}
    boneToModify {boneName=fix_spine_l_02}
    constraintSetup [{targetBone {boneName=spine_05}}, {targetBone {boneName=clavicle_l}}, {targetBone {boneName=head}}]
    constraintWeights [1,1,1]
   }
   n92 = "Constraint - fix_spine_r_02 (PPP)"!(ComponentPose=@n91, Alpha=1, ConstraintWeights_0=0.4, ConstraintWeights_1:float=0.4, ConstraintWeights_2:float=0.2) -> (Pose) {
    alphaBoolBlend {blendOption=Linear}
    boneToModify {boneName=fix_spine_r_02}
    constraintSetup [{targetBone {boneName=spine_05}}, {targetBone {boneName=clavicle_r}}, {targetBone {boneName=head}}]
    constraintWeights [1,1,1]
   }
   n93 = "Transform (Modify) Bone - Bone: fix_collar"(ComponentPose=@n92, Alpha=Get Head_rot_f() -> (Head_rot_f:double), Translation="-5,0,0", Rotation="0,0,0", Scale="1,1,1") {
    rotationMode=BMM_Additive
    rotationSpace=BCS_BoneSpace
    scaleMode=BMM_Additive
    scaleSpace=BCS_BoneSpace
    translationMode=BMM_Additive
    translationSpace=BCS_BoneSpace
    alphaBoolBlend {blendOption=Linear}
    boneToModify {boneName=fix_collar}
    translation {x=-5}
   }
   Save cached pose 'ADSPose'(Pose=Component To Local(ComponentPose=@n93)) {cacheName=ADSPose}
   n94 = Get thigh_shape_fn_r_0() -> (thigh_shape_fn_r_0:double)
   n95 = "Transform (Modify) Bone - Bone: dyn_Cloth_volume_l_C"(
    ComponentPose=Local To Component(LocalPose=Use cached pose 'CrouchingPose'())
    Alpha=@n94
    Translation="0,0,0"
    Rotation="0,32,0"
    Scale="1,1,1"
   ) {
    rotationMode=BMM_Additive
    rotationSpace=BCS_BoneSpace
    scaleMode=BMM_Additive
    scaleSpace=BCS_BoneSpace
    translationMode=BMM_Additive
    translationSpace=BCS_BoneSpace
    alphaBoolBlend {blendOption=Linear}
    boneToModify {boneName=dyn_Cloth_volume_l_C}
    rotation {yaw=32}
   }
   n96 = "Transform (Modify) Bone - Bone: dyn_Cloth_volume_l_C03"(ComponentPose=@n95, Alpha=@n94, Translation="0,0,0", Rotation="0,27,0", Scale="1,1,1") {
    rotationMode=BMM_Additive
    rotationSpace=BCS_BoneSpace
    scaleMode=BMM_Additive
    scaleSpace=BCS_BoneSpace
    translationMode=BMM_Additive
    translationSpace=BCS_BoneSpace
    alphaBoolBlend {blendOption=Linear}
    boneToModify {boneName=dyn_Cloth_volume_l_C03}
    rotation {yaw=27}
   }
   n97 = "Transform (Modify) Bone - Bone: dyn_Cloth_volume_l_B02"(ComponentPose=@n96, Alpha=@n94, Translation="0,0,0", Rotation="0,10,0", Scale="1,1,1") {
    rotationMode=BMM_Additive
    rotationSpace=BCS_BoneSpace
    scaleMode=BMM_Additive
    scaleSpace=BCS_BoneSpace
    translationMode=BMM_Additive
    translationSpace=BCS_BoneSpace
    alphaBoolBlend {blendOption=Linear}
    boneToModify {boneName=dyn_Cloth_volume_l_B02}
    rotation {yaw=10}
   }
   n98 = "Transform (Modify) Bone - Bone: dyn_Cloth_volume_r_C02"(ComponentPose=@n97, Alpha=1, Translation="0,0,0", Rotation="0,20,0", Scale="1,1,1") {
    rotationMode=BMM_Additive
    rotationSpace=BCS_BoneSpace
    scaleMode=BMM_Additive
    scaleSpace=BCS_BoneSpace
    translationMode=BMM_Additive
    translationSpace=BCS_BoneSpace
    alphaBoolBlend {blendOption=Linear}
    boneToModify {boneName=dyn_Cloth_volume_r_C02}
    rotation {yaw=20}
   }
   n99 = "Transform (Modify) Bone - Bone: dyn_Cloth_volume_r_C03"(ComponentPose=@n98, Alpha=1, Translation="0,0,0", Rotation="0,18,0", Scale="1,1,1") {
    rotationMode=BMM_Additive
    rotationSpace=BCS_BoneSpace
    scaleMode=BMM_Additive
    scaleSpace=BCS_BoneSpace
    translationMode=BMM_Additive
    translationSpace=BCS_BoneSpace
    alphaBoolBlend {blendOption=Linear}
    boneToModify {boneName=dyn_Cloth_volume_r_C03}
    rotation {yaw=18}
   }
   n100 = "Transform (Modify) Bone - Bone: dyn_Cloth_volume_r_C04"(ComponentPose=@n99, Alpha=1, Translation="0,0,0", Rotation="0,10,0", Scale="1,1,1") {
    rotationMode=BMM_Additive
    rotationSpace=BCS_BoneSpace
    scaleMode=BMM_Additive
    scaleSpace=BCS_BoneSpace
    translationMode=BMM_Additive
    translationSpace=BCS_BoneSpace
    alphaBoolBlend {blendOption=Linear}
    boneToModify {boneName=dyn_Cloth_volume_r_C04}
    rotation {yaw=10}
   }
   n101 = "Transform (Modify) Bone - Bone: dyn_Cloth_volume_r_B01"(ComponentPose=@n100, Alpha=1, Translation="0,0,0", Rotation="0,18,0", Scale="1,1,1") {
    rotationMode=BMM_Additive
    rotationSpace=BCS_BoneSpace
    scaleMode=BMM_Additive
    scaleSpace=BCS_BoneSpace
    translationMode=BMM_Additive
    translationSpace=BCS_BoneSpace
    alphaBoolBlend {blendOption=Linear}
    boneToModify {boneName=dyn_Cloth_volume_r_B01}
    rotation {yaw=18}
   }
   n102 = Component To Local(ComponentPose=@n101)
   n103 = Use cached pose 'CrouchingPose'()
   n104 = Get CrouchingAndNotMovingAndNotShape() -> (CrouchingAndNotMovingAndNotShape:bool)
   n105 = "Constraint - dyn_Cloth_l_02 (PP)"!(
    ComponentPose=Local To Component(LocalPose=Blend Poses by bool(BlendPose_0=@n102, BlendPose_1=@n103, bActiveValue=@n104) {blendType=Linear})
    Alpha=1
    ConstraintWeights_0=0
    ConstraintWeights_1:float=1
   ) -> (Pose) {
    alphaBoolBlend {blendOption=Linear}
    boneToModify {boneName=dyn_Cloth_l_02}
    constraintSetup [{targetBone {boneName=dyn_Cloth_volume_l_C02}}, {targetBone {boneName=dyn_Cloth_volume_l_C03}}]
    constraintWeights [1,1]
   }
   n106 = "Constraint - dyn_Cloth_l_03 (PP)"!(ComponentPose=@n105, Alpha=1, ConstraintWeights_0=0, ConstraintWeights_1:float=1) -> (Pose) {
    alphaBoolBlend {blendOption=Linear}
    boneToModify {boneName=dyn_Cloth_l_03}
    constraintSetup [{targetBone {boneName=dyn_Cloth_volume_l_C02}}, {targetBone {boneName=dyn_Cloth_volume_l_C04}}]
    constraintWeights [1,1]
   }
   n107 = "Constraint - dyn_Cloth_l_04 (PP)"!(ComponentPose=@n106, Alpha=1, ConstraintWeights_0=0, ConstraintWeights_1:float=1) -> (Pose) {
    alphaBoolBlend {blendOption=Linear}
    boneToModify {boneName=dyn_Cloth_l_04}
    constraintSetup [{targetBone {boneName=dyn_Cloth_volume_l_C02}}, {targetBone {boneName=dyn_Cloth_volume_l_C05}}]
    constraintWeights [1,1]
   }
   n108 = "Constraint - dyn_Cloth_r_02 (PP)"!(ComponentPose=@n107, Alpha=1, ConstraintWeights_0=0, ConstraintWeights_1:float=1) -> (Pose) {
    alphaBoolBlend {blendOption=Linear}
    boneToModify {boneName=dyn_Cloth_r_02}
    constraintSetup [{targetBone {boneName=dyn_Cloth_volume_r_C02}}, {targetBone {boneName=dyn_Cloth_volume_r_C03}}]
    constraintWeights [1,1]
   }
   n109 = "Constraint - dyn_Cloth_r_03 (PP)"!(ComponentPose=@n108, Alpha=1, ConstraintWeights_0=0, ConstraintWeights_1:float=1) -> (Pose) {
    alphaBoolBlend {blendOption=Linear}
    boneToModify {boneName=dyn_Cloth_r_03}
    constraintSetup [{targetBone {boneName=dyn_Cloth_volume_r_C02}}, {targetBone {boneName=dyn_Cloth_volume_r_C04}}]
    constraintWeights [1,1]
   }
   n110 = "Constraint - dyn_Cloth_r_04 (PP)"!(ComponentPose=@n109, Alpha=1, ConstraintWeights_0=0, ConstraintWeights_1:float=1) -> (Pose) {
    alphaBoolBlend {blendOption=Linear}
    boneToModify {boneName=dyn_Cloth_r_04}
    constraintSetup [{targetBone {boneName=dyn_Cloth_volume_r_C02}}, {targetBone {boneName=dyn_Cloth_volume_r_C05}}]
    constraintWeights [1,1]
   }
   n111 = "Constraint - dyn_Cloth_volume_bk (PP)"(ComponentPose=@n110, Alpha=1, ConstraintWeights_0=0.5, ConstraintWeights_1=0.5) {
    alphaBoolBlend {blendOption=Linear}
    boneToModify {boneName=dyn_Cloth_volume_bk}
    constraintSetup [{targetBone {boneName=dyn_Cloth_volume_l_C}}, {targetBone {boneName=dyn_Cloth_volume_r_C}}]
    constraintWeights [1,1]
   }
   n112 = "Constraint - dyn_Cloth_volume_bk01 (PP)"(ComponentPose=@n111, Alpha=1, ConstraintWeights_0=0.5, ConstraintWeights_1=0.5) {
    alphaBoolBlend {blendOption=Linear}
    boneToModify {boneName=dyn_Cloth_volume_bk01}
    constraintSetup [{targetBone {boneName=dyn_Cloth_volume_l_C01}}, {targetBone {boneName=dyn_Cloth_volume_r_C01}}]
    constraintWeights [1,1]
   }
   n113 = "Constraint - dyn_Cloth_volume_bk02 (PP)"(ComponentPose=@n112, Alpha=1, ConstraintWeights_0=0.5, ConstraintWeights_1=0.5) {
    alphaBoolBlend {blendOption=Linear}
    boneToModify {boneName=dyn_Cloth_volume_bk02}
    constraintSetup [{targetBone {boneName=dyn_Cloth_volume_l_C02}}, {targetBone {boneName=dyn_Cloth_volume_r_C02}}]
    constraintWeights [1,1]
   }
   n114 = "Constraint - dyn_Cloth_volume_bk03 (PP)"(ComponentPose=@n113, Alpha=1, ConstraintWeights_0=0.5, ConstraintWeights_1=0.5) {
    alphaBoolBlend {blendOption=Linear}
    boneToModify {boneName=dyn_Cloth_volume_bk03}
    constraintSetup [{targetBone {boneName=dyn_Cloth_volume_l_C03}}, {targetBone {boneName=dyn_Cloth_volume_r_C03}}]
    constraintWeights [1,1]
   }
   n115 = "Constraint - dyn_Cloth_volume_bk04 (PP)"(ComponentPose=@n114, Alpha=1, ConstraintWeights_0=0.5, ConstraintWeights_1=0.5) {
    alphaBoolBlend {blendOption=Linear}
    boneToModify {boneName=dyn_Cloth_volume_bk04}
    constraintSetup [{targetBone {boneName=dyn_Cloth_volume_l_C04}}, {targetBone {boneName=dyn_Cloth_volume_r_C04}}]
    constraintWeights [1,1]
   }
   Save cached pose 'SkirtCachePose'(Pose=Component To Local(ComponentPose=@n115)) {cacheName=SkirtCachePose}
   n116 = Get offset_weapon() -> (offset_weapon:double)
   n117 = "Transform (Modify) Bone - Bone: fix_spine_r_02"(ComponentPose=Local To Component(LocalPose=Use cached pose 'ADSPose'()), Alpha=@n116, Translation="0,0,-10", Rotation="0,0,0", Scale="1,1,1") {
    rotationMode=BMM_Additive
    rotationSpace=BCS_BoneSpace
    scaleMode=BMM_Additive
    scaleSpace=BCS_BoneSpace
    translationMode=BMM_Additive
    translationSpace=BCS_BoneSpace
    alphaBoolBlend {blendOption=Linear}
    boneToModify {boneName=fix_spine_r_02}
    translation {z=-10}
   }
   n118 = "Transform (Modify) Bone - Bone: fix_spine_r_01"(ComponentPose=@n117, Alpha=@n116, Translation="0,0,-10", Rotation="0,0,0", Scale="1,1,1") {
    rotationMode=BMM_Additive
    rotationSpace=BCS_BoneSpace
    scaleMode=BMM_Additive
    scaleSpace=BCS_BoneSpace
    translationMode=BMM_Additive
    translationSpace=BCS_BoneSpace
    alphaBoolBlend {blendOption=Linear}
    boneToModify {boneName=fix_spine_r_01}
    translation {z=-10}
   }
   n119 = Get offset_weapon()
   n120 = "Transform (Modify) Bone - Bone: fix_spine_r_02"(ComponentPose=Local To Component(LocalPose=Use cached pose 'ADSPose'()), Alpha=@n119, Translation="0,0,-10", Rotation="0,0,0", Scale="1,1,1") {
    rotationMode=BMM_Additive
    rotationSpace=BCS_BoneSpace
    scaleMode=BMM_Additive
    scaleSpace=BCS_BoneSpace
    translationMode=BMM_Additive
    translationSpace=BCS_BoneSpace
    alphaBoolBlend {blendOption=Linear}
    boneToModify {boneName=fix_spine_r_02}
    translation {z=-10}
   }
   n121 = "Transform (Modify) Bone - Bone: fix_spine_r_01"(ComponentPose=@n120, Alpha=@n119, Translation="0,0,-10", Rotation="0,0,0", Scale="1,1,1") {
    rotationMode=BMM_Additive
    rotationSpace=BCS_BoneSpace
    scaleMode=BMM_Additive
    scaleSpace=BCS_BoneSpace
    translationMode=BMM_Additive
    translationSpace=BCS_BoneSpace
    alphaBoolBlend {blendOption=Linear}
    boneToModify {boneName=fix_spine_r_01}
    translation {z=-10}
   }
   Save cached pose 'CrouchingPose'(Pose=Blend Poses by bool(BlendPose_0=Component To Local(ComponentPose=@n121), BlendPose_1=Component To Local(ComponentPose=@n118), bActiveValue=Get bIsTargeting()) {blendType=Linear}) {cacheName=CrouchingPose}
   n122 = RigidBody!(
    ComponentPose=Local To Component(LocalPose=Use cached pose 'MovingPose'())
    bAlphaBoolEnabled:bool=Get is_WearingCape() -> (is_WearingCape:bool)
    OverrideWorldGravity:Vector="0,0,-980"
    ExternalForce
   ) -> (Pose) {
    alphaInputType=Bool
    bClampLinearTranslationLimitToRefPose=true
    bOverrideWorldGravity=true
    overridePhysicsAsset=/Game/Characters/Player/Male/Medium/Bodies/M_MED_DriftTrooper/Meshes/M_MED_DriftTrooper_Physics
    simulationSpace=BaseBoneSpace
    alphaBoolBlend {blendOption=Linear}
    baseBoneRef {boneName=gravity}
    componentAppliedLinearAccClamp=<3000 3000 2700>
    componentLinearAccScale=<0.01 0.4 0.35>
    componentLinearVelScale=<0.08 0.4 0.4>
    overrideWorldGravity {z=-980}
   }
   n123 = Component To Local(ComponentPose=@n122)
   n124 = RigidBody(
    ComponentPose=Local To Component(LocalPose=Use cached pose 'MovingPose'())
    bAlphaBoolEnabled=Get is_WearingCape()
    OverrideWorldGravity="0,0,-980"
   ) {
    alphaInputType=Bool
    bClampLinearTranslationLimitToRefPose=true
    bOverrideWorldGravity=true
    overridePhysicsAsset=&/M_MED_DriftTrooper_MovingPhysics
    simulationSpace=BaseBoneSpace
    alphaBoolBlend {blendOption=Linear}
    baseBoneRef {boneName=gravity}
    componentAppliedLinearAccClamp=<3000 3000 2700>
    componentLinearAccScale=<0.01 0.4 0.35>
    componentLinearVelScale=<0.08 0.4 0.4>
    overrideWorldGravity {z=-980}
   }
   n125 = Component To Local(ComponentPose=@n124)
   n126 = Get MovingAnd NotCrouchNotDivingNotEmote()
   n127 = RigidBody(
    ComponentPose=Local To Component(LocalPose=Use cached pose 'MovingPose'())
    bAlphaBoolEnabled=Get is_WearingCape()
    OverrideWorldGravity="0,0,-980"
   ) {
    alphaInputType=Bool
    bClampLinearTranslationLimitToRefPose=true
    bOverrideWorldGravity=true
    overridePhysicsAsset=&/M_MED_DriftTrooper_FrontendPhysics
    simulationSpace=BaseBoneSpace
    alphaBoolBlend {blendOption=Linear}
    baseBoneRef {boneName=gravity}
    componentAppliedLinearAccClamp=<3000 3000 2700>
    componentLinearAccScale=<0.01 0.4 0.35>
    componentLinearVelScale=<0.08 0.4 0.4>
    overrideWorldGravity {z=-980}
   }
   Save cached pose 'SkirtDynPose'(Pose=Blend Poses by bool(
     BlendPose_0=Component To Local(ComponentPose=@n127)
     BlendPose_1=Blend Poses by bool(BlendPose_0=@n125, BlendPose_1=@n123, BlendTime_1=0.5, bActiveValue=@n126) {blendType=Linear}
     BlendTime_1=0.5
     bActiveValue=Get bIsInFrontEnd()
    ) {blendType=Linear}) {cacheName=SkirtDynPose}
   n128 = Anim Dynamics - Bone: dyn_upperarm_cloth_l!(
    ComponentPose=Local To Component(LocalPose=Use cached pose 'SkirtCachePose'())
    Alpha=0.5
    GravityOverride:Vector="0,0,500"
    ExternalForce="0,0,0"
   ) -> (Pose) {
    angularDampingOverride=0.99
    angularSpringConstant=800
    bAngularSpring=true
    bOverrideAngularDamping=true
    bUseGravityOverride=true
    gravityScale=0.3
    simulationSpace=BoneRelative
    alphaBoolBlend {blendOption=Linear}
    boundBone {boneName=dyn_upperarm_cloth_l}
    componentAppliedLinearAccClamp=<2000 2000 1600>
    componentLinearAccScale=<0.5 0.5 0.5>
    componentLinearVelScale=<0.55 0.55 0.55>
    gravityOverride {z=500}
    physicsBodyDefinitions [{
      collisionType=CustomSphere
      sphereCollisionRadius=2
      boundBone {boneName=dyn_upperarm_cloth_l}
      boxExtents=<20 20 20>
      constraintSetup {angularLimitsMax=<5 5 10>, angularLimitsMin=<-5 -5 -10>, angularTarget=<0 0 0>}
      localJointOffset=<2 0.2 -0.1>
     }]
    relativeSpaceBone {boneName=gravity_upperarm_l}
   }
   n129 = Anim Dynamics - Bone: dyn_upperarm_cloth_bl!(ComponentPose=@n128, Alpha=1, GravityOverride:Vector="0,0,500", ExternalForce="0,0,0") -> (Pose) {
    angularDampingOverride=0.99
    angularSpringConstant=800
    bAngularSpring=true
    bOverrideAngularDamping=true
    bUseGravityOverride=true
    gravityScale=0.3
    simulationSpace=BoneRelative
    alphaBoolBlend {blendOption=Linear}
    boundBone {boneName=dyn_upperarm_cloth_bl}
    componentAppliedLinearAccClamp=<2000 2000 1600>
    componentLinearAccScale=<0.5 0.5 0.5>
    componentLinearVelScale=<0.55 0.55 0.55>
    gravityOverride {z=500}
    physicsBodyDefinitions [{
      collisionType=CustomSphere
      sphereCollisionRadius=2
      boundBone {boneName=dyn_upperarm_cloth_bl}
      boxExtents=<20 20 20>
      constraintSetup {angularLimitsMax=<5 5 3>, angularLimitsMin=<-5 -5 -3>, angularTarget=<0 0 0>}
      localJointOffset=<2 0.2 -0.1>
     }]
    relativeSpaceBone {boneName=gravity_upperarm_l}
   }
   n130 = Anim Dynamics - Bone: dyn_upperarm_cloth_al!(ComponentPose=@n129, Alpha=0.7, GravityOverride:Vector="0,0,500", ExternalForce="0,0,0") -> (Pose) {
    angularDampingOverride=0.99
    angularSpringConstant=800
    bAngularSpring=true
    bOverrideAngularDamping=true
    bUseGravityOverride=true
    gravityScale=0.3
    simulationSpace=BoneRelative
    alphaBoolBlend {blendOption=Linear}
    boundBone {boneName=dyn_upperarm_cloth_al}
    componentAppliedLinearAccClamp=<2000 2000 1600>
    componentLinearAccScale=<0.5 0.5 0.5>
    componentLinearVelScale=<0.55 0.55 0.55>
    gravityOverride {z=500}
    physicsBodyDefinitions [{
      collisionType=CustomSphere
      sphereCollisionRadius=2
      boundBone {boneName=dyn_upperarm_cloth_al}
      boxExtents=<20 20 20>
      constraintSetup {angularLimitsMax=<5 5 2>, angularLimitsMin=<-5 -5 -5>, angularTarget=<0 0 0>}
      localJointOffset=<2 0.2 -0.1>
     }]
    relativeSpaceBone {boneName=gravity_upperarm_l}
   }
   n131 = Anim Dynamics - Bone: dyn_upperarm_croth_r!(ComponentPose=@n130, Alpha=0.5, GravityOverride:Vector="0,0,500", ExternalForce="0,0,0") -> (Pose) {
    angularDampingOverride=0.99
    angularSpringConstant=800
    bAngularSpring=true
    bOverrideAngularDamping=true
    bUseGravityOverride=true
    gravityScale=0.3
    simulationSpace=BoneRelative
    alphaBoolBlend {blendOption=Linear}
    boundBone {boneName=dyn_upperarm_croth_r}
    componentAppliedLinearAccClamp=<2000 2000 1600>
    componentLinearAccScale=<0.5 0.5 0.5>
    componentLinearVelScale=<0.55 0.55 0.55>
    gravityOverride {z=500}
    physicsBodyDefinitions [{
      collisionType=CustomSphere
      sphereCollisionRadius=2
      boundBone {boneName=dyn_upperarm_croth_r}
      boxExtents=<20 20 20>
      constraintSetup {angularLimitsMax=<5 5 10>, angularLimitsMin=<-5 -5 -10>, angularTarget=<0 0 0>}
      localJointOffset=<-2 0.2 -0.1>
     }]
    relativeSpaceBone {boneName=gravity_upperarm_r}
   }
   n132 = Anim Dynamics - Bone: dyn_upperarm_croth_br!(ComponentPose=@n131, Alpha=1, GravityOverride:Vector="0,0,500", ExternalForce="0,0,0") -> (Pose) {
    angularDampingOverride=0.99
    angularSpringConstant=800
    bAngularSpring=true
    bOverrideAngularDamping=true
    bUseGravityOverride=true
    gravityScale=0.3
    simulationSpace=BoneRelative
    alphaBoolBlend {blendOption=Linear}
    boundBone {boneName=dyn_upperarm_croth_br}
    componentAppliedLinearAccClamp=<2000 2000 1600>
    componentLinearAccScale=<0.5 0.5 0.5>
    componentLinearVelScale=<0.55 0.55 0.55>
    gravityOverride {z=500}
    physicsBodyDefinitions [{
      collisionType=CustomSphere
      sphereCollisionRadius=2
      boundBone {boneName=dyn_upperarm_croth_br}
      boxExtents=<20 20 20>
      constraintSetup {twistAxis=AxisY, angularLimitsMax=<5 5 10>, angularLimitsMin=<-5 -5 -10>, angularTarget=<0 0 0>}
      localJointOffset=<-2 -2 -0.1>
     }]
    relativeSpaceBone {boneName=gravity_upperarm_r}
   }
   n133 = Anim Dynamics - Bone: dyn_upperarm_croth_ar!(ComponentPose=@n132, Alpha=0.7, GravityOverride:Vector="0,0,500", ExternalForce="0,0,0") -> (Pose) {
    angularDampingOverride=0.99
    angularSpringConstant=800
    bAngularSpring=true
    bOverrideAngularDamping=true
    bUseGravityOverride=true
    gravityScale=0.3
    simulationSpace=BoneRelative
    alphaBoolBlend {blendOption=Linear}
    boundBone {boneName=dyn_upperarm_croth_ar}
    componentAppliedLinearAccClamp=<2000 2000 1600>
    componentLinearAccScale=<0.5 0.5 0.5>
    componentLinearVelScale=<0.55 0.55 0.55>
    gravityOverride {z=500}
    physicsBodyDefinitions [{
      collisionType=CustomSphere
      sphereCollisionRadius=2
      boundBone {boneName=dyn_upperarm_croth_ar}
      boxExtents=<20 20 20>
      constraintSetup {twistAxis=AxisY, angularLimitsMax=<5 5 5>, angularLimitsMin=<-5 -5 -5>, angularTarget=<0 0 0>}
      localJointOffset=<-2 -2 -0.1>
     }]
    relativeSpaceBone {boneName=gravity_upperarm_r}
   }
   n134 = Anim Dynamics - Bone: dyn_zipper_l_01!(ComponentPose=@n133, Alpha=1, GravityOverride:Vector="0,0,500", ExternalForce="0,0,0") -> (Pose) {
    angularDampingOverride=0.99
    angularSpringConstant=800
    bAngularSpring=true
    bOverrideAngularDamping=true
    bUseGravityOverride=true
    gravityScale=0.3
    simulationSpace=BoneRelative
    alphaBoolBlend {blendOption=Linear}
    boundBone {boneName=dyn_zipper_l_01}
    componentAppliedLinearAccClamp=<2000 2000 1600>
    componentLinearAccScale=<0.45 0.45 0.45>
    componentLinearVelScale=<0.55 0.55 0.55>
    gravityOverride {z=500}
    physicsBodyDefinitions [{
      collisionType=CustomSphere
      sphereCollisionRadius=2
      boundBone {boneName=dyn_zipper_l_01}
      boxExtents=<20 20 20>
      constraintSetup {angularLimitsMax=<0.1 0.1 10>, angularLimitsMin=<-0.1 -0.1 -20>, angularTarget=<0 0 0>}
      localJointOffset=<-5 0.2 1>
     }]
    relativeSpaceBone {boneName=gravity}
   }
   n135 = Anim Dynamics - Bone: dyn_zipper_r_01!(ComponentPose=@n134, Alpha=1, GravityOverride:Vector="0,0,500", ExternalForce="0,0,0") -> (Pose) {
    angularDampingOverride=0.99
    angularSpringConstant=800
    bAngularSpring=true
    bOverrideAngularDamping=true
    bUseGravityOverride=true
    gravityScale=0.3
    simulationSpace=BoneRelative
    alphaBoolBlend {blendOption=Linear}
    boundBone {boneName=dyn_zipper_r_01}
    componentAppliedLinearAccClamp=<2000 2000 1600>
    componentLinearAccScale=<0.45 0.45 0.45>
    componentLinearVelScale=<0.55 0.55 0.55>
    gravityOverride {z=500}
    physicsBodyDefinitions [{
      collisionType=CustomSphere
      sphereCollisionRadius=2
      boundBone {boneName=dyn_zipper_r_01}
      boxExtents=<20 20 20>
      constraintSetup {angularLimitsMax=<0.1 0.1 10>, angularLimitsMin=<-0.1 -0.1 -20>, angularTarget=<0 0 0>}
      localJointOffset=<-5 0.2 1>
     }]
    relativeSpaceBone {boneName=gravity}
   }
   n136 = Anim Dynamics - Bone: dyn_Cloth_volume_l_A!(ComponentPose=@n135, Alpha=1, GravityOverride:Vector="0,0,500", ExternalForce="0,0,0") -> (Pose) {
    angularDampingOverride=0.99
    angularSpringConstant=800
    bAngularSpring=true
    bOverrideAngularDamping=true
    bUseGravityOverride=true
    gravityScale=0.3
    simulationSpace=BoneRelative
    alphaBoolBlend {blendOption=Linear}
    boundBone {boneName=dyn_Cloth_volume_l_A}
    componentAppliedLinearAccClamp=<2000 2000 1600>
    componentLinearAccScale=<0.45 0.45 0.45>
    componentLinearVelScale=<0.55 0.55 0.55>
    gravityOverride {z=500}
    physicsBodyDefinitions [{
      collisionType=CustomSphere
      sphereCollisionRadius=2
      boundBone {boneName=dyn_Cloth_volume_l_A}
      boxExtents=<30 30 30>
      constraintSetup {angularLimitsMax=<0.1 0.1 20>, angularLimitsMin=<-0.1 -0.1 0>, angularTarget=<0 0 0>}
      localJointOffset=<-5 0.2 1>
     }]
    relativeSpaceBone {boneName=gravity}
   }
   n137 = Anim Dynamics - Bone: dyn_Cloth_volume_l_A01!(ComponentPose=@n136, Alpha=1, GravityOverride:Vector="0,0,500", ExternalForce="0,0,0") -> (Pose) {
    angularDampingOverride=0.99
    angularSpringConstant=800
    bAngularSpring=true
    bOverrideAngularDamping=true
    bUseGravityOverride=true
    gravityScale=0.3
    simulationSpace=BoneRelative
    alphaBoolBlend {blendOption=Linear}
    boundBone {boneName=dyn_Cloth_volume_l_A01}
    componentAppliedLinearAccClamp=<2000 2000 1600>
    componentLinearAccScale=<0.45 0.45 0.45>
    componentLinearVelScale=<0.55 0.55 0.55>
    gravityOverride {z=500}
    physicsBodyDefinitions [{
      collisionType=CustomSphere
      sphereCollisionRadius=2
      boundBone {boneName=dyn_Cloth_volume_l_A01}
      boxExtents=<20 20 20>
      constraintSetup {angularLimitsMax=<0.1 0.1 20>, angularLimitsMin=<-0.1 -0.1 0>, angularTarget=<0 0 0>}
      localJointOffset=<-5 0.2 1>
     }]
    relativeSpaceBone {boneName=gravity}
   }
   n138 = Anim Dynamics - Bone: dyn_Cloth_volume_l_A02!(ComponentPose=@n137, Alpha=1, GravityOverride:Vector="0,0,500", ExternalForce="0,0,0") -> (Pose) {
    angularDampingOverride=0.99
    angularSpringConstant=800
    bAngularSpring=true
    bOverrideAngularDamping=true
    bUseGravityOverride=true
    gravityScale=0.3
    simulationSpace=BoneRelative
    alphaBoolBlend {blendOption=Linear}
    boundBone {boneName=dyn_Cloth_volume_l_A02}
    componentAppliedLinearAccClamp=<2000 2000 1600>
    componentLinearAccScale=<0.45 0.45 0.45>
    componentLinearVelScale=<0.55 0.55 0.55>
    gravityOverride {z=500}
    physicsBodyDefinitions [{
      collisionType=CustomSphere
      sphereCollisionRadius=2
      boundBone {boneName=dyn_Cloth_volume_l_A02}
      boxExtents=<20 20 20>
      constraintSetup {angularLimitsMax=<0.1 0.1 20>, angularLimitsMin=<-0.1 -0.1 0>, angularTarget=<0 0 0>}
      localJointOffset=<-5 0.2 1>
     }]
    relativeSpaceBone {boneName=gravity}
   }
   n139 = Anim Dynamics - Bone: dyn_Cloth_volume_l_B!(ComponentPose=@n138, Alpha=1, GravityOverride:Vector="0,0,500", ExternalForce="0,0,0") -> (Pose) {
    angularDampingOverride=0.99
    angularSpringConstant=800
    bAngularSpring=true
    bOverrideAngularDamping=true
    bUseGravityOverride=true
    gravityScale=0.3
    simulationSpace=BoneRelative
    alphaBoolBlend {blendOption=Linear}
    boundBone {boneName=dyn_Cloth_volume_l_B}
    componentAppliedLinearAccClamp=<2000 2000 1600>
    componentLinearAccScale=<0.4 0.4 0.4>
    componentLinearVelScale=<0.45 0.45 0.55>
    gravityOverride {z=500}
    physicsBodyDefinitions [{
      collisionType=CustomSphere
      sphereCollisionRadius=2
      boundBone {boneName=dyn_Cloth_volume_l_B}
      boxExtents=<50 50 50>
      constraintSetup {angularLimitsMax=<10 1 31>, angularLimitsMin=<-10 -1 0>, angularTarget=<0 0 0>}
      localJointOffset=<-5 0.2 -0.1>
     }]
    relativeSpaceBone {boneName=gravity}
   }
   n140 = Anim Dynamics - Bone: dyn_Cloth_volume_l_B01!(ComponentPose=@n139, Alpha=1, GravityOverride:Vector="0,0,500", ExternalForce="0,0,0") -> (Pose) {
    angularDampingOverride=0.99
    angularSpringConstant=800
    bAngularSpring=true
    bOverrideAngularDamping=true
    bUseGravityOverride=true
    gravityScale=0.3
    simulationSpace=BoneRelative
    alphaBoolBlend {blendOption=Linear}
    boundBone {boneName=dyn_Cloth_volume_l_B01}
    componentAppliedLinearAccClamp=<2000 2000 1600>
    componentLinearAccScale=<0.4 0.4 0.4>
    componentLinearVelScale=<0.45 0.45 0.55>
    gravityOverride {z=500}
    physicsBodyDefinitions [{
      collisionType=CustomSphere
      sphereCollisionRadius=2
      boundBone {boneName=dyn_Cloth_volume_l_B01}
      boxExtents=<33 33 33>
      constraintSetup {angularLimitsMax=<10 1 31>, angularLimitsMin=<-10 -1 0>, angularTarget=<0 0 0>}
      localJointOffset=<-5 0.2 -0.1>
     }]
    relativeSpaceBone {boneName=gravity}
   }
   n141 = Anim Dynamics - Bone: dyn_Cloth_volume_l_B02!(ComponentPose=@n140, Alpha=1, GravityOverride:Vector="0,0,500", ExternalForce="0,0,0") -> (Pose) {
    angularDampingOverride=0.99
    angularSpringConstant=800
    bAngularSpring=true
    bOverrideAngularDamping=true
    bUseGravityOverride=true
    gravityScale=0.3
    simulationSpace=BoneRelative
    alphaBoolBlend {blendOption=Linear}
    boundBone {boneName=dyn_Cloth_volume_l_B02}
    componentAppliedLinearAccClamp=<2000 2000 1600>
    componentLinearAccScale=<0.4 0.4 0.4>
    componentLinearVelScale=<0.45 0.45 0.55>
    gravityOverride {z=500}
    physicsBodyDefinitions [{
      collisionType=CustomSphere
      sphereCollisionRadius=2
      boundBone {boneName=dyn_Cloth_volume_l_B02}
      boxExtents=<27 27 27>
      constraintSetup {angularLimitsMax=<10 1 31>, angularLimitsMin=<-10 -1 0>, angularTarget=<0 0 0>}
      localJointOffset=<-5 0.2 -0.1>
     }]
    relativeSpaceBone {boneName=gravity}
   }
   n142 = Anim Dynamics - Bone: dyn_Cloth_volume_l_B03!(ComponentPose=@n141, Alpha=1, GravityOverride:Vector="0,0,500", ExternalForce="0,0,0") -> (Pose) {
    angularDampingOverride=0.99
    angularSpringConstant=800
    bAngularSpring=true
    bOverrideAngularDamping=true
    bUseGravityOverride=true
    gravityScale=0.3
    simulationSpace=BoneRelative
    alphaBoolBlend {blendOption=Linear}
    boundBone {boneName=dyn_Cloth_volume_l_B03}
    componentAppliedLinearAccClamp=<2000 2000 1600>
    componentLinearAccScale=<0.4 0.4 0.4>
    componentLinearVelScale=<0.45 0.45 0.55>
    gravityOverride {z=500}
    physicsBodyDefinitions [{
      collisionType=CustomSphere
      sphereCollisionRadius=2
      boundBone {boneName=dyn_Cloth_volume_l_B03}
      boxExtents=<20 20 20>
      constraintSetup {angularLimitsMax=<10 1 31>, angularLimitsMin=<-10 -1 0>, angularTarget=<0 0 0>}
      localJointOffset=<-5 0.2 -0.1>
     }]
    relativeSpaceBone {boneName=gravity}
   }
   n143 = Anim Dynamics - Bone: dyn_Cloth_volume_l_C!(ComponentPose=@n142, Alpha=1, GravityOverride:Vector="0,0,500", ExternalForce="0,0,0") -> (Pose) {
    angularDampingOverride=0.99
    angularSpringConstant=800
    bAngularSpring=true
    bOverrideAngularDamping=true
    bUseGravityOverride=true
    gravityScale=0.3
    simulationSpace=BoneRelative
    alphaBoolBlend {blendOption=Linear}
    boundBone {boneName=dyn_Cloth_volume_l_C}
    componentAppliedLinearAccClamp=<2000 2000 1600>
    componentLinearAccScale=<0.5 0.5 0.4>
    componentLinearVelScale=<0.55 0.55 0.55>
    gravityOverride {z=500}
    physicsBodyDefinitions [{
      collisionType=CustomSphere
      sphereCollisionRadius=2
      boundBone {boneName=dyn_Cloth_volume_l_C}
      boxExtents=<60 60 60>
      constraintSetup {angularLimitsMax=<5 5 20>, angularLimitsMin=<-1 -2 0>, angularTarget=<0 0 0>}
      localJointOffset=<-5 0.2 -0.1>
     }]
    relativeSpaceBone {boneName=gravity}
   }
   n144 = Anim Dynamics - Bone: dyn_Cloth_volume_l_C(ComponentPose=@n143, Alpha=1, GravityOverride="0,0,500", ExternalForce="0,0,0") {
    angularDampingOverride=0.99
    angularSpringConstant=800
    bAngularSpring=true
    bOverrideAngularDamping=true
    bUseGravityOverride=true
    gravityScale=0.3
    simulationSpace=BoneRelative
    alphaBoolBlend {blendOption=Linear}
    boundBone {boneName=dyn_Cloth_volume_l_C}
    componentAppliedLinearAccClamp=<2000 2000 1600>
    componentLinearAccScale=<0.5 0.5 0.4>
    componentLinearVelScale=<0.55 0.55 0.55>
    gravityOverride {z=500}
    physicsBodyDefinitions [{
      collisionType=CustomSphere
      sphereCollisionRadius=2
      boundBone {boneName=dyn_Cloth_volume_l_C}
      boxExtents=<50 50 50>
      constraintSetup {angularLimitsMax=<5 5 20>, angularLimitsMin=<-1 -2 0>, angularTarget=<0 0 0>}
      localJointOffset=<-5 0.2 -0.1>
     }]
    relativeSpaceBone {boneName=gravity}
   }
   n145 = Anim Dynamics - Bone: dyn_Cloth_volume_l_C02!(ComponentPose=@n144, Alpha=1, GravityOverride:Vector="0,0,500", ExternalForce="0,0,0") -> (Pose) {
    angularDampingOverride=0.99
    angularSpringConstant=800
    bAngularSpring=true
    bOverrideAngularDamping=true
    bUseGravityOverride=true
    gravityScale=0.3
    simulationSpace=BoneRelative
    alphaBoolBlend {blendOption=Linear}
    boundBone {boneName=dyn_Cloth_volume_l_C02}
    componentAppliedLinearAccClamp=<2000 2000 1600>
    componentLinearAccScale=<0.5 0.5 0.4>
    componentLinearVelScale=<0.55 0.55 0.55>
    gravityOverride {z=500}
    physicsBodyDefinitions [{
      collisionType=CustomSphere
      sphereCollisionRadius=2
      boundBone {boneName=dyn_Cloth_volume_l_C02}
      boxExtents=<40 40 40>
      constraintSetup {angularLimitsMax=<5 5 20>, angularLimitsMin=<-1 -2 0>, angularTarget=<0 0 0>}
      localJointOffset=<-5 0.2 -0.1>
     }]
    relativeSpaceBone {boneName=gravity}
   }
   n146 = Anim Dynamics - Bone: dyn_Cloth_volume_l_C03!(ComponentPose=@n145, Alpha=1, GravityOverride:Vector="0,0,500", ExternalForce="0,0,0") -> (Pose) {
    angularDampingOverride=0.99
    angularSpringConstant=800
    bAngularSpring=true
    bOverrideAngularDamping=true
    bUseGravityOverride=true
    gravityScale=0.3
    simulationSpace=BoneRelative
    alphaBoolBlend {blendOption=Linear}
    boundBone {boneName=dyn_Cloth_volume_l_C03}
    componentAppliedLinearAccClamp=<2000 2000 1600>
    componentLinearAccScale=<0.5 0.5 0.4>
    componentLinearVelScale=<0.55 0.55 0.55>
    gravityOverride {z=500}
    physicsBodyDefinitions [{
      collisionType=CustomSphere
      sphereCollisionRadius=2
      boundBone {boneName=dyn_Cloth_volume_l_C03}
      boxExtents=<30 30 30>
      constraintSetup {angularLimitsMax=<5 5 20>, angularLimitsMin=<-1 -2 0>, angularTarget=<0 0 0>}
      localJointOffset=<-5 0.2 -0.1>
     }]
    relativeSpaceBone {boneName=gravity}
   }
   n147 = Anim Dynamics - Bone: dyn_Cloth_volume_l_C04!(ComponentPose=@n146, Alpha=1, GravityOverride:Vector="0,0,500", ExternalForce="0,0,0") -> (Pose) {
    angularDampingOverride=0.99
    angularSpringConstant=800
    bAngularSpring=true
    bOverrideAngularDamping=true
    bUseGravityOverride=true
    bUseSphericalLimits=true
    gravityScale=0.3
    simulationSpace=BoneRelative
    alphaBoolBlend {blendOption=Linear}
    boundBone {boneName=dyn_Cloth_volume_l_C04}
    componentAppliedLinearAccClamp=<2000 2000 1600>
    componentLinearAccScale=<0.5 0.5 0.4>
    componentLinearVelScale=<0.55 0.55 0.55>
    gravityOverride {z=500}
    physicsBodyDefinitions [{
      collisionType=CustomSphere
      sphereCollisionRadius=5
      boundBone {boneName=dyn_Cloth_volume_l_C04}
      boxExtents=<20 20 20>
      constraintSetup {angularLimitsMax=<5 5 20>, angularLimitsMin=<-1 -2 0>, angularTarget=<0 0 0>}
      localJointOffset=<-5 0.2 -0.1>
     }]
    relativeSpaceBone {boneName=gravity}
    sphericalLimits [
     {limitRadius=8, drivingBone {boneName=foot_l}, sphereLocalOffset=<-4 0 0>}
     {limitRadius=8, drivingBone {boneName=foot_r}, sphereLocalOffset=<4 0 0>}
    ]
   }
   n148 = Anim Dynamics - Bone: dyn_Cloth_volume_l_C05!(ComponentPose=@n147, Alpha=1, GravityOverride:Vector="0,0,500", ExternalForce="0,0,0") -> (Pose) {
    angularDampingOverride=0.99
    angularSpringConstant=800
    bAngularSpring=true
    bOverrideAngularDamping=true
    bUseGravityOverride=true
    bUseSphericalLimits=true
    gravityScale=0.3
    simulationSpace=BoneRelative
    alphaBoolBlend {blendOption=Linear}
    boundBone {boneName=dyn_Cloth_volume_l_C05}
    componentAppliedLinearAccClamp=<2000 2000 1600>
    componentLinearAccScale=<0.5 0.5 0.4>
    componentLinearVelScale=<0.55 0.55 0.55>
    gravityOverride {z=500}
    physicsBodyDefinitions [{
      collisionType=CustomSphere
      sphereCollisionRadius=5
      boundBone {boneName=dyn_Cloth_volume_l_C05}
      boxExtents=<10 10 10>
      constraintSetup {angularLimitsMax=<5 5 20>, angularLimitsMin=<-1 -2 0>, angularTarget=<0 0 0>}
      localJointOffset=<-5 0.2 -0.1>
     }]
    relativeSpaceBone {boneName=gravity}
    sphericalLimits [
     {limitRadius=8, drivingBone {boneName=foot_l}, sphereLocalOffset=<-4 0 0>}
     {limitRadius=8, drivingBone {boneName=foot_r}, sphereLocalOffset=<4 0 0>}
    ]
   }
   n149 = Anim Dynamics - Bone: dyn_Cloth_volume_r_A!(ComponentPose=@n148, Alpha=1, GravityOverride:Vector="0,0,-980", ExternalForce="0,0,0") -> (Pose) {
    angularDampingOverride=0.99
    angularSpringConstant=800
    bAngularSpring=true
    bOverrideAngularDamping=true
    bUseGravityOverride=true
    gravityScale=0.3
    simulationSpace=BoneRelative
    alphaBoolBlend {blendOption=Linear}
    boundBone {boneName=dyn_Cloth_volume_r_A}
    componentAppliedLinearAccClamp=<2000 2000 1600>
    componentLinearAccScale=<0.45 0.45 0.45>
    componentLinearVelScale=<0.55 0.55 0.55>
    gravityOverride {z=-980}
    physicsBodyDefinitions [{
      collisionType=CustomSphere
      sphereCollisionRadius=2
      boundBone {boneName=dyn_Cloth_volume_r_A}
      boxExtents=<30 30 30>
      constraintSetup {angularLimitsMax=<0.1 0.1 20>, angularLimitsMin=<-0.1 -0.1 0>, angularTarget=<0 0 0>}
      localJointOffset=<-5 0.2 1>
     }]
    relativeSpaceBone {boneName=gravity}
   }
   n150 = Anim Dynamics - Bone: dyn_Cloth_volume_r_A01!(ComponentPose=@n149, Alpha=1, GravityOverride:Vector="0,0,-980", ExternalForce="0,0,0") -> (Pose) {
    angularDampingOverride=0.99
    angularSpringConstant=800
    bAngularSpring=true
    bOverrideAngularDamping=true
    bUseGravityOverride=true
    gravityScale=0.3
    simulationSpace=BoneRelative
    alphaBoolBlend {blendOption=Linear}
    boundBone {boneName=dyn_Cloth_volume_r_A01}
    componentAppliedLinearAccClamp=<2000 2000 1600>
    componentLinearAccScale=<0.45 0.45 0.45>
    componentLinearVelScale=<0.55 0.55 0.55>
    gravityOverride {z=-980}
    physicsBodyDefinitions [{
      collisionType=CustomSphere
      sphereCollisionRadius=2
      boundBone {boneName=dyn_Cloth_volume_r_A01}
      boxExtents=<20 20 20>
      constraintSetup {angularLimitsMax=<0.1 0.1 20>, angularLimitsMin=<-0.1 -0.1 0>, angularTarget=<0 0 0>}
      localJointOffset=<-5 0.2 1>
     }]
    relativeSpaceBone {boneName=gravity}
   }
   n151 = Anim Dynamics - Bone: dyn_Cloth_volume_r_A02!(ComponentPose=@n150, Alpha=1, GravityOverride:Vector="0,0,-980", ExternalForce="0,0,0") -> (Pose) {
    angularDampingOverride=0.99
    angularSpringConstant=800
    bAngularSpring=true
    bOverrideAngularDamping=true
    bUseGravityOverride=true
    gravityScale=0.3
    simulationSpace=BoneRelative
    alphaBoolBlend {blendOption=Linear}
    boundBone {boneName=dyn_Cloth_volume_r_A02}
    componentAppliedLinearAccClamp=<2000 2000 1600>
    componentLinearAccScale=<0.45 0.45 0.45>
    componentLinearVelScale=<0.55 0.55 0.55>
    gravityOverride {z=-980}
    physicsBodyDefinitions [{
      collisionType=CustomSphere
      sphereCollisionRadius=2
      boundBone {boneName=dyn_Cloth_volume_r_A02}
      boxExtents=<20 20 20>
      constraintSetup {angularLimitsMax=<0.1 0.1 20>, angularLimitsMin=<-0.1 -0.1 0>, angularTarget=<0 0 0>}
      localJointOffset=<-5 0.2 1>
     }]
    relativeSpaceBone {boneName=gravity}
   }
   n152 = Anim Dynamics - Bone: dyn_Cloth_volume_r_B!(ComponentPose=@n151, Alpha=1, GravityOverride:Vector="0,0,-980", ExternalForce="0,0,0") -> (Pose) {
    angularDampingOverride=0.99
    angularSpringConstant=800
    bAngularSpring=true
    bOverrideAngularDamping=true
    bUseGravityOverride=true
    gravityScale=0.3
    simulationSpace=BoneRelative
    alphaBoolBlend {blendOption=Linear}
    boundBone {boneName=dyn_Cloth_volume_r_B}
    componentAppliedLinearAccClamp=<2000 2000 1600>
    componentLinearAccScale=<0.4 0.4 0.4>
    componentLinearVelScale=<0.45 0.45 0.55>
    gravityOverride {z=-980}
    physicsBodyDefinitions [{
      collisionType=CustomSphere
      sphereCollisionRadius=2
      boundBone {boneName=dyn_Cloth_volume_r_B}
      boxExtents=<50 50 50>
      constraintSetup {angularLimitsMax=<5 5 31>, angularLimitsMin=<0 -5 0>, angularTarget=<0 0 0>}
      localJointOffset=<-5 0.2 -0.1>
     }]
    relativeSpaceBone {boneName=gravity}
   }
   n153 = Anim Dynamics - Bone: dyn_Cloth_volume_r_B01!(ComponentPose=@n152, Alpha=1, GravityOverride:Vector="0,0,-980", ExternalForce="0,0,0") -> (Pose) {
    angularDampingOverride=0.99
    angularSpringConstant=800
    bAngularSpring=true
    bOverrideAngularDamping=true
    bUseGravityOverride=true
    gravityScale=0.3
    simulationSpace=BoneRelative
    alphaBoolBlend {blendOption=Linear}
    boundBone {boneName=dyn_Cloth_volume_r_B01}
    componentAppliedLinearAccClamp=<2000 2000 1600>
    componentLinearAccScale=<0.4 0.4 0.4>
    componentLinearVelScale=<0.45 0.45 0.55>
    gravityOverride {z=-980}
    physicsBodyDefinitions [{
      collisionType=CustomSphere
      sphereCollisionRadius=2
      boundBone {boneName=dyn_Cloth_volume_r_B01}
      boxExtents=<33 33 33>
      constraintSetup {angularLimitsMax=<5 5 31>, angularLimitsMin=<0 -5 0>, angularTarget=<0 0 0>}
      localJointOffset=<-5 0.2 -0.1>
     }]
    relativeSpaceBone {boneName=gravity}
   }
   n154 = Anim Dynamics - Bone: dyn_Cloth_volume_r_B02!(ComponentPose=@n153, Alpha=1, GravityOverride:Vector="0,0,-980", ExternalForce="0,0,0") -> (Pose) {
    angularDampingOverride=0.99
    angularSpringConstant=800
    bAngularSpring=true
    bOverrideAngularDamping=true
    bUseGravityOverride=true
    gravityScale=0.3
    simulationSpace=BoneRelative
    alphaBoolBlend {blendOption=Linear}
    boundBone {boneName=dyn_Cloth_volume_r_B02}
    componentAppliedLinearAccClamp=<2000 2000 1600>
    componentLinearAccScale=<0.4 0.4 0.4>
    componentLinearVelScale=<0.45 0.45 0.55>
    gravityOverride {z=-980}
    physicsBodyDefinitions [{
      collisionType=CustomSphere
      sphereCollisionRadius=2
      boundBone {boneName=dyn_Cloth_volume_r_B02}
      boxExtents=<27 27 27>
      constraintSetup {angularLimitsMax=<5 5 31>, angularLimitsMin=<0 -5 0>, angularTarget=<0 0 0>}
      localJointOffset=<-5 0.2 -0.1>
     }]
    relativeSpaceBone {boneName=gravity}
   }
   n155 = Anim Dynamics - Bone: dyn_Cloth_volume_r_B03!(ComponentPose=@n154, Alpha=1, GravityOverride:Vector="0,0,-980", ExternalForce="0,0,0") -> (Pose) {
    angularDampingOverride=0.99
    angularSpringConstant=800
    bAngularSpring=true
    bOverrideAngularDamping=true
    bUseGravityOverride=true
    gravityScale=0.3
    simulationSpace=BoneRelative
    alphaBoolBlend {blendOption=Linear}
    boundBone {boneName=dyn_Cloth_volume_r_B03}
    componentAppliedLinearAccClamp=<2000 2000 1600>
    componentLinearAccScale=<0.4 0.4 0.4>
    componentLinearVelScale=<0.45 0.45 0.55>
    gravityOverride {z=-980}
    physicsBodyDefinitions [{
      collisionType=CustomSphere
      sphereCollisionRadius=2
      boundBone {boneName=dyn_Cloth_volume_r_B03}
      boxExtents=<20 20 20>
      constraintSetup {angularLimitsMax=<5 5 31>, angularLimitsMin=<0 -5 0>, angularTarget=<0 0 0>}
      localJointOffset=<-5 0.2 -0.1>
     }]
    relativeSpaceBone {boneName=gravity}
   }
   n156 = Anim Dynamics - Bone: dyn_Cloth_volume_r_C!(ComponentPose=@n155, Alpha=1, GravityOverride:Vector="0,0,-980", ExternalForce="0,0,0") -> (Pose) {
    angularDampingOverride=0.99
    angularSpringConstant=800
    bAngularSpring=true
    bOverrideAngularDamping=true
    bUseGravityOverride=true
    gravityScale=0.3
    simulationSpace=BoneRelative
    alphaBoolBlend {blendOption=Linear}
    boundBone {boneName=dyn_Cloth_volume_r_C}
    componentAppliedLinearAccClamp=<2000 2000 1600>
    componentLinearAccScale=<0.5 0.5 0.4>
    componentLinearVelScale=<0.55 0.55 0.55>
    gravityOverride {z=-980}
    physicsBodyDefinitions [{
      collisionType=CustomSphere
      sphereCollisionRadius=2
      boundBone {boneName=dyn_Cloth_volume_r_C}
      boxExtents=<45 45 45>
      constraintSetup {angularLimitsMax=<1 2 20>, angularLimitsMin=<-5 -5 0>, angularTarget=<0 0 0>}
      localJointOffset=<-5 0.2 -0.1>
     }]
    relativeSpaceBone {boneName=gravity}
   }
   n157 = Anim Dynamics - Bone: dyn_Cloth_volume_r_C01!(ComponentPose=@n156, Alpha=1, GravityOverride:Vector="0,0,-980", ExternalForce="0,0,0") -> (Pose) {
    angularDampingOverride=0.99
    angularSpringConstant=800
    bAngularSpring=true
    bOverrideAngularDamping=true
    bUseGravityOverride=true
    gravityScale=0.3
    simulationSpace=BoneRelative
    alphaBoolBlend {blendOption=Linear}
    boundBone {boneName=dyn_Cloth_volume_r_C01}
    componentAppliedLinearAccClamp=<2000 2000 1600>
    componentLinearAccScale=<0.5 0.5 0.4>
    componentLinearVelScale=<0.55 0.55 0.55>
    gravityOverride {z=-980}
    physicsBodyDefinitions [{
      collisionType=CustomSphere
      sphereCollisionRadius=2
      boundBone {boneName=dyn_Cloth_volume_r_C01}
      boxExtents=<40 40 40>
      constraintSetup {angularLimitsMax=<1 2 20>, angularLimitsMin=<-5 -5 0>, angularTarget=<0 0 0>}
      localJointOffset=<-5 0.2 -0.1>
     }]
    relativeSpaceBone {boneName=gravity}
   }
   n158 = Anim Dynamics - Bone: dyn_Cloth_volume_r_C02!(ComponentPose=@n157, Alpha=1, GravityOverride:Vector="0,0,-980", ExternalForce="0,0,0") -> (Pose) {
    angularDampingOverride=0.99
    angularSpringConstant=800
    bAngularSpring=true
    bOverrideAngularDamping=true
    bUseGravityOverride=true
    gravityScale=0.3
    simulationSpace=BoneRelative
    alphaBoolBlend {blendOption=Linear}
    boundBone {boneName=dyn_Cloth_volume_r_C02}
    componentAppliedLinearAccClamp=<2000 2000 1600>
    componentLinearAccScale=<0.5 0.5 0.4>
    componentLinearVelScale=<0.55 0.55 0.55>
    gravityOverride {z=-980}
    physicsBodyDefinitions [{
      collisionType=CustomSphere
      sphereCollisionRadius=2
      boundBone {boneName=dyn_Cloth_volume_r_C02}
      boxExtents=<33 33 33>
      constraintSetup {angularLimitsMax=<1 2 20>, angularLimitsMin=<-5 -5 0>, angularTarget=<0 0 0>}
      localJointOffset=<-5 0.2 -0.1>
     }]
    relativeSpaceBone {boneName=gravity}
   }
   n159 = Anim Dynamics - Bone: dyn_Cloth_volume_r_C03!(ComponentPose=@n158, Alpha=1, GravityOverride:Vector="0,0,-980", ExternalForce="0,0,0") -> (Pose) {
    angularDampingOverride=0.99
    angularSpringConstant=800
    bAngularSpring=true
    bOverrideAngularDamping=true
    bUseGravityOverride=true
    gravityScale=0.3
    simulationSpace=BoneRelative
    alphaBoolBlend {blendOption=Linear}
    boundBone {boneName=dyn_Cloth_volume_r_C03}
    componentAppliedLinearAccClamp=<2000 2000 1600>
    componentLinearAccScale=<0.5 0.5 0.4>
    componentLinearVelScale=<0.55 0.55 0.55>
    gravityOverride {z=-980}
    physicsBodyDefinitions [{
      collisionType=CustomSphere
      sphereCollisionRadius=2
      boundBone {boneName=dyn_Cloth_volume_r_C03}
      boxExtents=<27 27 27>
      constraintSetup {angularLimitsMax=<1 2 20>, angularLimitsMin=<-5 -5 0>, angularTarget=<0 0 0>}
      localJointOffset=<-5 0.2 -0.1>
     }]
    relativeSpaceBone {boneName=gravity}
   }
   n160 = Anim Dynamics - Bone: dyn_Cloth_volume_r_C04!(ComponentPose=@n159, Alpha=1, GravityOverride:Vector="0,0,-980", ExternalForce="0,0,0") -> (Pose) {
    angularDampingOverride=0.99
    angularSpringConstant=800
    bAngularSpring=true
    bOverrideAngularDamping=true
    bUseGravityOverride=true
    bUseSphericalLimits=true
    gravityScale=0.3
    simulationSpace=BoneRelative
    alphaBoolBlend {blendOption=Linear}
    boundBone {boneName=dyn_Cloth_volume_r_C04}
    componentAppliedLinearAccClamp=<2000 2000 1600>
    componentLinearAccScale=<0.5 0.5 0.4>
    componentLinearVelScale=<0.55 0.55 0.55>
    gravityOverride {z=-980}
    physicsBodyDefinitions [{
      collisionType=CustomSphere
      sphereCollisionRadius=5
      boundBone {boneName=dyn_Cloth_volume_r_C04}
      boxExtents=<20 20 20>
      constraintSetup {angularLimitsMax=<1 2 20>, angularLimitsMin=<-5 -5 0>, angularTarget=<0 0 0>}
      localJointOffset=<-5 0.2 -0.1>
     }]
    relativeSpaceBone {boneName=gravity}
    sphericalLimits [
     {limitRadius=8, drivingBone {boneName=foot_l}, sphereLocalOffset=<-4 0 0>}
     {limitRadius=8, drivingBone {boneName=foot_r}, sphereLocalOffset=<4 0 0>}
    ]
   }
   n161 = Anim Dynamics - Bone: dyn_Cloth_volume_r_C05!(ComponentPose=@n160, Alpha=1, GravityOverride:Vector="0,0,-980", ExternalForce="0,0,0") -> (Pose) {
    angularDampingOverride=0.99
    angularSpringConstant=800
    bAngularSpring=true
    bOverrideAngularDamping=true
    bUseGravityOverride=true
    bUseSphericalLimits=true
    gravityScale=0.3
    simulationSpace=BoneRelative
    alphaBoolBlend {blendOption=Linear}
    boundBone {boneName=dyn_Cloth_volume_r_C05}
    componentAppliedLinearAccClamp=<2000 2000 1600>
    componentLinearAccScale=<0.5 0.5 0.4>
    componentLinearVelScale=<0.55 0.55 0.55>
    gravityOverride {z=-980}
    physicsBodyDefinitions [{
      collisionType=CustomSphere
      sphereCollisionRadius=5
      boundBone {boneName=dyn_Cloth_volume_r_C05}
      boxExtents=<10 10 10>
      constraintSetup {angularLimitsMax=<1 2 20>, angularLimitsMin=<-5 -5 0>, angularTarget=<0 0 0>}
      localJointOffset=<-5 0.2 -0.1>
     }]
    relativeSpaceBone {boneName=gravity}
    sphericalLimits [
     {limitRadius=8, drivingBone {boneName=foot_l}, sphereLocalOffset=<-4 0 0>}
     {limitRadius=8, drivingBone {boneName=foot_r}, sphereLocalOffset=<4 0 0>}
    ]
   }
   n162 = "Constraint - dyn_Cloth_l_01 (PP)"!(ComponentPose=@n161, Alpha=1, ConstraintWeights_0=0, ConstraintWeights_1:float=1) -> (Pose) {
    alphaBoolBlend {blendOption=Linear}
    boneToModify {boneName=dyn_Cloth_l_01}
    constraintSetup [{targetBone {boneName=dyn_Cloth_volume_l_C02}}, {targetBone {boneName=dyn_Cloth_volume_l_C01}}]
    constraintWeights [1,1]
   }
   n163 = "Constraint - dyn_Cloth_r_01 (PP)"!(ComponentPose=@n162, Alpha=1, ConstraintWeights_0=0, ConstraintWeights_1:float=1) -> (Pose) {
    alphaBoolBlend {blendOption=Linear}
    boneToModify {boneName=dyn_Cloth_r_01}
    constraintSetup [{targetBone {boneName=dyn_Cloth_volume_r_C02}}, {targetBone {boneName=dyn_Cloth_volume_r_C01}}]
    constraintWeights [1,1]
   }
   n164 = "Constraint - dyn_Cloth_l_02 (PP)"(ComponentPose=@n163, Alpha=1, ConstraintWeights_0=0, ConstraintWeights_1=1) {
    alphaBoolBlend {blendOption=Linear}
    boneToModify {boneName=dyn_Cloth_l_02}
    constraintSetup [{targetBone {boneName=dyn_Cloth_volume_l_C02}}, {targetBone {boneName=dyn_Cloth_volume_l_C03}}]
    constraintWeights [1,1]
   }
   n165 = "Constraint - dyn_Cloth_l_03 (PP)"(ComponentPose=@n164, Alpha=1, ConstraintWeights_0=0, ConstraintWeights_1=1) {
    alphaBoolBlend {blendOption=Linear}
    boneToModify {boneName=dyn_Cloth_l_03}
    constraintSetup [{targetBone {boneName=dyn_Cloth_volume_l_C02}}, {targetBone {boneName=dyn_Cloth_volume_l_C04}}]
    constraintWeights [1,1]
   }
   n166 = "Constraint - dyn_Cloth_l_04 (PP)"(ComponentPose=@n165, Alpha=1, ConstraintWeights_0=0, ConstraintWeights_1=1) {
    alphaBoolBlend {blendOption=Linear}
    boneToModify {boneName=dyn_Cloth_l_04}
    constraintSetup [{targetBone {boneName=dyn_Cloth_volume_l_C02}}, {targetBone {boneName=dyn_Cloth_volume_l_C05}}]
    constraintWeights [1,1]
   }
   n167 = "Constraint - dyn_Cloth_r_02 (PP)"(ComponentPose=@n166, Alpha=1, ConstraintWeights_0=0, ConstraintWeights_1=1) {
    alphaBoolBlend {blendOption=Linear}
    boneToModify {boneName=dyn_Cloth_r_02}
    constraintSetup [{targetBone {boneName=dyn_Cloth_volume_r_C02}}, {targetBone {boneName=dyn_Cloth_volume_r_C03}}]
    constraintWeights [1,1]
   }
   n168 = "Constraint - dyn_Cloth_r_03 (PP)"(ComponentPose=@n167, Alpha=1, ConstraintWeights_0=0, ConstraintWeights_1=1) {
    alphaBoolBlend {blendOption=Linear}
    boneToModify {boneName=dyn_Cloth_r_03}
    constraintSetup [{targetBone {boneName=dyn_Cloth_volume_r_C02}}, {targetBone {boneName=dyn_Cloth_volume_r_C04}}]
    constraintWeights [1,1]
   }
   n169 = "Constraint - dyn_Cloth_r_04 (PP)"(ComponentPose=@n168, Alpha=1, ConstraintWeights_0=0, ConstraintWeights_1=1) {
    alphaBoolBlend {blendOption=Linear}
    boneToModify {boneName=dyn_Cloth_r_04}
    constraintSetup [{targetBone {boneName=dyn_Cloth_volume_r_C02}}, {targetBone {boneName=dyn_Cloth_volume_r_C05}}]
    constraintWeights [1,1]
   }
   n170 = "Constraint - dyn_Cloth_volume_bk (PP)"(ComponentPose=@n169, Alpha=1, ConstraintWeights_0=0.5, ConstraintWeights_1=0.5) {
    alphaBoolBlend {blendOption=Linear}
    boneToModify {boneName=dyn_Cloth_volume_bk}
    constraintSetup [{targetBone {boneName=dyn_Cloth_volume_l_C}}, {targetBone {boneName=dyn_Cloth_volume_r_C}}]
    constraintWeights [1,1]
   }
   n171 = "Constraint - dyn_Cloth_volume_bk01 (PP)"(ComponentPose=@n170, Alpha=1, ConstraintWeights_0=0.5, ConstraintWeights_1=0.5) {
    alphaBoolBlend {blendOption=Linear}
    boneToModify {boneName=dyn_Cloth_volume_bk01}
    constraintSetup [{targetBone {boneName=dyn_Cloth_volume_l_C01}}, {targetBone {boneName=dyn_Cloth_volume_r_C01}}]
    constraintWeights [1,1]
   }
   n172 = "Constraint - dyn_Cloth_volume_bk02 (PP)"(ComponentPose=@n171, Alpha=1, ConstraintWeights_0=0.5, ConstraintWeights_1=0.5) {
    alphaBoolBlend {blendOption=Linear}
    boneToModify {boneName=dyn_Cloth_volume_bk02}
    constraintSetup [{targetBone {boneName=dyn_Cloth_volume_l_C02}}, {targetBone {boneName=dyn_Cloth_volume_r_C02}}]
    constraintWeights [1,1]
   }
   n173 = "Constraint - dyn_Cloth_volume_bk03 (PP)"(ComponentPose=@n172, Alpha=1, ConstraintWeights_0=0.5, ConstraintWeights_1=0.5) {
    alphaBoolBlend {blendOption=Linear}
    boneToModify {boneName=dyn_Cloth_volume_bk03}
    constraintSetup [{targetBone {boneName=dyn_Cloth_volume_l_C03}}, {targetBone {boneName=dyn_Cloth_volume_r_C03}}]
    constraintWeights [1,1]
   }
   n174 = "Constraint - dyn_Cloth_volume_bk04 (PP)"(ComponentPose=@n173, Alpha=1, ConstraintWeights_0=0.5, ConstraintWeights_1=0.5) {
    alphaBoolBlend {blendOption=Linear}
    boneToModify {boneName=dyn_Cloth_volume_bk04}
    constraintSetup [{targetBone {boneName=dyn_Cloth_volume_l_C04}}, {targetBone {boneName=dyn_Cloth_volume_r_C04}}]
    constraintWeights [1,1]
   }
   n175 = "Constraint - dyn_Cloth_collision_l_B01 (PPPPPP)"!(
    ComponentPose=@n174
    Alpha=1
    ConstraintWeights_0=0.3396
    ConstraintWeights_1:float=0.293
    ConstraintWeights_2:float=0.1842
    ConstraintWeights_3:float=0.1395
    ConstraintWeights_4:float=0.0408
    ConstraintWeights_5:float=0.0029
   ) -> (Pose) {
    alphaBoolBlend {blendOption=Linear}
    boneToModify {boneName=dyn_Cloth_collision_l_B01}
    constraintSetup [
     {targetBone {boneName=dyn_Cloth_volume_l_B02}}
     {targetBone {boneName=dyn_Cloth_volume_l_C02}}
     {targetBone {boneName=dyn_Cloth_volume_l_B03}}
     {targetBone {boneName=dyn_Cloth_volume_l_C03}}
     {targetBone {boneName=dyn_Cloth_volume_l_C01}}
     {targetBone {boneName=dyn_Cloth_volume_l_B01}}
    ]
    constraintWeights [1,1,1,1,1,1]
   }
   n176 = "Constraint - dyn_Cloth_collision_l_B02 (PPPPPP)"!(
    ComponentPose=@n175
    Alpha=1
    ConstraintWeights_0=0.5054
    ConstraintWeights_1:float=0.2607
    ConstraintWeights_2:float=0.199
    ConstraintWeights_3:float=0.0195
    ConstraintWeights_4:float=0.0142
    ConstraintWeights_5:float=0.0012
   ) -> (Pose) {
    alphaBoolBlend {blendOption=Linear}
    boneToModify {boneName=dyn_Cloth_collision_l_B02}
    constraintSetup [
     {targetBone {boneName=dyn_Cloth_volume_l_B03}}
     {targetBone {boneName=dyn_Cloth_volume_l_C03}}
     {targetBone {boneName=dyn_Cloth_volume_l_C04}}
     {targetBone {boneName=dyn_Cloth_volume_l_C05}}
     {targetBone {boneName=dyn_Cloth_volume_l_B02}}
     {targetBone {boneName=dyn_Cloth_volume_l_C02}}
    ]
    constraintWeights [1,1,1,1,1,1]
   }
   n177 = "Constraint - dyn_Cloth_collision_l_B03 (PPPPPP)"!(
    ComponentPose=@n176
    Alpha=1
    ConstraintWeights_0=0.4098
    ConstraintWeights_1:float=0.3353
    ConstraintWeights_2:float=0.2133
    ConstraintWeights_3:float=0.0301
    ConstraintWeights_4:float=0.0105
    ConstraintWeights_5:float=0.0011
   ) -> (Pose) {
    alphaBoolBlend {blendOption=Linear}
    boneToModify {boneName=dyn_Cloth_collision_l_B03}
    constraintSetup [
     {targetBone {boneName=dyn_Cloth_volume_l_B03}}
     {targetBone {boneName=dyn_Cloth_volume_l_A02}}
     {targetBone {boneName=dyn_Cloth_volume_l_B02}}
     {targetBone {boneName=dyn_Cloth_volume_l_C05}}
     {targetBone {boneName=dyn_Cloth_volume_l_C03}}
     {targetBone {boneName=dyn_Cloth_volume_l_C04}}
    ]
    constraintWeights [1,1,1,1,1,1]
   }
   n178 = "Constraint - dyn_Cloth_collision_r_B01 (PPPPPPPP)"!(
    ComponentPose=@n177
    Alpha=1
    ConstraintWeights_0=0.3502
    ConstraintWeights_1:float=0.2783
    ConstraintWeights_2:float=0.1896
    ConstraintWeights_3:float=0.1252
    ConstraintWeights_4:float=0.0375
    ConstraintWeights_5:float=0.0123
    ConstraintWeights_6:float=0.0067
    ConstraintWeights_7:float=0.0003
   ) -> (Pose) {
    alphaBoolBlend {blendOption=Linear}
    boneToModify {boneName=dyn_Cloth_collision_r_B01}
    constraintSetup [
     {targetBone {boneName=dyn_Cloth_volume_r_B02}}
     {targetBone {boneName=dyn_Cloth_volume_r_C02}}
     {targetBone {boneName=dyn_Cloth_volume_r_B03}}
     {targetBone {boneName=dyn_Cloth_volume_r_C03}}
     {targetBone {boneName=dyn_Cloth_volume_r_C01}}
     {targetBone {boneName=dyn_Cloth_volume_r_A02}}
     {targetBone {boneName=dyn_Cloth_volume_r_B01}}
     {targetBone {boneName=dyn_Cloth_volume_r_A01}}
    ]
    constraintWeights [1,1,1,1,1,1,1,1]
   }
   n179 = "Constraint - dyn_Cloth_collision_r_B02 (PPPPPPP)"!(
    ComponentPose=@n178
    Alpha=1
    ConstraintWeights_0=0.5117
    ConstraintWeights_1:float=0.2555
    ConstraintWeights_2:float=0.1934
    ConstraintWeights_3:float=0.0194
    ConstraintWeights_4:float=0.0147
    ConstraintWeights_5:float=0.0039
    ConstraintWeights_6:float=0.0013
   ) -> (Pose) {
    alphaBoolBlend {blendOption=Linear}
    boneToModify {boneName=dyn_Cloth_collision_r_B02}
    constraintSetup [
     {targetBone {boneName=dyn_Cloth_volume_r_B03}}
     {targetBone {boneName=dyn_Cloth_volume_r_C03}}
     {targetBone {boneName=dyn_Cloth_volume_r_C04}}
     {targetBone {boneName=dyn_Cloth_volume_r_C05}}
     {targetBone {boneName=dyn_Cloth_volume_r_B02}}
     {targetBone {boneName=dyn_Cloth_volume_r_A02}}
     {targetBone {boneName=dyn_Cloth_volume_r_C02}}
    ]
    constraintWeights [1,1,1,1,1,1,1]
   }
   n180 = "Constraint - dyn_Cloth_collision_r_B03 (PPPPPP)"!(
    ComponentPose=@n179
    Alpha=1
    ConstraintWeights_0=0.403
    ConstraintWeights_1:float=0.3454
    ConstraintWeights_2:float=0.212
    ConstraintWeights_3:float=0.0288
    ConstraintWeights_4:float=0.0098
    ConstraintWeights_5:float=0.001
   ) -> (Pose) {
    alphaBoolBlend {blendOption=Linear}
    boneToModify {boneName=dyn_Cloth_collision_r_B03}
    constraintSetup [
     {targetBone {boneName=dyn_Cloth_volume_r_B03}}
     {targetBone {boneName=dyn_Cloth_volume_r_A02}}
     {targetBone {boneName=dyn_Cloth_volume_r_B02}}
     {targetBone {boneName=dyn_Cloth_volume_r_C05}}
     {targetBone {boneName=dyn_Cloth_volume_r_C03}}
     {targetBone {boneName=dyn_Cloth_volume_r_C04}}
    ]
    constraintWeights [1,1,1,1,1,1]
   }
   n181 = "Constraint - dyn_Cloth_collision_l_C01 (PPPPPP)"!(
    ComponentPose=@n180
    Alpha=1
    ConstraintWeights_0=0.568
    ConstraintWeights_1:float=0.2713
    ConstraintWeights_2:float=0.0808
    ConstraintWeights_3:float=0.0756
    ConstraintWeights_4:float=0.0034
    ConstraintWeights_5:float=0.001
   ) -> (Pose) {
    alphaBoolBlend {blendOption=Linear}
    boneToModify {boneName=dyn_Cloth_collision_l_C01}
    constraintSetup [
     {targetBone {boneName=dyn_Cloth_volume_l_C02}}
     {targetBone {boneName=dyn_Cloth_volume_l_C03}}
     {targetBone {boneName=dyn_Cloth_volume_bk02}}
     {targetBone {boneName=dyn_Cloth_volume_bk03}}
     {targetBone {boneName=dyn_Cloth_volume_bk04}}
     {targetBone {boneName=dyn_Cloth_volume_bk01}}
    ]
    constraintWeights [1,1,1,1,1,1]
   }
   n182 = "Constraint - dyn_Cloth_collision_l_C02 (PPPPPP)"!(
    ComponentPose=@n181
    Alpha=1
    ConstraintWeights_0=0.7038
    ConstraintWeights_1:float=0.0808
    ConstraintWeights_2:float=0.0796
    ConstraintWeights_3:float=0.0793
    ConstraintWeights_4:float=0.0548
    ConstraintWeights_5:float=0.0017
   ) -> (Pose) {
    alphaBoolBlend {blendOption=Linear}
    boneToModify {boneName=dyn_Cloth_collision_l_C02}
    constraintSetup [
     {targetBone {boneName=dyn_Cloth_volume_l_C03}}
     {targetBone {boneName=dyn_Cloth_volume_l_C02}}
     {targetBone {boneName=dyn_Cloth_volume_bk04}}
     {targetBone {boneName=dyn_Cloth_volume_l_C04}}
     {targetBone {boneName=dyn_Cloth_volume_bk03}}
     {targetBone {boneName=dyn_Cloth_volume_l_B03}}
    ]
    constraintWeights [1,1,1,1,1,1]
   }
   n183 = "Constraint - dyn_Cloth_collision_l_C03 (PPPP)"!(
    ComponentPose=@n182
    Alpha=1
    ConstraintWeights_0=0.5057
    ConstraintWeights_1:float=0.2944
    ConstraintWeights_2:float=0.1138
    ConstraintWeights_3:float=0.0861
   ) -> (Pose) {
    alphaBoolBlend {blendOption=Linear}
    boneToModify {boneName=dyn_Cloth_collision_l_C03}
    constraintSetup [
     {targetBone {boneName=dyn_Cloth_volume_l_C04}}
     {targetBone {boneName=dyn_Cloth_volume_l_C05}}
     {targetBone {boneName=dyn_Cloth_volume_l_C03}}
     {targetBone {boneName=dyn_Cloth_volume_bk04}}
    ]
    constraintWeights [1,1,1,1]
   }
   n184 = "Constraint - dyn_Cloth_collision_r_C01 (PPPPPP)"!(
    ComponentPose=@n183
    Alpha=1
    ConstraintWeights_0=0.5667
    ConstraintWeights_1:float=0.2718
    ConstraintWeights_2:float=0.0809
    ConstraintWeights_3:float=0.076
    ConstraintWeights_4:float=0.0036
    ConstraintWeights_5:float=0.001
   ) -> (Pose) {
    alphaBoolBlend {blendOption=Linear}
    boneToModify {boneName=dyn_Cloth_collision_r_C01}
    constraintSetup [
     {targetBone {boneName=dyn_Cloth_volume_r_C02}}
     {targetBone {boneName=dyn_Cloth_volume_r_C03}}
     {targetBone {boneName=dyn_Cloth_volume_bk02}}
     {targetBone {boneName=dyn_Cloth_volume_bk03}}
     {targetBone {boneName=dyn_Cloth_volume_bk04}}
     {targetBone {boneName=dyn_Cloth_volume_bk01}}
    ]
    constraintWeights [1,1,1,1,1,1]
   }
   n185 = "Constraint - dyn_Cloth_collision_r_C02 (PPPPPP)"!(
    ComponentPose=@n184
    Alpha=1
    ConstraintWeights_0=0.6983
    ConstraintWeights_1:float=0.082
    ConstraintWeights_2:float=0.0787
    ConstraintWeights_3:float=0.0754
    ConstraintWeights_4:float=0.0527
    ConstraintWeights_5:float=0.0129
   ) -> (Pose) {
    alphaBoolBlend {blendOption=Linear}
    boneToModify {boneName=dyn_Cloth_collision_r_C02}
    constraintSetup [
     {targetBone {boneName=dyn_Cloth_volume_r_C03}}
     {targetBone {boneName=dyn_Cloth_volume_r_C02}}
     {targetBone {boneName=dyn_Cloth_volume_r_C04}}
     {targetBone {boneName=dyn_Cloth_volume_bk04}}
     {targetBone {boneName=dyn_Cloth_volume_bk03}}
     {targetBone {boneName=dyn_Cloth_volume_r_B03}}
    ]
    constraintWeights [1,1,1,1,1,1]
   }
   n186 = "Constraint - dyn_Cloth_collision_r_C03 (PPPPP)"!(
    ComponentPose=@n185
    Alpha=1
    ConstraintWeights_0=0.505
    ConstraintWeights_1:float=0.293
    ConstraintWeights_2:float=0.1151
    ConstraintWeights_3:float=0.086
    ConstraintWeights_4:float=0.0009
   ) -> (Pose) {
    alphaBoolBlend {blendOption=Linear}
    boneToModify {boneName=dyn_Cloth_collision_r_C03}
    constraintSetup [
     {targetBone {boneName=dyn_Cloth_volume_r_C04}}
     {targetBone {boneName=dyn_Cloth_volume_r_C05}}
     {targetBone {boneName=dyn_Cloth_volume_r_C03}}
     {targetBone {boneName=dyn_Cloth_volume_bk04}}
     {targetBone {boneName=dyn_Cloth_volume_r_B03}}
    ]
    constraintWeights [1,1,1,1,1]
   }
   n187 = "Constraint - dyn_Cloth_collision_bk01 (PPPPPPP)"!(
    ComponentPose=@n186
    Alpha=1
    ConstraintWeights_0=0.5897
    ConstraintWeights_1:float=0.2214
    ConstraintWeights_2:float=0.116
    ConstraintWeights_3:float=0.0182
    ConstraintWeights_4:float=0.0182
    ConstraintWeights_5:float=0.0182
    ConstraintWeights_6:float=0.0182
   ) -> (Pose) {
    alphaBoolBlend {blendOption=Linear}
    boneToModify {boneName=dyn_Cloth_collision_bk01}
    constraintSetup [
     {targetBone {boneName=dyn_Cloth_volume_bk02}}
     {targetBone {boneName=dyn_Cloth_volume_bk03}}
     {targetBone {boneName=dyn_Cloth_volume_r_C02}}
     {targetBone {boneName=dyn_Cloth_volume_l_C04}}
     {targetBone {boneName=dyn_Cloth_volume_l_C03}}
     {targetBone {boneName=dyn_Cloth_volume_l_C02}}
     {targetBone {boneName=dyn_Cloth_volume_l_C01}}
    ]
    constraintWeights [1,1,1,1,1,1,1]
   }
   n188 = "Constraint - dyn_Cloth_collision_bk02 (PPPP)"!(
    ComponentPose=@n187
    Alpha=1
    ConstraintWeights_0=0.6713
    ConstraintWeights_1:float=0.2094
    ConstraintWeights_2:float=0.0619
    ConstraintWeights_3:float=0.0574
   ) -> (Pose) {
    alphaBoolBlend {blendOption=Linear}
    boneToModify {boneName=dyn_Cloth_collision_bk02}
    constraintSetup [
     {targetBone {boneName=dyn_Cloth_volume_bk04}}
     {targetBone {boneName=dyn_Cloth_volume_bk03}}
     {targetBone {boneName=dyn_Cloth_volume_l_C03}}
     {targetBone {boneName=dyn_Cloth_volume_r_C03}}
    ]
    constraintWeights [1,1,1,1]
   }
   Save cached pose 'MovingPose'(Pose=Component To Local(ComponentPose=@n188)) {cacheName=MovingPose}
   n189 = "Constraint - dyn_Cloth_volume_r_C02 (RR)"!(
    ComponentPose=Local To Component(LocalPose=Use cached pose 'SkirtCachePose'())
    Alpha=1
    ConstraintWeights_0=0.4
    ConstraintWeights_1:float=0.6
   ) -> (Pose) {
    alphaBoolBlend {blendOption=Linear}
    boneToModify {boneName=dyn_Cloth_volume_r_C02}
    constraintSetup [{transformType=Rotation, targetBone {boneName=calf_r}}, {transformType=Rotation, targetBone {boneName=dyn_Cloth_volume_r_C01}}]
    constraintWeights [1,1]
   }
   n190 = "Constraint - dyn_Cloth_volume_l_C02 (RR)"!(ComponentPose=@n189, Alpha=1, ConstraintWeights_0=0.4, ConstraintWeights_1:float=0.6) -> (Pose) {
    alphaBoolBlend {blendOption=Linear}
    boneToModify {boneName=dyn_Cloth_volume_l_C02}
    constraintSetup [{transformType=Rotation, targetBone {boneName=calf_l}}, {transformType=Rotation, targetBone {boneName=dyn_Cloth_volume_l_C01}}]
    constraintWeights [1,1]
   }
   n191 = "Constraint - dyn_Cloth_volume_r_C03 (RR)"!(ComponentPose=@n190, Alpha=1, ConstraintWeights_0=0.4, ConstraintWeights_1:float=0.6) -> (Pose) {
    alphaBoolBlend {blendOption=Linear}
    boneToModify {boneName=dyn_Cloth_volume_r_C03}
    constraintSetup [{transformType=Rotation, targetBone {boneName=calf_r}}, {transformType=Rotation, targetBone {boneName=dyn_Cloth_volume_r_C02}}]
    constraintWeights [1,1]
   }
   n192 = "Constraint - dyn_Cloth_volume_l_C03 (RR)"!(ComponentPose=@n191, Alpha=1, ConstraintWeights_0=0.4, ConstraintWeights_1:float=0.6) -> (Pose) {
    alphaBoolBlend {blendOption=Linear}
    boneToModify {boneName=dyn_Cloth_volume_l_C03}
    constraintSetup [{transformType=Rotation, targetBone {boneName=calf_l}}, {transformType=Rotation, targetBone {boneName=dyn_Cloth_volume_l_C02}}]
    constraintWeights [1,1]
   }
   n193 = "Constraint - dyn_Cloth_volume_bk03 (RRR)"!(ComponentPose=@n192, Alpha=1, ConstraintWeights_0=0.15, ConstraintWeights_1:float=0.15, ConstraintWeights_2:float=0.7) -> (Pose) {
    alphaBoolBlend {blendOption=Linear}
    boneToModify {boneName=dyn_Cloth_volume_bk03}
    constraintSetup [
     {transformType=Rotation, targetBone {boneName=calf_l}}
     {transformType=Rotation, targetBone {boneName=calf_r}}
     {transformType=Rotation, targetBone {boneName=dyn_Cloth_volume_bk01}}
    ]
    constraintWeights [1,1,1]
   }
   n194 = "Constraint - dyn_Cloth_r_03 (PP)"(ComponentPose=@n193, Alpha=1, ConstraintWeights_0=0.8, ConstraintWeights_1=0.2) {
    alphaBoolBlend {blendOption=Linear}
    boneToModify {boneName=dyn_Cloth_r_03}
    constraintSetup [{targetBone {boneName=dyn_Cloth_volume_r_C03}}, {targetBone {boneName=dyn_Cloth_r_03}}]
    constraintWeights [1,1]
   }
   n195 = "Constraint - dyn_Cloth_l_03 (PP)"(ComponentPose=@n194, Alpha=1, ConstraintWeights_0=0.8, ConstraintWeights_1=0.2) {
    alphaBoolBlend {blendOption=Linear}
    boneToModify {boneName=dyn_Cloth_l_03}
    constraintSetup [{targetBone {boneName=dyn_Cloth_volume_l_C03}}, {targetBone {boneName=dyn_Cloth_l_03}}]
    constraintWeights [1,1]
   }
   n196 = "Constraint - dyn_Cloth_r_04 (PP)"(ComponentPose=@n195, Alpha=1, ConstraintWeights_0=0.8, ConstraintWeights_1=0.2) {
    alphaBoolBlend {blendOption=Linear}
    boneToModify {boneName=dyn_Cloth_r_04}
    constraintSetup [{targetBone {boneName=dyn_Cloth_volume_r_C05}}, {targetBone {boneName=dyn_Cloth_r_03}}]
    constraintWeights [1,1]
   }
   n197 = "Constraint - dyn_Cloth_l_04 (PP)"(ComponentPose=@n196, Alpha=1, ConstraintWeights_0=0.8, ConstraintWeights_1=0.2) {
    alphaBoolBlend {blendOption=Linear}
    boneToModify {boneName=dyn_Cloth_l_04}
    constraintSetup [{targetBone {boneName=dyn_Cloth_volume_l_C05}}, {targetBone {boneName=dyn_Cloth_l_03}}]
    constraintWeights [1,1]
   }
   n198 = Component To Local(ComponentPose=@n197)
   n199 = "Constraint - dyn_Cloth_volume_r_C03 (RR)"(ComponentPose=Local To Component(LocalPose=Use cached pose 'SkirtCachePose'()), Alpha=1, ConstraintWeights_0=0.6, ConstraintWeights_1=0.4) {
    alphaBoolBlend {blendOption=Linear}
    boneToModify {boneName=dyn_Cloth_volume_r_C03}
    constraintSetup [{transformType=Rotation, targetBone {boneName=calf_r}}, {transformType=Rotation, targetBone {boneName=dyn_Cloth_volume_r_C02}}]
    constraintWeights [1,1]
   }
   n200 = "Constraint - dyn_Cloth_volume_l_C03 (RR)"(ComponentPose=@n199, Alpha=1, ConstraintWeights_0=0.6, ConstraintWeights_1=0.4) {
    alphaBoolBlend {blendOption=Linear}
    boneToModify {boneName=dyn_Cloth_volume_l_C03}
    constraintSetup [{transformType=Rotation, targetBone {boneName=calf_l}}, {transformType=Rotation, targetBone {boneName=dyn_Cloth_volume_l_C02}}]
    constraintWeights [1,1]
   }
   n201 = Component To Local(ComponentPose=@n200)
   n202 = Get Not DiveFloatingAndDiveMode() -> (Not DiveFloatingAndDiveMode:bool)
   n203 = "Constraint - dyn_Cloth_volume_bk (PP)"(
    ComponentPose=Local To Component(LocalPose=Blend Poses by bool(BlendPose_0=@n201, BlendPose_1=@n198, BlendTime_0=0.5, bActiveValue=@n202) {blendType=Linear})
    Alpha=1
    ConstraintWeights_0=0.5
    ConstraintWeights_1=0.5
   ) {
    alphaBoolBlend {blendOption=Linear}
    boneToModify {boneName=dyn_Cloth_volume_bk}
    constraintSetup [{targetBone {boneName=dyn_Cloth_volume_l_C}}, {targetBone {boneName=dyn_Cloth_volume_r_C}}]
    constraintWeights [1,1]
   }
   n204 = "Constraint - dyn_Cloth_volume_bk01 (PP)"(ComponentPose=@n203, Alpha=1, ConstraintWeights_0=0.5, ConstraintWeights_1=0.5) {
    alphaBoolBlend {blendOption=Linear}
    boneToModify {boneName=dyn_Cloth_volume_bk01}
    constraintSetup [{targetBone {boneName=dyn_Cloth_volume_l_C01}}, {targetBone {boneName=dyn_Cloth_volume_r_C01}}]
    constraintWeights [1,1]
   }
   n205 = "Constraint - dyn_Cloth_volume_bk02 (PP)"(ComponentPose=@n204, Alpha=1, ConstraintWeights_0=0.5, ConstraintWeights_1=0.5) {
    alphaBoolBlend {blendOption=Linear}
    boneToModify {boneName=dyn_Cloth_volume_bk02}
    constraintSetup [{targetBone {boneName=dyn_Cloth_volume_l_C02}}, {targetBone {boneName=dyn_Cloth_volume_r_C02}}]
    constraintWeights [1,1]
   }
   n206 = "Constraint - dyn_Cloth_volume_bk03 (PP)"(ComponentPose=@n205, Alpha=1, ConstraintWeights_0=0.5, ConstraintWeights_1=0.5) {
    alphaBoolBlend {blendOption=Linear}
    boneToModify {boneName=dyn_Cloth_volume_bk03}
    constraintSetup [{targetBone {boneName=dyn_Cloth_volume_l_C03}}, {targetBone {boneName=dyn_Cloth_volume_r_C03}}]
    constraintWeights [1,1]
   }
   n207 = "Constraint - dyn_Cloth_volume_bk04 (PP)"(ComponentPose=@n206, Alpha=1, ConstraintWeights_0=0.5, ConstraintWeights_1=0.5) {
    alphaBoolBlend {blendOption=Linear}
    boneToModify {boneName=dyn_Cloth_volume_bk04}
    constraintSetup [{targetBone {boneName=dyn_Cloth_volume_l_C04}}, {targetBone {boneName=dyn_Cloth_volume_r_C04}}]
    constraintWeights [1,1]
   }
   n208 = "Constraint - dyn_Cloth_collision_l_B01 (PPPPPPPPPPP)"!(
    ComponentPose=@n207
    Alpha=1
    ConstraintWeights_0=0.3262
    ConstraintWeights_1:float=0.323
    ConstraintWeights_2:float=0.1804
    ConstraintWeights_3:float=0.1354
    ConstraintWeights_4:float=0.0235
    ConstraintWeights_5:float=0.007
    ConstraintWeights_6:float=0.003
    ConstraintWeights_7:float=0.0006
    ConstraintWeights_8:float=0.0005
    ConstraintWeights_9:float=0.0003
    ConstraintWeights_10:float=0.0002
   ) -> (Pose) {
    alphaBoolBlend {blendOption=Linear}
    boneToModify {boneName=dyn_Cloth_collision_l_B01}
    constraintSetup [
     {targetBone {boneName=dyn_Cloth_volume_l_C02}}
     {targetBone {boneName=dyn_Cloth_volume_l_B02}}
     {targetBone {boneName=dyn_Cloth_volume_l_B03}}
     {targetBone {boneName=dyn_Cloth_volume_l_C03}}
     {targetBone {boneName=dyn_Cloth_volume_l_A02}}
     {targetBone {boneName=dyn_Cloth_volume_l_B01}}
     {targetBone {boneName=dyn_Cloth_volume_l_C01}}
     {targetBone {boneName=dyn_Cloth_volume_bk03}}
     {targetBone {boneName=dyn_Cloth_volume_bk02}}
     {targetBone {boneName=dyn_Cloth_volume_bk04}}
     {targetBone {boneName=dyn_Cloth_volume_l_A01}}
    ]
    constraintWeights [1,1,1,1,1,1,1,1,1,1,1]
   }
   n209 = "Constraint - dyn_Cloth_collision_l_B02 (PPPPPPPP)"!(
    ComponentPose=@n208
    Alpha=1
    ConstraintWeights_0=0.5353
    ConstraintWeights_1:float=0.2362
    ConstraintWeights_2:float=0.1698
    ConstraintWeights_3:float=0.0346
    ConstraintWeights_4:float=0.0106
    ConstraintWeights_5:float=0.0102
    ConstraintWeights_6:float=0.0028
    ConstraintWeights_7:float=0.0005
   ) -> (Pose) {
    alphaBoolBlend {blendOption=Linear}
    boneToModify {boneName=dyn_Cloth_collision_l_B02}
    constraintSetup [
     {targetBone {boneName=dyn_Cloth_volume_l_B03}}
     {targetBone {boneName=dyn_Cloth_volume_l_C03}}
     {targetBone {boneName=dyn_Cloth_volume_l_C04}}
     {targetBone {boneName=dyn_Cloth_volume_l_B02}}
     {targetBone {boneName=dyn_Cloth_volume_l_A02}}
     {targetBone {boneName=dyn_Cloth_volume_bk04}}
     {targetBone {boneName=dyn_Cloth_volume_l_C05}}
     {targetBone {boneName=dyn_Cloth_volume_l_C02}}
    ]
    constraintWeights [1,1,1,1,1,1,1,1]
   }
   n210 = "Constraint - dyn_Cloth_collision_l_B03 (PPPPP)"!(
    ComponentPose=@n209
    Alpha=1
    ConstraintWeights_0=0.4157
    ConstraintWeights_1:float=0.3457
    ConstraintWeights_2:float=0.224
    ConstraintWeights_3:float=0.009
    ConstraintWeights_4:float=0.0055
   ) -> (Pose) {
    alphaBoolBlend {blendOption=Linear}
    boneToModify {boneName=dyn_Cloth_collision_l_B03}
    constraintSetup [
     {targetBone {boneName=dyn_Cloth_volume_l_B03}}
     {targetBone {boneName=dyn_Cloth_volume_l_A02}}
     {targetBone {boneName=dyn_Cloth_volume_l_B02}}
     {targetBone {boneName=dyn_Cloth_volume_l_C02}}
     {targetBone {boneName=dyn_Cloth_volume_l_C03}}
    ]
    constraintWeights [1,1,1,1,1]
   }
   n211 = "Constraint - dyn_Cloth_collision_r_B01 (PPPPPPPPPPP)"!(
    ComponentPose=@n210
    Alpha=1
    ConstraintWeights_0=0.3364
    ConstraintWeights_1:float=0.3063
    ConstraintWeights_2:float=0.1819
    ConstraintWeights_3:float=0.1235
    ConstraintWeights_4:float=0.0353
    ConstraintWeights_5:float=0.0105
    ConstraintWeights_6:float=0.0046
    ConstraintWeights_7:float=0.0005
    ConstraintWeights_8:float=0.0004
    ConstraintWeights_9:float=0.0003
    ConstraintWeights_10:float=0.0002
   ) -> (Pose) {
    alphaBoolBlend {blendOption=Linear}
    boneToModify {boneName=dyn_Cloth_collision_r_B01}
    constraintSetup [
     {targetBone {boneName=dyn_Cloth_volume_r_B02}}
     {targetBone {boneName=dyn_Cloth_volume_r_C02}}
     {targetBone {boneName=dyn_Cloth_volume_r_B03}}
     {targetBone {boneName=dyn_Cloth_volume_r_C03}}
     {targetBone {boneName=dyn_Cloth_volume_r_A02}}
     {targetBone {boneName=dyn_Cloth_volume_r_B01}}
     {targetBone {boneName=dyn_Cloth_volume_r_C01}}
     {targetBone {boneName=dyn_Cloth_volume_bk03}}
     {targetBone {boneName=dyn_Cloth_volume_bk02}}
     {targetBone {boneName=dyn_Cloth_volume_r_A01}}
     {targetBone {boneName=dyn_Cloth_volume_bk04}}
    ]
    constraintWeights [1,1,1,1,1,1,1,1,1,1,1]
   }
   n212 = "Constraint - dyn_Cloth_collision_r_B02 (PPPPPPPP)"!(
    ComponentPose=@n211
    Alpha=1
    ConstraintWeights_0=0.542
    ConstraintWeights_1:float=0.2309
    ConstraintWeights_2:float=0.1657
    ConstraintWeights_3:float=0.0338
    ConstraintWeights_4:float=0.0142
    ConstraintWeights_5:float=0.01
    ConstraintWeights_6:float=0.0027
    ConstraintWeights_7:float=0.0007
   ) -> (Pose) {
    alphaBoolBlend {blendOption=Linear}
    boneToModify {boneName=dyn_Cloth_collision_r_B02}
    constraintSetup [
     {targetBone {boneName=dyn_Cloth_volume_r_B03}}
     {targetBone {boneName=dyn_Cloth_volume_r_C03}}
     {targetBone {boneName=dyn_Cloth_volume_r_C04}}
     {targetBone {boneName=dyn_Cloth_volume_r_B02}}
     {targetBone {boneName=dyn_Cloth_volume_r_A02}}
     {targetBone {boneName=dyn_Cloth_volume_bk04}}
     {targetBone {boneName=dyn_Cloth_volume_r_C05}}
     {targetBone {boneName=dyn_Cloth_volume_r_C02}}
    ]
    constraintWeights [1,1,1,1,1,1,1,1]
   }
   n213 = "Constraint - dyn_Cloth_collision_r_B03 (PPPPP)"!(
    ComponentPose=@n212
    Alpha=1
    ConstraintWeights_0=0.4088
    ConstraintWeights_1:float=0.3551
    ConstraintWeights_2:float=0.2224
    ConstraintWeights_3:float=0.0085
    ConstraintWeights_4:float=0.0052
   ) -> (Pose) {
    alphaBoolBlend {blendOption=Linear}
    boneToModify {boneName=dyn_Cloth_collision_r_B03}
    constraintSetup [
     {targetBone {boneName=dyn_Cloth_volume_r_B03}}
     {targetBone {boneName=dyn_Cloth_volume_r_A02}}
     {targetBone {boneName=dyn_Cloth_volume_r_B02}}
     {targetBone {boneName=dyn_Cloth_volume_r_C02}}
     {targetBone {boneName=dyn_Cloth_volume_r_C03}}
    ]
    constraintWeights [1,1,1,1,1]
   }
   n214 = "Constraint - dyn_Cloth_collision_l_C01 (PPPPPPPP)"!(
    ComponentPose=@n213
    Alpha=1
    ConstraintWeights_0=0.4717
    ConstraintWeights_1:float=0.2256
    ConstraintWeights_2:float=0.1169
    ConstraintWeights_3:float=0.096
    ConstraintWeights_4:float=0.0489
    ConstraintWeights_5:float=0.0382
    ConstraintWeights_6:float=0.0015
    ConstraintWeights_7:float=0.0011
   ) -> (Pose) {
    alphaBoolBlend {blendOption=Linear}
    boneToModify {boneName=dyn_Cloth_collision_l_C01}
    constraintSetup [
     {targetBone {boneName=dyn_Cloth_volume_l_C02}}
     {targetBone {boneName=dyn_Cloth_volume_l_C03}}
     {targetBone {boneName=dyn_Cloth_volume_bk03}}
     {targetBone {boneName=dyn_Cloth_volume_bk02}}
     {targetBone {boneName=dyn_Cloth_volume_bk04}}
     {targetBone {boneName=dyn_Cloth_volume_l_C01}}
     {targetBone {boneName=dyn_Cloth_volume_l_B03}}
     {targetBone {boneName=dyn_Cloth_volume_l_B02}}
    ]
    constraintWeights [1,1,1,1,1,1,1,1]
   }
   n215 = "Constraint - dyn_Cloth_collision_l_C02 (PPPPPPP)"!(
    ComponentPose=@n214
    Alpha=1
    ConstraintWeights_0=0.4949
    ConstraintWeights_1:float=0.2679
    ConstraintWeights_2:float=0.0617
    ConstraintWeights_3:float=0.0594
    ConstraintWeights_4:float=0.0581
    ConstraintWeights_5:float=0.0497
    ConstraintWeights_6:float=0.0083
   ) -> (Pose) {
    alphaBoolBlend {blendOption=Linear}
    boneToModify {boneName=dyn_Cloth_collision_l_C02}
    constraintSetup [
     {targetBone {boneName=dyn_Cloth_volume_l_C03}}
     {targetBone {boneName=dyn_Cloth_volume_bk04}}
     {targetBone {boneName=dyn_Cloth_volume_bk03}}
     {targetBone {boneName=dyn_Cloth_volume_l_C02}}
     {targetBone {boneName=dyn_Cloth_volume_l_C04}}
     {targetBone {boneName=dyn_Cloth_volume_l_B03}}
     {targetBone {boneName=dyn_Cloth_volume_l_B02}}
    ]
    constraintWeights [1,1,1,1,1,1,1]
   }
   n216 = "Constraint - dyn_Cloth_collision_l_C03 (PPPPP)"!(
    ComponentPose=@n215
    Alpha=1
    ConstraintWeights_0=0.3433
    ConstraintWeights_1:float=0.2882
    ConstraintWeights_2:float=0.2697
    ConstraintWeights_3:float=0.0786
    ConstraintWeights_4:float=0.0201
   ) -> (Pose) {
    alphaBoolBlend {blendOption=Linear}
    boneToModify {boneName=dyn_Cloth_collision_l_C03}
    constraintSetup [
     {targetBone {boneName=dyn_Cloth_volume_l_C04}}
     {targetBone {boneName=dyn_Cloth_volume_l_C05}}
     {targetBone {boneName=dyn_Cloth_volume_bk04}}
     {targetBone {boneName=dyn_Cloth_volume_l_C03}}
     {targetBone {boneName=dyn_Cloth_volume_l_B03}}
    ]
    constraintWeights [1,1,1,1,1]
   }
   n217 = "Constraint - dyn_Cloth_collision_r_C01 (PPPPPPPP)"!(
    ComponentPose=@n216
    Alpha=1
    ConstraintWeights_0=0.4706
    ConstraintWeights_1:float=0.226
    ConstraintWeights_2:float=0.1174
    ConstraintWeights_3:float=0.0961
    ConstraintWeights_4:float=0.0492
    ConstraintWeights_5:float=0.0381
    ConstraintWeights_6:float=0.0015
    ConstraintWeights_7:float=0.0012
   ) -> (Pose) {
    alphaBoolBlend {blendOption=Linear}
    boneToModify {boneName=dyn_Cloth_collision_r_C01}
    constraintSetup [
     {targetBone {boneName=dyn_Cloth_volume_r_C02}}
     {targetBone {boneName=dyn_Cloth_volume_r_C03}}
     {targetBone {boneName=dyn_Cloth_volume_bk03}}
     {targetBone {boneName=dyn_Cloth_volume_bk02}}
     {targetBone {boneName=dyn_Cloth_volume_bk04}}
     {targetBone {boneName=dyn_Cloth_volume_r_C01}}
     {targetBone {boneName=dyn_Cloth_volume_r_B03}}
     {targetBone {boneName=dyn_Cloth_volume_r_B02}}
    ]
    constraintWeights [1,1,1,1,1,1,1,1]
   }
   n218 = "Constraint - dyn_Cloth_collision_r_C02 (PPPPPPPP)"!(
    ComponentPose=@n217
    Alpha=1
    ConstraintWeights_0=0.4944
    ConstraintWeights_1:float=0.2573
    ConstraintWeights_2:float=0.0608
    ConstraintWeights_3:float=0.0594
    ConstraintWeights_4:float=0.0587
    ConstraintWeights_5:float=0.0578
    ConstraintWeights_6:float=0.0114
    ConstraintWeights_7:float=0.0002
   ) -> (Pose) {
    alphaBoolBlend {blendOption=Linear}
    boneToModify {boneName=dyn_Cloth_collision_r_C02}
    constraintSetup [
     {targetBone {boneName=dyn_Cloth_volume_r_C03}}
     {targetBone {boneName=dyn_Cloth_volume_bk04}}
     {targetBone {boneName=dyn_Cloth_volume_r_C02}}
     {targetBone {boneName=dyn_Cloth_volume_r_B03}}
     {targetBone {boneName=dyn_Cloth_volume_bk03}}
     {targetBone {boneName=dyn_Cloth_volume_r_C04}}
     {targetBone {boneName=dyn_Cloth_volume_r_B02}}
     {targetBone {boneName=dyn_Cloth_volume_r_A02}}
    ]
    constraintWeights [1,1,1,1,1,1,1,1]
   }
   n219 = "Constraint - dyn_Cloth_collision_r_C03 (PPPPP)"(
    ComponentPose=@n218
    Alpha=1
    ConstraintWeights_0=0.3432
    ConstraintWeights_1=0.2869
    ConstraintWeights_2=0.2694
    ConstraintWeights_3=0.0795
    ConstraintWeights_4=0.021
   ) {
    alphaBoolBlend {blendOption=Linear}
    boneToModify {boneName=dyn_Cloth_collision_r_C03}
    constraintSetup [
     {targetBone {boneName=dyn_Cloth_volume_r_C04}}
     {targetBone {boneName=dyn_Cloth_volume_r_C05}}
     {targetBone {boneName=dyn_Cloth_volume_bk04}}
     {targetBone {boneName=dyn_Cloth_volume_r_C03}}
     {targetBone {boneName=dyn_Cloth_volume_r_B03}}
    ]
    constraintWeights [1,1,1,1,1]
   }
   n220 = "Constraint - dyn_Cloth_collision_bk01 (PPPPPPPPP)"!(
    ComponentPose=@n219
    Alpha=1
    ConstraintWeights_0=0.6247
    ConstraintWeights_1:float=0.2431
    ConstraintWeights_2:float=0.1171
    ConstraintWeights_3:float=0.0088
    ConstraintWeights_4:float=0.0022
    ConstraintWeights_5:float=0.0021
    ConstraintWeights_6:float=0.0009
    ConstraintWeights_7:float=0.0009
    ConstraintWeights_8:float=0.0002
   ) -> (Pose) {
    alphaBoolBlend {blendOption=Linear}
    boneToModify {boneName=dyn_Cloth_collision_bk01}
    constraintSetup [
     {targetBone {boneName=dyn_Cloth_volume_bk02}}
     {targetBone {boneName=dyn_Cloth_volume_bk03}}
     {targetBone {boneName=dyn_Cloth_volume_l_C02}}
     {targetBone {boneName=dyn_Cloth_volume_r_C02}}
     {targetBone {boneName=dyn_Cloth_volume_l_C01}}
     {targetBone {boneName=dyn_Cloth_volume_r_C01}}
     {targetBone {boneName=dyn_Cloth_volume_l_C03}}
     {targetBone {boneName=dyn_Cloth_volume_r_C03}}
     {targetBone {boneName=dyn_Cloth_volume_bk01}}
    ]
    constraintWeights [1,1,1,1,1,1,1,1,1]
   }
   n221 = "Constraint - dyn_Cloth_collision_bk02 (PPPP)"(ComponentPose=@n220, Alpha=1, ConstraintWeights_0=0.6713, ConstraintWeights_1=0.2094, ConstraintWeights_2=0.0619, ConstraintWeights_3=0.0574) {
    alphaBoolBlend {blendOption=Linear}
    boneToModify {boneName=dyn_Cloth_collision_bk02}
    constraintSetup [
     {targetBone {boneName=dyn_Cloth_volume_bk04}}
     {targetBone {boneName=dyn_Cloth_volume_bk03}}
     {targetBone {boneName=dyn_Cloth_volume_r_C03}}
     {targetBone {boneName=dyn_Cloth_volume_l_C03}}
    ]
    constraintWeights [1,1,1,1]
   }
   n222 = RigidBody(ComponentPose=@n221, bAlphaBoolEnabled=Get DivingAndCape() -> (DivingAndCape:bool), OverrideWorldGravity="0,0,-980") {
    alphaInputType=Bool
    bClampLinearTranslationLimitToRefPose=true
    bOverrideWorldGravity=true
    overridePhysicsAsset=&M_MED_DriftTrooper_Physics
    simulationSpace=BaseBoneSpace
    alphaBoolBlend {blendOption=Linear}
    baseBoneRef {boneName=gravity}
    componentAppliedLinearAccClamp=<3000 3000 2700>
    componentLinearAccScale=<0.01 0.4 0.35>
    componentLinearVelScale=<0.08 0.4 0.4>
    overrideWorldGravity {z=-980}
   }
   Save cached pose 'DBNOPose'(Pose=Component To Local(ComponentPose=@n222)) {cacheName=DBNOPose}
  ]}
 EventGraph {nodes [
   n1 = Event BlueprintUpdateAnimation()
   > n2 = Sequence!(execute) -> (then_0, then_1, then_2:Exec, then_3:Exec, then_4:Exec, then_5:Exec, then_6:Exec, then_7:Exec)
   n3 = Get Body Mesh() -> (Body Mesh:Skeletal Mesh Component Object Reference)
   > Set calf_shape_back_r(calf_shape_back_r:double=GetAlphaFromReferenceTransformDelta(SkeletalMeshComponent=@n3, BoneName=thigh_twist_01_r, BaseName=pelvis, InputRange=(X=-20,Y=-120), TwistAxis="0,0,1")) -> (Output_Get:double)
   > Set calf_shape_back_l(calf_shape_back_l:double=GetAlphaFromReferenceTransformDelta(SkeletalMeshComponent=@n3, BoneName=thigh_l, BaseName=pelvis, InputRange=(X=-20,Y=-150), TwistAxis="0,0,1")) -> (Output_Get:double)
   n4 = Get Body Mesh()
   > Set lowerarm_rot_up_l(lowerarm_rot_up_l:double=GetAlphaFromReferenceTransformDelta(SkeletalMeshComponent=@n4, BoneName=lowerarm_l, BaseName=upperarm_l, InputRange=(X=-20,Y=70), TwistAxis="0,0,1")) -> (Output_Get:double)
   > Set lowerarm_rot_up_r(lowerarm_rot_up_r:double=GetAlphaFromReferenceTransformDelta(SkeletalMeshComponent=@n4, BoneName=lowerarm_r, BaseName=upperarm_r, InputRange=(X=-20,Y=70), TwistAxis="0,0,1")) -> (Output_Get:double)
   > Set upperarm_rot_dn_l(upperarm_rot_dn_l:double=GetAlphaFromReferenceTransformDelta(SkeletalMeshComponent=@n4, BoneName=spine_05, BaseName=upperarm_l, InputRange=(X=0,Y=90), TwistAxis="0,0,1")) -> (Output_Get:double)
   > Set upperarm_rot_dn_r(upperarm_rot_dn_r:double=GetAlphaFromReferenceTransformDelta(SkeletalMeshComponent=@n4, BoneName=spine_05, BaseName=upperarm_r, InputRange=(X=0,Y=90), TwistAxis="0,0,1")) -> (Output_Get:double)
   > Set Shoulder_rot_f_r(Shoulder_rot_f_r:double=GetAlphaFromReferenceTransformDelta(SkeletalMeshComponent=@n4, BoneName=upperarm_r, BaseName=clavicle_r, InputRange=(X=0,Y=110), TwistAxis="0,0,1")) -> (Output_Get:double)
   > Set Shoulder_rot_f_l(Shoulder_rot_f_l:double=GetAlphaFromReferenceTransformDelta(SkeletalMeshComponent=@n4, BoneName=upperarm_l, BaseName=clavicle_l, InputRange=(X=0,Y=110), TwistAxis="0,0,1")) -> (Output_Get:double)
   > Set Head_rot_f(Head_rot_f:double=GetAlphaFromReferenceTransformDelta(SkeletalMeshComponent=@n4, BoneName=head, BaseName=neck_02, InputRange=(X=0,Y=30), TwistAxis="0,0,1")) -> (Output_Get:double)
   n5 = Get Character Mesh() -> (Character Mesh:Skeletal Mesh Component Object Reference)
   Set thigh_shape_fn_l(
    execute=@n2.then_2
    thigh_shape_fn_l:double=GetAlphaFromReferenceTransformDelta(SkeletalMeshComponent=@n5, BoneName=thigh_l, BaseName=pelvis, InputRange=(X=0,Y=-110), TwistAxis="0,0,1")
   ) -> (Output_Get:double)
   > Set thigh_shape_fn_r(thigh_shape_fn_r:double=GetAlphaFromReferenceTransformDelta(SkeletalMeshComponent=@n5, BoneName=thigh_r, BaseName=pelvis, InputRange=(X=0,Y=-110), TwistAxis="0,0,1")) -> (Output_Get:double)
   > Set thigh_shape_fn_r_1(thigh_shape_fn_r_1:double=GetAlphaFromReferenceTransformDelta(SkeletalMeshComponent=@n5, BoneName=calf_r, BaseName=thigh_r, InputRange=(X=0,Y=150), TwistAxis="0,0,1")) -> (Output_Get:double)
   > Set thigh_shape(thigh_shape:bool="float < float"(A=Get thigh_shape_fn_r_1() -> (thigh_shape_fn_r_1:double), B=0.8)) -> (Output_Get:bool)
   > Set thigh_shape_fn_r_0(thigh_shape_fn_r_0:double=float - float(A=1.5, B=Get thigh_shape_fn_r())) -> (Output_Get:double)
   n6 = Cast To FortPlayerAnimInstance(execute=@n2.then_3, Object=GetAnimInstance(self=GetOwnerSkeletalMeshComponent()))
   > Set offset_weapon(offset_weapon:double=0) -> (Output_Get:double)
   > n7 = Switch on EFortWeaponCoreAnimation!(execute, Selection:EFortWeaponCoreAnimation Enum=Get WeaponCoreAnim(self=@n6)) -> (
    Melee:Exec
    Pistol:Exec
    Shotgun:Exec
    PaperBlueprint:Exec
    Rifle:Exec
    MeleeOneHand:Exec
    MachinePistol:Exec
    RocketLauncher:Exec
    GrenadeLauncher:Exec
    GoingCommando:Exec
    AssaultRifle:Exec
    TacticalShotgun:Exec
    SniperRifle:Exec
    TrapPlacement:Exec
    ShoulderLauncher:Exec
    AbilityDecoTool:Exec
    Crossbow:Exec
    C4:Exec
    RemoteControl:Exec
    DualWield:Exec
    AR_BullPup:Exec
    AR_ForwardGrip:Exec
    MedPackPaddles:Exec
    SMG_P90:Exec
    AR_DrumGun:Exec
    Consumable_Small:Exec
    Consumable_Large:Exec
    Balloon:Exec
    MountedTurret:Exec
    CreativeTool:Exec
    ExplosiveBow:Exec
    AshtonIndigo:Exec
    AshtonChicago:Exec
    MeleeDualWield:Exec
    Unarmed:Exec
   )
   n8 = Branch(
    execute=[@n7.AR_BullPup, @n7.AssaultRifle, @n7.Pistol, @n7.Rifle, @n7.Shotgun, @n7.SniperRifle, @n7.TacticalShotgun]
    Condition=OR Boolean(A=Get bIsPlayingEmote(), B=Get bIsSkyDiving())
   )
   > Set offset_weapon(offset_weapon=0)
   Set offset_weapon(
    execute=@n8.else
    offset_weapon=GetAlphaFromReferenceTransformDelta(SkeletalMeshComponent=Get Character Mesh(), BoneName=spine_05, BaseName=upperarm_r, InputRange=(X=0,Y=90), TwistAxis="0,0,1")
   )
   Set is sliding(execute=@n2.then_5, is sliding:double="To Float (Boolean)"(InBool=Get bIsSliding())) -> (Output_Get:double)
   n9 = Get bIsSkydiveFloating()
   > Set Not DiveFloatingAndDiveMode(Not DiveFloatingAndDiveMode:bool=AND Boolean(A=NOT Boolean(A=@n9), B=Get bIsSkydiveDiveMode())) -> (Output_Get:bool)
   n10 = Get thigh_shape() -> (thigh_shape:bool)
   n11 = Get bIsMoving2D()
   > Set CrouchingAndNotMovingAndNotShape(CrouchingAndNotMovingAndNotShape:bool=AND Boolean!(A=Get bIsCrouching(), B=NOT Boolean(A=@n11), C:bool=NOT Boolean(A=@n10)) -> (ReturnValue)) -> (Output_Get:bool)
   n12 = Get bIsPlayingEmote()
   n13 = Get bIsSkyDiving()
   n14 = Get bIsCrouching()
   > Set MovingAnd NotCrouchNotDivingNotEmote(MovingAnd NotCrouchNotDivingNotEmote:bool=AND Boolean!(A=Get bIsMoving2D(), B=NOT Boolean(A=@n14), C=NOT Boolean(A=@n13), D:bool=NOT Boolean(A=@n12)) -> (ReturnValue)) -> (Output_Get:bool)
   > Set DivingAndCape(DivingAndCape:bool=AND Boolean!(A=Get bIsSkyDiving(), B=Get is_WearingCape()) -> (ReturnValue)) -> (Output_Get:bool)
   > Set DivingOrDBNO(DivingOrDBNO:bool=OR Boolean(A=Get bIsSkyDiving(), B=Get bIsDBNO())) -> (Output_Get:bool)
   > Set DBNO_Alpha(DBNO_Alpha:double=SelectFloat(A=0.95, B=0, bPickA=Get bIsDBNO())) -> (Output_Get:double)
   n15 = Sequence!(execute=@n2.then_4) -> (then_0, then_1, then_2)
   > Is Valid(InputObject=Get PlayerPawn() -> (PlayerPawn:Fort Player Pawn Athena Object Reference))
   > n16 = Branch(Condition=Get bIsInFrontEnd())
   Set WPO Globals Function(
    execute=@n16.else
    Material=Get Mat Instance Dynamic()
    Overlay Material=Get Overlay Mat Instance Dynamic()
    Skeletal Mesh=Get SkelMesh()
    "Local Noise Transform  Location"=Get Local Noise Transform Location() -> (Local Noise Transform Location:Vector)
    Local Noise Offset=Get Local Noise Offset()
   )
   > Set WPO Regional Function!(
    execute
    Material:Material Instance Dynamic Object Reference=Get Mat Instance Dynamic()
    Overlay Material:Material Instance Dynamic Object Reference=Get Overlay Mat Instance Dynamic()
    Skeletal Mesh:Skeletal Mesh Component Object Reference=Get SkelMesh()
    Velocity Transform:Name=root
    Velocity Interp Speed:double=Get Velocity Interp Speed()
    Collision Transform:Name=spine_04
    Collision Transform Axis:EAxis Enum
    Invert Collision Transform Axis:bool=true
    Delta Time X:double=@n1.DeltaTimeX
    Previous Velocity Transform Location:"Vector (by ref)"=Get R1 Velocity Transform Location()
    Previous Velocity Transform Velocity:"Vector (by ref)"=Get R1 Velocity Transform Velocity()
    Region:int=1
   ) -> (then)
   > Set WPO Regional Function(
    Material=Get Mat Instance Dynamic()
    Overlay Material=Get Overlay Mat Instance Dynamic()
    Skeletal Mesh=Get SkelMesh()
    Velocity Transform=root
    Velocity Interp Speed=Get Velocity Interp Speed()
    Collision Transform=pelvis
    Invert Collision Transform Axis=true
    Delta Time X=@n1.DeltaTimeX
    Previous Velocity Transform Location=Get R2 Velocity Transform Location()
    Previous Velocity Transform Velocity=Get R2 Velocity Transform Velocity()
    Region=2
   )
   > n17 = Branch(Condition=Get bIsSkydiveFloating())
   > n18 = Set WPO Pho Wind Intensity(WPO Pho Wind Intensity=1)
   SetScalarParameterValue(execute=[@n19.then, @n18.then], self=Get Mat Instance Dynamic(), ParameterName=ClothSim_WindStrength, Value=Get WPO Pho Wind Intensity())
   n19 = Set WPO Pho Wind Intensity(execute=@n17.else, WPO Pho Wind Intensity=0)
   n20 = Branch(execute=@n15.then_1, Condition=AND Boolean(A=Get bIsParachuteOpened(), B=Get custom_cape_equipped() -> (custom_cape_equipped:bool)))
   > Do Once(Reset=@n20.else)
   > n21 = Cast To FortPlayerPawn(Object=TryGetFortPawn())
   > Is Valid(InputObject=@n21)
   > n22 = Cast To B_Athena_PartModifier_ArcticIce_FaceAcc!(
    execute
    Object=GetModifierForPartType!(self:Fort Player Pawn Object Reference=@n21, PartType:EFortCustomPartType Enum=Face) -> (ReturnValue:Custom Character Part Modifier Object Reference)
   ) -> (then, CastFailed, AsB Athena Part Modifier Arctic Ice Face Acc:B Athena Part Modifier Arctic Ice Face Acc Object Reference) {is=title K2Node_DynamicCast}
   > Set Scarf Visible(Scarf Visible:bool=Get Scarf Visible!(self:B Athena Part Modifier Arctic Ice Face Acc Object Reference=@n22) -> (Scarf Visible:bool)) -> (Output_Get:bool)
   Set pantlegAlphaFrontend(execute=@n15.then_2, pantlegAlphaFrontend:double=SelectFloat(A=0.65, B=1, bPickA=Get bIsInFrontEnd())) -> (Output_Get:double)
   n23 = Gravity Override State Function DEPRECATED(
    execute=@n2.then_1
    CharacterPartAnimInstance=Self-Reference()
    GameTimeInSeconds=GetGameTimeInSeconds()
    GravityOverrideParams=Get Gravity Override Params()
   )
   > Set Gravity Override Cape(Gravity Override Cape:Vector=@n23) -> (Output_Get:Vector)
   Event BlueprintInitializeAnimation()
   > Set SkelMesh(SkelMesh=GetOwnerSkeletalMeshComponent())
   n24 = TryGetPawnOwner()
   > Is Valid(InputObject=@n24)
   > n25 = Cast To FortPlayerPawnAthena(Object=@n24)
   > Set PlayerPawn(PlayerPawn:Fort Player Pawn Athena Object Reference=@n25) -> (Output_Get:Fort Player Pawn Athena Object Reference)
   Event OnParachuteClosed()
   > n26 = Branch(Condition=Get custom_cape_equipped())
   > Set Scarf Visible(Scarf Visible=true)
   Set Scarf Visible(execute=@n26.else, Scarf Visible=false)
   n27 = Event OnCharacterCustomizationLoaded()
   > n28 = Sequence!(execute) -> (then_0, then_1, then_2, then_3:Exec)
   > n29 = Sequence!(execute) -> (then_0, then_1)
   > n30 = Set SkelMesh(SkelMesh=GetPartSkeletalMeshComponent())
   > n31 = Get Dynamic Material Instance(SkeletalMeshComponent=@n30)
   > Set Mat Instance Dynamic(Mat Instance Dynamic=@n31)
   > n32 = GetOrCreateDynamicOverlayMaterial(MeshComponent=@n30, MaterialName=None)
   > Set Overlay Mat Instance Dynamic(Overlay Mat Instance Dynamic=@n32)
   n33 = CheckForBackbling()
   n34 = Branch(execute=@n29.then_1, Condition=@n33.bIsBackblingACape)
   > n35 = Branch(Condition=BackpackItemDefinitionHasAnyMetaTag!(Tags:Array of Gameplay Tag Structures=Make Array("[0]"=(TagName="Cosmetics.MetaTags.ArcticIce.Cape")), bRequiresEquipped:bool=false) -> (ReturnValue:bool))
   > Set custom_cape_equipped(custom_cape_equipped:bool=true) -> (Output_Get:bool)
   > Set Scarf Visible(Scarf Visible=true)
   Set custom_cape_equipped(execute=@n35.else, custom_cape_equipped=false)
   > Set Scarf Visible(Scarf Visible=false)
   Set custom_cape_equipped(execute=@n34.else, custom_cape_equipped=false)
   > Set Scarf Visible(Scarf Visible=false)
   Set Character Mesh(execute=@n28.then_1, Character Mesh:Skeletal Mesh Component Object Reference=@n27.CustomBaseMesh) -> (Output_Get:Skeletal Mesh Component Object Reference)
   Set is_WearingCape(execute=@n28.then_3, is_WearingCape:bool=true) -> (Output_Get:bool)
   > Set isWearingPackBool(isWearingPackBool:bool=OR Boolean!(A=@n33, B=@n33.bIsBackblingACape, C:bool=@n33.bIsBackblingLong) -> (ReturnValue)) -> (Output_Get:bool)
   > Branch(Condition=OR Boolean(A=@n33, B=@n33.bIsBackblingACape, C=@n33.bIsBackblingLong))
   > Set is_WearingCape(is_WearingCape=true)
   > n36 = Branch(Condition=NOT Boolean(A=Get bIsSkyDiving()))
   > Set is_WearingCape(is_WearingCape=true)
   Set is_WearingCape(execute=@n36.else, is_WearingCape=NOT Boolean(A=@n33.bIsBackblingLong))
   Set Body Mesh(
    execute=@n28.then_2
    Body Mesh:Skeletal Mesh Component Object Reference=Get Array Item with EFort Custom Part Type(Array=@n27.CustomMeshParts)
   ) -> (Output_Get:Skeletal Mesh Component Object Reference)
   Set thigh_shape_lt_l(thigh_shape_lt_l:double=GetAlphaFromReferenceTransformDelta(
     SkeletalMeshComponent=@n5
     BoneName=thigh_l
     BaseName=pelvis
     DeltaComponentType=EulerY
     InputRange=(X=130,Y=0)
     OutputRange=(X=1,Y=0)
     TwistAxis="0,0,1"
    )) -> (Output_Get:double)
   > Set thigh_shape_rt_r(thigh_shape_rt_r:double=GetAlphaFromReferenceTransformDelta(
     SkeletalMeshComponent=@n5
     BoneName=thigh_r
     BaseName=pelvis
     DeltaComponentType=EulerY
     InputRange=(X=130,Y=0)
     OutputRange=(X=1,Y=0)
     TwistAxis="0,0,1"
    )) -> (Output_Get:double)
   GetSkeletalMesh!(self:Custom Character Part Object Reference=Get Array Item with EFort Custom Part Type(Array:Array of Custom Character Part Object References=@n27.CustomParts) -> (Output:Custom Character Part Object Reference)) -> (ReturnValue:Skeletal Mesh Object Reference)
   Set is_WearingCape(is_WearingCape=false)
  ]}
]
variables [
 Body Mesh {type=SkeletalMeshComponent}
 Calf Shape Back R {type=number}
 calf_shape_back_l {type=number}
 calf_shape_back_r {type=number}
 Character Mesh {type=SkeletalMeshComponent}
 CrouchingAndNotMovingAndNotShape {type=boolean}
 custom_cape_equipped {type=boolean}
 DBNO_Alpha {type=number}
 DivingAndCape {type=boolean}
 DivingOrDBNO {type=boolean}
 FrontEndCape {type=boolean}
 Gravity Override Cape {type=Vector}
 Gravity Override Params {
  type=GravityOverrideParamsStruct
  default {
   riding_Moving_Targeting {gravityOverride=<0 -200 -800>}
   riding_Standing {gravityOverride=<0 -350 -400>}
   skydive_DivingDown {gravityOverride=<0 0 50>, windAmplitude=<0 0 0>, windFrequency=<0 0 0>}
   skydive_FloatingDown {jointName=pelvis, gravityOverride=<0 0 50>, windAmplitude=<0 0 0>, windFrequency=<0 0 0>, windOffset=<0 0 0>}
   skydive_GoingUp {gravityOverride=<0 -600 -150>}
   skydive_Parachuting {gravityOverride=<0 -500 -800>, windAmplitude=<0 0 0>, windFrequency=<0 0 0>}
   skydive_Surfing {gravityOverride=<0 -500 -980>, windAmplitude=<0 0 0>, windFrequency=<0 0 0>}
  }
 }
 Head_rot_f {type=number}
 is sliding {type=number}
 is_WearingCape {type=boolean}
 isWearingPackBool {type=boolean}
 Local Noise Offset {type=Vector}
 Local Noise Transform Location {type=Vector}
 lowerarm_rot_up_l {type=number}
 lowerarm_rot_up_r {type=number}
 Mat Instance Dynamic {type=MaterialInstanceDynamic}
 MovingAnd NotCrouchNotDivingNotEmote {type=boolean}
 Not DiveFloatingAndDiveMode {type=boolean}
 offset_weapon {type=number}
 Overlay Mat Instance Dynamic {type=MaterialInstanceDynamic}
 pantlegAlphaFrontend {type=number}
 PlayerPawn {type=FortPlayerPawnAthena}
 R1 Velocity Transform Location {type=Vector}
 R1 Velocity Transform Velocity {type=Vector}
 R2 Velocity Transform Location {type=Vector}
 R2 Velocity Transform Velocity {type=Vector}
 Scarf Visible {type=boolean}
 Shoulder_rot_f_l {type=number}
 Shoulder_rot_f_r {type=number}
 SkelMesh {type=SkeletalMeshComponent}
 thigh_shape {type=boolean}
 thigh_shape_fn_l {type=number}
 thigh_shape_fn_r {type=number}
 thigh_shape_fn_r_0 {type=number}
 thigh_shape_fn_r_1 {type=number}
 thigh_shape_lt_l {type=number}
 thigh_shape_rt_r {type=number}
 upperarm_rot_dn_l {type=number}
 upperarm_rot_dn_r {type=number}
 Velocity Interp Speed {type=number, default=30}
 WPO Pho Wind Intensity {type=number}
]
