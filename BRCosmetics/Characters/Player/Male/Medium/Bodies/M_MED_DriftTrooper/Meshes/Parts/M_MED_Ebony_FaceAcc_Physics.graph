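# Physics asset for the face accessory part of the Drift Trooper medium male character, simulating a hood around the head.
# Capsule bodies cover the head and four dynamic hood bones at front, back, left and right.
# Each hood bone is constrained directly to the head with tight swing and twist limits, mirrored between left and right, keeping its motion small.
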PhysicsAsset /BRCosmetics/Characters/Player/Male/Medium/Bodies/M_MED_DriftTrooper/Meshes/Parts/M_MED_Ebony_FaceAcc_Physics
bodies {
 dyn_hood_bk [{shapeType=Capsule, length=5, radius=3, shapeName=dyn_hatTop_capsule, center=<3.2375 0 -3.2375>, rotation=<rot 45 0 0>}]
 dyn_hood_fr [{shapeType=Capsule, length=3, radius=3, shapeName=dyn_hatTop_capsule, center=<4 0 0>, rotation=<rot 90 0 0>}]
 dyn_hood_l [{shapeType=Capsule, length=3, radius=3, shapeName=dyn_hatLeft_capsule, center=<4 0 0>, rotation=<rot 90 0 0>}]
 dyn_hood_r [{shapeType=Capsule, length=3, radius=3, shapeName=dyn_hatLeft_capsule, center=<4 0 0>, rotation=<rot 90 0 0>}]
 head [{shapeType=Capsule, length=6, radius=7.505, shapeName=head_capsule, center=<1.679 1.7907 -0.0094>, rotation=<rot -0.2919 -24.0299 90.1301>}]
}
constraints [
 {bone1Name=dyn_hood_l, bone2Name=head, swing1LimitDegrees=20, swing2LimitDegrees=7, twistLimitDegrees=3}
 {bone1Name=dyn_hood_r, bone2Name=head, swing1LimitDegrees=20, swing2LimitDegrees=7, twistLimitDegrees=3}
 {bone1Name=dyn_hood_bk, bone2Name=head, swing1LimitDegrees=23, swing2LimitDegrees=7, twistLimitDegrees=1}
 {bone1Name=dyn_hood_fr, bone2Name=head, swing1LimitDegrees=3, swing2LimitDegrees=15, twistLimitDegrees=2}
]
