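# Post-process style AnimBlueprint for the DriftTrooper face accessory, animating its hood and spine helper bones over the pose copied from the attached parent.
# Constrained helper bones feed a cached pose, with RigidBody physics when not targeting and none when targeting.
# Both branches add hood and spine offsets driven by upper-arm rotation for certain weapons.
# The event graph also feeds material WPO wind/motion parameters and tracks cape-dependent scarf visibility.
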AnimBlueprint /BRCosmetics/Characters/Player/Male/Medium/Bodies/M_MED_DriftTrooper/Meshes/Parts/M_MED_DriftTrooper_FaceAcc_AnimBP
target_skeleton=&/M_MED_DriftTrooper_FaceAcc_Skeleton
graphs [
 AnimGraph {nodes [
   n1 = RigidBody!(
    ComponentPose=Local To Component(LocalPose=Use cached pose 'ADSPose'())
    Alpha=Get ADS() -> (ADS:double)
    OverrideWorldGravity:Vector="0,0,-980"
    ExternalForce
   ) -> (Pose) {
    bClampLinearTranslationLimitToRefPose=true
    bOverrideWorldGravity=true
    overridePhysicsAsset=&/M_MED_Ebony_FaceAcc_Physics
    simulationSpace=BaseBoneSpace
    alphaBoolBlend {blendOption=Linear}
    alphaScaleBiasClamp {interpSpeedDecreasing=4, interpSpeedIncreasing=4}
    baseBoneRef {boneName=spine_05}
    componentAppliedLinearAccClamp=<6000 6000 6000>
    componentLinearAccScale=<0.2 0.2 0.1>
    componentLinearVelScale=<0.2 0.2 0.1>
    overrideWorldGravity {z=-980}
   }
   n2 = Get upperarm_rot_dn_r() -> (upperarm_rot_dn_r:double)
   n3 = "Transform (Modify) Bone - Bone: fix_spine_r_02"(ComponentPose=@n1, Alpha=@n2, Translation="0,0,-10", Rotation="0,0,0", Scale="1,1,1") {
    rotationMode=BMM_Additive
    rotationSpace=BCS_BoneSpace
    scaleMode=BMM_Additive
    scaleSpace=BCS_BoneSpace
    translationMode=BMM_Additive
    translationSpace=BCS_BoneSpace
    alphaBoolBlend {blendOption=Linear}
    boneToModify {boneName=fix_spine_r_02}
    translation {z=-10}
   }
   n4 = "Transform (Modify) Bone - Bone: fix_spine_r_01"(ComponentPose=@n3, Alpha=@n2, Translation="0,0,-10", Rotation="0,0,0", Scale="1,1,1") {
    rotationMode=BMM_Additive
    rotationSpace=BCS_BoneSpace
    scaleMode=BMM_Additive
    scaleSpace=BCS_BoneSpace
    translationMode=BMM_Additive
    translationSpace=BCS_BoneSpace
    alphaBoolBlend {blendOption=Linear}
    boneToModify {boneName=fix_spine_r_01}
    translation {z=-10}
   }
   n5 = "Transform (Modify) Bone - Bone: dyn_hood_r"(ComponentPose=@n4, Alpha=@n2, Translation="0,0,0", Rotation="0,-30,10", Scale="1,1,1") {
    rotationMode=BMM_Additive
    rotationSpace=BCS_BoneSpace
    scaleMode=BMM_Additive
    scaleSpace=BCS_BoneSpace
    translationMode=BMM_Additive
    translationSpace=BCS_BoneSpace
    alphaBoolBlend {blendOption=Linear}
    boneToModify {boneName=dyn_hood_r}
    rotation {roll=10, yaw=-30}
   }
   n6 = "Transform (Modify) Bone - Bone: dyn_hood_l"(ComponentPose=@n5, Alpha=Get upperarm_rot_dn_l() -> (upperarm_rot_dn_l:double), Translation="0,1,2", Rotation="0,10,-10", Scale="1,1,1") {
    rotationMode=BMM_Additive
    rotationSpace=BCS_BoneSpace
    scaleMode=BMM_Additive
    scaleSpace=BCS_BoneSpace
    translationMode=BMM_Additive
    translationSpace=BCS_BoneSpace
    alphaBoolBlend {blendOption=Linear}
    boneToModify {boneName=dyn_hood_l}
    rotation {roll=-10, yaw=10}
    translation {y=1, z=2}
   }
   n7 = Get upperarm_rot_dn_r()
   n8 = "Transform (Modify) Bone - Bone: fix_spine_r_02"(ComponentPose=Local To Component(LocalPose=Use cached pose 'ADSPose'()), Alpha=@n7, Translation="0,0,-10", Rotation="0,0,0", Scale="1,1,1") {
    rotationMode=BMM_Additive
    rotationSpace=BCS_BoneSpace
    scaleMode=BMM_Additive
    scaleSpace=BCS_BoneSpace
    translationMode=BMM_Additive
    translationSpace=BCS_BoneSpace
    alphaBoolBlend {blendOption=Linear}
    boneToModify {boneName=fix_spine_r_02}
    translation {z=-10}
   }
   n9 = "Transform (Modify) Bone - Bone: fix_spine_r_01"(ComponentPose=@n8, Alpha=@n7, Translation="0,0,-10", Rotation="0,0,0", Scale="1,1,1") {
    rotationMode=BMM_Additive
    rotationSpace=BCS_BoneSpace
    scaleMode=BMM_Additive
    scaleSpace=BCS_BoneSpace
    translationMode=BMM_Additive
    translationSpace=BCS_BoneSpace
    alphaBoolBlend {blendOption=Linear}
    boneToModify {boneName=fix_spine_r_01}
    translation {z=-10}
   }
   n10 = "Transform (Modify) Bone - Bone: dyn_hood_r"(ComponentPose=@n9, Alpha=@n7, Translation="0,0,0", Rotation="0,-30,10", Scale="1,1,1") {
    rotationMode=BMM_Additive
    rotationSpace=BCS_BoneSpace
    scaleMode=BMM_Additive
    scaleSpace=BCS_BoneSpace
    translationMode=BMM_Additive
    translationSpace=BCS_BoneSpace
    alphaBoolBlend {blendOption=Linear}
    boneToModify {boneName=dyn_hood_r}
    rotation {roll=10, yaw=-30}
   }
   n11 = "Transform (Modify) Bone - Bone: dyn_hood_l"(ComponentPose=@n10, Alpha=Get upperarm_rot_dn_l(), Translation="0,1,2", Rotation="0,10,-10", Scale="1,1,1") {
    rotationMode=BMM_Additive
    rotationSpace=BCS_BoneSpace
    scaleMode=BMM_Additive
    scaleSpace=BCS_BoneSpace
    translationMode=BMM_Additive
    translationSpace=BCS_BoneSpace
    alphaBoolBlend {blendOption=Linear}
    boneToModify {boneName=dyn_hood_l}
    rotation {roll=-10, yaw=10}
    translation {y=1, z=2}
   }
   Output Pose(Result=Blend Poses by bool(BlendPose_0=Component To Local(ComponentPose=@n11), BlendPose_1=Component To Local(ComponentPose=@n6), bActiveValue=Get bIsTargeting()) {blendType=Linear}) {layerGroup=None}
   n12 = "Constraint - fix_spine_l_02 (PPP)"!(
    ComponentPose=Local To Component(LocalPose=Copy Pose From Mesh() {bUseAttachedParent=true})
    Alpha=1
    ConstraintWeights_0=0.4
    ConstraintWeights_1:float=0.4
    ConstraintWeights_2:float=0.2
   ) -> (Pose) {
    alphaBoolBlend {blendOption=Linear}
    boneToModify {boneName=fix_spine_l_02}
    constraintSetup [{targetBone {boneName=spine_05}}, {targetBone {boneName=clavicle_l}}, {targetBone {boneName=head}}]
    constraintWeights [1,1,1]
   }
   n13 = "Constraint - fix_spine_r_02 (PPP)"!(ComponentPose=@n12, Alpha=1, ConstraintWeights_0=0.4, ConstraintWeights_1:float=0.4, ConstraintWeights_2:float=0.2) -> (Pose) {
    alphaBoolBlend {blendOption=Linear}
    boneToModify {boneName=fix_spine_r_02}
    constraintSetup [{targetBone {boneName=spine_05}}, {targetBone {boneName=clavicle_r}}, {targetBone {boneName=head}}]
    constraintWeights [1,1,1]
   }
   Save cached pose 'ADSPose'(Pose=Component To Local(ComponentPose=@n13)) {cacheName=ADSPose}
  ]}
 EventGraph {nodes [
   n1 = Event BlueprintUpdateAnimation()
   > n2 = Sequence!(execute) -> (then_0, then_1, then_2:Exec, then_3:Exec)
   Set ADS(execute=@n2.then_1, ADS:double=SelectFloat(A=0.1, B=1, bPickA=Get bIsTargeting())) -> (Output_Get:double)
   n3 = Sequence!(execute=@n2.then_2) -> (then_0, then_1, then_2)
   > Is Valid(InputObject=Get PlayerPawn() -> (PlayerPawn:Fort Player Pawn Athena Object Reference))
   > n4 = Branch(Condition=Get bIsInFrontEnd())
   Set WPO Globals Function(
    execute=@n4.else
    Material=Get Mat Instance Dynamic()
    Overlay Material=Get Overlay Mat Instance Dynamic()
    Skeletal Mesh=Get SkelMesh()
    "Local Noise Transform  Location"=Get Local Noise Transform Location() -> (Local Noise Transform Location:Vector)
    Local Noise Offset=Get Local Noise Offset()
   )
   > Set WPO Regional Function!(
    execute
    Material:Material Instance Dynamic Object Reference=Get Mat Instance Dynamic()
    Overlay Material:Material Instance Dynamic Object Reference=Get Overlay Mat Instance Dynamic()
    Skeletal Mesh:Skeletal Mesh Component Object Reference=Get SkelMesh()
    Velocity Transform:Name=root
    Velocity Interp Speed:double=Get Velocity Interp Speed()
    Collision Transform:Name=head
    Collision Transform Axis:EAxis Enum
    Invert Collision Transform Axis:bool=true
    Delta Time X:double=@n1.DeltaTimeX
    Previous Velocity Transform Location:"Vector (by ref)"=Get R1 Velocity Transform Location()
    Previous Velocity Transform Velocity:"Vector (by ref)"=Get R1 Velocity Transform Velocity()
    Region:int=1
   ) -> (then)
   > Set WPO Regional Function(
    Material=Get Mat Instance Dynamic()
    Overlay Material=Get Overlay Mat Instance Dynamic()
    Skeletal Mesh=Get SkelMesh()
    Velocity Transform=root
    Velocity Interp Speed=Get Velocity Interp Speed()
    Collision Transform=pelvis
    Invert Collision Transform Axis=true
    Delta Time X=@n1.DeltaTimeX
    Previous Velocity Transform Location=Get R2 Velocity Transform Location()
    Previous Velocity Transform Velocity=Get R2 Velocity Transform Velocity()
    Region=2
   )
   > n5 = Branch(Condition=Get bIsSkydiveFloating())
   > n6 = Set WPO Pho Wind Intensity(WPO Pho Wind Intensity=1)
   SetScalarParameterValue(execute=[@n7.then, @n6.then], self=Get Mat Instance Dynamic(), ParameterName=ClothSim_WindStrength, Value=Get WPO Pho Wind Intensity())
   n7 = Set WPO Pho Wind Intensity(execute=@n5.else, WPO Pho Wind Intensity=0)
   n8 = Branch(execute=@n3.then_1, Condition=AND Boolean(A=Get bIsParachuteOpened(), B=Get custom_cape_equipped() -> (custom_cape_equipped:bool)))
   > Do Once(Reset=@n8.else)
   > n9 = Cast To FortPlayerPawn(Object=TryGetFortPawn())
   > Is Valid(InputObject=@n9)
   > n10 = Cast To B_Athena_PartModifier_ArcticIce_FaceAcc!(
    execute
    Object=GetModifierForPartType!(self:Fort Player Pawn Object Reference=@n9, PartType:EFortCustomPartType Enum=Face) -> (ReturnValue:Custom Character Part Modifier Object Reference)
   ) -> (then, CastFailed, AsB Athena Part Modifier Arctic Ice Face Acc:B Athena Part Modifier Arctic Ice Face Acc Object Reference) {is=title K2Node_DynamicCast}
   > Set Scarf Visible(Scarf Visible:bool=Get Scarf Visible!(self:B Athena Part Modifier Arctic Ice Face Acc Object Reference=@n10) -> (Scarf Visible:bool)) -> (Output_Get:bool)
   Set pantlegAlphaFrontend(execute=@n3.then_2, pantlegAlphaFrontend:double=SelectFloat(A=0.65, B=1, bPickA=Get bIsInFrontEnd())) -> (Output_Get:double)
   n11 = Gravity Override State Function DEPRECATED(
    execute=@n2.then_0
    CharacterPartAnimInstance=Self-Reference()
    GameTimeInSeconds=GetGameTimeInSeconds()
    GravityOverrideParams=Get Gravity Override Params()
   )
   > Set Gravity Override Cape(Gravity Override Cape:Vector=@n11) -> (Output_Get:Vector)
   > n12 = Cast To FortPlayerAnimInstance(Object=GetAnimInstance(self=GetOwnerSkeletalMeshComponent()))
   > Set upperarm_rot_dn_r(upperarm_rot_dn_r:double=0) -> (Output_Get:double)
   > n13 = Switch on EFortWeaponCoreAnimation!(execute, Selection:EFortWeaponCoreAnimation Enum=Get WeaponCoreAnim(self=@n12)) -> (
    Melee:Exec
    Pistol:Exec
    Shotgun:Exec
    PaperBlueprint:Exec
    Rifle:Exec
    MeleeOneHand:Exec
    MachinePistol:Exec
    RocketLauncher:Exec
    GrenadeLauncher:Exec
    GoingCommando:Exec
    AssaultRifle:Exec
    TacticalShotgun:Exec
    SniperRifle:Exec
    TrapPlacement:Exec
    ShoulderLauncher:Exec
    AbilityDecoTool:Exec
    Crossbow:Exec
    C4:Exec
    RemoteControl:Exec
    DualWield:Exec
    AR_BullPup:Exec
    AR_ForwardGrip:Exec
    MedPackPaddles:Exec
    SMG_P90:Exec
    AR_DrumGun:Exec
    Consumable_Small:Exec
    Consumable_Large:Exec
    Balloon:Exec
    MountedTurret:Exec
    CreativeTool:Exec
    ExplosiveBow:Exec
    AshtonIndigo:Exec
    AshtonChicago:Exec
    MeleeDualWield:Exec
    Unarmed:Exec
   )
   n14 = Branch(
    execute=[@n13.AR_BullPup, @n13.AssaultRifle, @n13.Pistol, @n13.Rifle, @n13.Shotgun, @n13.SniperRifle, @n13.TacticalShotgun]
    Condition=OR Boolean(A=Get bIsPlayingEmote(), B=Get bIsSkyDiving())
   )
   > Set upperarm_rot_dn_r(upperarm_rot_dn_r=0)
   > Set upperarm_rot_dn_l(upperarm_rot_dn_l:double=0) -> (Output_Get:double)
   n15 = Get Character Mesh() -> (Character Mesh:Skeletal Mesh Component Object Reference)
   Set upperarm_rot_dn_r(
    execute=@n14.else
    upperarm_rot_dn_r=GetAlphaFromReferenceTransformDelta(SkeletalMeshComponent=@n15, BoneName=spine_05, BaseName=upperarm_r, InputRange=(X=0,Y=90), TwistAxis="0,0,1")
   )
   > Set upperarm_rot_dn_l(upperarm_rot_dn_l=GetAlphaFromReferenceTransformDelta(SkeletalMeshComponent=@n15, BoneName=spine_05, BaseName=upperarm_l, InputRange=(X=0,Y=90), TwistAxis="0,0,1"))
   Event BlueprintInitializeAnimation()
   > Set SkelMesh(SkelMesh=GetOwnerSkeletalMeshComponent())
   n16 = TryGetPawnOwner()
   > Is Valid(InputObject=@n16)
   > n17 = Cast To FortPlayerPawnAthena(Object=@n16)
   > Set PlayerPawn(PlayerPawn:Fort Player Pawn Athena Object Reference=@n17) -> (Output_Get:Fort Player Pawn Athena Object Reference)
   n18 = Event OnCharacterCustomizationLoaded()
   > n19 = Sequence()
   > Set Character Mesh(Character Mesh:Skeletal Mesh Component Object Reference=@n18.CustomBaseMesh) -> (Output_Get:Skeletal Mesh Component Object Reference)
   n20 = Sequence!(execute=@n19.then_2) -> (then_0, then_1)
   > n21 = Set SkelMesh(SkelMesh=GetPartSkeletalMeshComponent())
   > n22 = Get Dynamic Material Instance(SkeletalMeshComponent=@n21)
   > Set Mat Instance Dynamic(Mat Instance Dynamic=@n22)
   > n23 = GetOrCreateDynamicOverlayMaterial(MeshComponent=@n21, MaterialName=None)
   > Set Overlay Mat Instance Dynamic(Overlay Mat Instance Dynamic=@n23)
   n24 = Branch(execute=@n20.then_1)
   > n25 = Branch(Condition=BackpackItemDefinitionHasAnyMetaTag!(Tags:Array of Gameplay Tag Structures=Make Array("[0]"=(TagName="Cosmetics.MetaTags.ArcticIce.Cape")), bRequiresEquipped:bool=false) -> (ReturnValue:bool))
   > Set custom_cape_equipped(custom_cape_equipped:bool=true) -> (Output_Get:bool)
   > Set Scarf Visible(Scarf Visible=true)
   Set custom_cape_equipped(execute=@n25.else, custom_cape_equipped=false)
   > Set Scarf Visible(Scarf Visible=false)
   Set custom_cape_equipped(execute=@n24.else, custom_cape_equipped=false)
   > Set Scarf Visible(Scarf Visible=false)
   Set Body Mesh(
    execute=@n19.then_1
    Body Mesh:Skeletal Mesh Component Object Reference=Get Array Item with EFort Custom Part Type(Array=@n18.CustomMeshParts)
   ) -> (Output_Get:Skeletal Mesh Component Object Reference)
   Event OnParachuteClosed()
   > n26 = Branch(Condition=Get custom_cape_equipped())
   > Set Scarf Visible(Scarf Visible=true)
   Set Scarf Visible(execute=@n26.else, Scarf Visible=false)
   GetSkeletalMesh!(self:Custom Character Part Object Reference=Get Array Item with EFort Custom Part Type(Array:Array of Custom Character Part Object References=@n18.CustomParts) -> (Output:Custom Character Part Object Reference)) -> (ReturnValue:Skeletal Mesh Object Reference)
   Get Character Mesh()
   Get Character Mesh()
   Get Character Mesh()
  ]}
]
variables [
 ADS {type=number}
 Body Mesh {type=SkeletalMeshComponent}
 Character Mesh {type=SkeletalMeshComponent}
 custom_cape_equipped {type=boolean}
 Gravity Override Cape {type=Vector}
 Gravity Override Params {
  type=GravityOverrideParamsStruct
  default {
   riding_Moving_Targeting {gravityOverride=<0 -200 -800>}
   riding_Standing {gravityOverride=<0 -350 -400>}
   skydive_DivingDown {gravityOverride=<0 0 300>, windAmplitude=<0 0 0>, windFrequency=<0 0 0>}
   skydive_FloatingDown {jointName=pelvis, gravityOverride=<0 0 200>, windAmplitude=<0 0 0>, windFrequency=<0 0 0>, windOffset=<0 0 0>}
   skydive_GoingUp {gravityOverride=<0 -600 -150>}
   skydive_Parachuting {gravityOverride=<0 -500 -800>, windAmplitude=<0 0 0>, windFrequency=<0 0 0>}
   skydive_Surfing {gravityOverride=<0 -500 -980>, windAmplitude=<0 0 0>, windFrequency=<0 0 0>}
  }
 }
 Local Noise Offset {type=Vector}
 Local Noise Transform Location {type=Vector}
 Mat Instance Dynamic {type=MaterialInstanceDynamic}
 Overlay Mat Instance Dynamic {type=MaterialInstanceDynamic}
 pantlegAlphaFrontend {type=number}
 PlayerPawn {type=FortPlayerPawnAthena}
 R1 Velocity Transform Location {type=Vector}
 R1 Velocity Transform Velocity {type=Vector}
 R2 Velocity Transform Location {type=Vector}
 R2 Velocity Transform Velocity {type=Vector}
 Scarf Visible {type=boolean}
 SkelMesh {type=SkeletalMeshComponent}
 upperarm_rot_dn_l {type=number}
 upperarm_rot_dn_r {type=number}
 Velocity Interp Speed {type=number, default=30}
 WPO Pho Wind Intensity {type=number}
]
